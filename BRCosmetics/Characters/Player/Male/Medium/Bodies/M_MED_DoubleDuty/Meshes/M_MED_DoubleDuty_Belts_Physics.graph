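# Physics asset for the front and back dangling belts on the male medium DoubleDuty cosmetic body.
# Capsule bodies form two constrained chains from base belt bones, six dynamic links in front and three behind, many using paired side-by-side capsules.
# The second front link is angularly locked to the first.
# Left thigh, calf and front loincloth capsules have no constraints, apparently serving as colliders for the belts.
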PhysicsAsset /BRCosmetics/Characters/Player/Male/Medium/Bodies/M_MED_DoubleDuty/Meshes/M_MED_DoubleDuty_Belts_Physics
bodies {
 base_belt_01_bck [{shapeType=Capsule, length=0.505, radius=0.505, shapeName=base_belt_01_bck_capsule, rotation=<rot 0 -90 0>}]
 base_belt_01_fr [{shapeType=Capsule, length=2, radius=1, shapeName=base_belt_01_fr_capsule, rotation=<rot 90 0 0>}]
 calf_l [{shapeType=Capsule, length=21.7871, radius=5.9839, shapeName=calf_l_capsule, center=<-21.1193 -1.5694 0>, rotation=<rot 90 0 0>}]
 dyn_belt_01_bck [
  {shapeType=Capsule, length=8.1228, radius=1.0772, shapeName=dyn_belt_01_bck_capsule, center=<4.0475 1.5427 0.8506>, rotation=<rot 90 0 0>}
  {shapeType=Capsule, length=8.1228, radius=1.0772, shapeName=dyn_belt_01_bck_capsule, center=<3.9465 -1.9087 0.8506>, rotation=<rot 90 0 0>}
 ]
 dyn_belt_01_fr [{shapeType=Capsule, length=5, radius=2, shapeName=dyn_belt_01_fr_capsule, center=<0 0 0.9062>, rotation=<rot 0 0 90>}]
 dyn_belt_02_bck [
  {shapeType=Capsule, length=12.544, radius=1.5894, shapeName=dyn_belt_02_bck_capsule, center=<4.0475 1.5427 0.8506>, rotation=<rot 90 0 0>}
  {shapeType=Capsule, length=12.544, radius=1.5894, shapeName=dyn_belt_02_bck_capsule, center=<3.9465 -1.9087 0.8506>, rotation=<rot 90 0 0>}
 ]
 dyn_belt_02_fr [{shapeType=Capsule, length=3, radius=2, shapeName=dyn_belt_02_fr_capsule, rotation=<rot 0 0 90>}]
 dyn_belt_03_bck [
  {
   shapeType=Capsule
   length=6.804
   radius=2.396
   shapeName=dyn_belt_03_bck_capsule
   center=<3.9493 1.5427 -0.2699>
   rotation=<rot 86.0763 -0.0001 -0.0002>
  }
  {
   shapeType=Capsule
   length=6.804
   radius=2.396
   shapeName=dyn_belt_03_bck_capsule
   center=<3.8483 -1.9087 -0.2699>
   rotation=<rot 86.0763 -0.0001 -0.0002>
  }
 ]
 dyn_belt_03_fr [
  {shapeType=Capsule, length=10.3437, radius=1.0956, shapeName=dyn_belt_03_fr_capsule, center=<5.6145 1.5427 1.2843>, rotation=<rot 90 0 0>}
  {shapeType=Capsule, length=10.3437, radius=1.0956, shapeName=dyn_belt_03_fr_capsule, center=<5.5135 -1.9087 1.2843>, rotation=<rot 90 0 0>}
 ]
 dyn_belt_04_fr [
  {shapeType=Capsule, length=12.8634, radius=3, shapeName=dyn_belt_04_fr_capsule, center=<4.0475 1.5 0.9065>, rotation=<rot 90 0 0>}
  {shapeType=Capsule, length=12.8634, radius=3, shapeName=dyn_belt_04_fr_capsule, center=<3.9465 -1.9 0.9065>, rotation=<rot 90 0 0>}
 ]
 dyn_belt_05_fr [
  {shapeType=Capsule, length=11.1865, radius=3, shapeName=dyn_belt_05_fr_capsule, center=<4.0475 1.5427 0.8294>, rotation=<rot 90 0 0>}
  {shapeType=Capsule, length=11.1865, radius=3, shapeName=dyn_belt_05_fr_capsule, center=<3.9465 -1.9087 0.8294>, rotation=<rot 90 0 0>}
 ]
 dyn_belt_06_fr [
  {shapeType=Capsule, length=9.5835, radius=5, shapeName=dyn_belt_06_fr_capsule, center=<4.0869 1.5427 0.0586>, rotation=<rot 90 0 0>}
  {shapeType=Capsule, length=9.5835, radius=5, shapeName=dyn_belt_06_fr_capsule, center=<3.9859 -1.9087 0.0586>, rotation=<rot 90 0 0>}
 ]
 dyn_loinClothFront_l_01 [{
   shapeType=Capsule
   length=32.9175
   radius=4.6171
   shapeName=dyn_loinClothFront_l_01_capsule
   center=<18.4654 0 -2.7568>
   rotation=<rot 88.7746 -179.9999 -180>
  }]
 dyn_loinClothFront_l_03 [{
   shapeType=Capsule
   length=15.0914
   radius=4.3088
   shapeName=dyn_loinClothFront_l_03_capsule
   center=<9.4149 0.9882 -1.2595>
   rotation=<rot 89.0083 -179.9999 -180>
  }]
 thigh_beltBody_l [{shapeType=Capsule, length=45.9518, radius=9.4909, shapeName=thigh_l_capsule, center=<-21.5205 0.6577 -0.2487>, rotation=<rot 90 0 0>}]
}
constraints [
 {bone1Name=dyn_belt_01_fr, bone2Name=base_belt_01_fr, swing1LimitDegrees=0, swing2LimitDegrees=12, twistLimitDegrees=0}
 {bone1Name=dyn_belt_03_fr, bone2Name=dyn_belt_02_fr, swing1LimitDegrees=30, swing2LimitDegrees=30, twistLimitDegrees=30}
 {bone1Name=dyn_belt_04_fr, bone2Name=dyn_belt_03_fr, swing1LimitDegrees=30, swing2LimitDegrees=30, twistLimitDegrees=30}
 {bone1Name=dyn_belt_05_fr, bone2Name=dyn_belt_04_fr, swing1LimitDegrees=30, swing2LimitDegrees=30, twistLimitDegrees=30}
 {bone1Name=dyn_belt_06_fr, bone2Name=dyn_belt_05_fr, swing1LimitDegrees=30, swing2LimitDegrees=30, twistLimitDegrees=30}
 {bone1Name=dyn_belt_01_bck, bone2Name=base_belt_01_bck, swing1LimitDegrees=30, swing2LimitDegrees=30, twistLimitDegrees=5}
 {bone1Name=dyn_belt_02_bck, bone2Name=dyn_belt_01_bck, swing1LimitDegrees=40, swing2LimitDegrees=40, twistLimitDegrees=40}
 {bone1Name=dyn_belt_03_bck, bone2Name=dyn_belt_02_bck, swing1LimitDegrees=40, swing2LimitDegrees=40, twistLimitDegrees=40}
 {
  bone1Name=dyn_belt_02_fr
  bone2Name=dyn_belt_01_fr
  swing1LimitDegrees=45
  swing1Motion=Locked
  swing2LimitDegrees=45
  swing2Motion=Locked
  twistLimitDegrees=45
  twistMotion=Locked
 }
]
