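# Physics asset for the DoubleDuty male medium body, named for crouch walk, covering the front loincloth's dynamic bones and lower body.
# Capsules on pelvis, spine_01, thighs and calves give the body shape, joined by free angular constraints.
# Left and right loincloth chains each run from a pelvis-linked master bone through four segments.
# Their upper joints allow no twist and narrower swing, while the last two segments bend and twist more.
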PhysicsAsset /BRCosmetics/Characters/Player/Male/Medium/Bodies/M_MED_DoubleDuty/Meshes/M_MED_DoubleDuty_CrouchWalk_Physics1
bodies {
 calf_l [{shapeType=Capsule, length=27.4886, radius=6.9623, shapeName=calf_l_capsule, center=<-14.1936 0 0>, rotation=<rot 90 -179.156 -179.156>}]
 calf_r [{shapeType=Capsule, length=27.7433, radius=6.7077, shapeName=calf_l_capsule, center=<14.1936 0 0>, rotation=<rot 90 -179.8417 0.1583>}]
 dyn_loinClothFront_l_01 [{shapeType=Capsule, length=1.6, radius=3.3, shapeName=dyn_loinClothFront_l_01_capsule, center=<1.9849 2.1 -2.2>, rotation=<rot 90 0 0>}]
 dyn_loinClothFront_l_02 [{shapeType=Capsule, length=4, radius=3.3, shapeName=dyn_loinClothFront_l_02_capsule, center=<3.8 2.1 -2.2>, rotation=<rot 90 0 0>}]
 dyn_loinClothFront_l_03 [{shapeType=Capsule, length=9.0173, radius=1.6743, shapeName=dyn_loinClothFront_l_03_capsule, center=<8.1781 0 0>, rotation=<rot 90 0 0>}]
 dyn_loinClothFront_l_04 [{shapeType=Capsule, length=8.2423, radius=3.0405, shapeName=dyn_loinClothFront_l_04_capsule, center=<5.627 0 0>, rotation=<rot 90 0 0>}]
 dyn_loinClothFront_l_MSTR [{shapeType=Capsule, length=1, radius=1, shapeName=dyn_loinClothFront_l_MSTR_capsule}]
 dyn_loinClothFront_r_01 [{shapeType=Capsule, length=1.6, radius=3.3, shapeName=dyn_loinClothFront_r_01_capsule, center=<1.9849 2.1 2.2>, rotation=<rot 90 0 0>}]
 dyn_loinClothFront_r_02 [{shapeType=Capsule, length=4, radius=3.3, shapeName=dyn_loinClothFront_r_02_capsule, center=<3.8 2.1 2.2>, rotation=<rot 90 0 0>}]
 dyn_loinClothFront_r_03 [{shapeType=Capsule, length=9.4775, radius=1.6743, shapeName=dyn_loinClothFront_r_03_capsule, center=<8.1781 0 0>, rotation=<rot 90 0 0>}]
 dyn_loinClothFront_r_04 [{shapeType=Capsule, length=8.703, radius=3.0815, shapeName=dyn_loinClothFront_r_04_capsule, center=<5.627 0 0>, rotation=<rot 90 0 0>}]
 dyn_loinClothFront_r_MSTR [{shapeType=Capsule, length=1, radius=1, shapeName=dyn_loinClothFront_r_MSTR_capsule}]
 pelvis [
  {shapeType=Capsule, length=13.1838, radius=10.1966, shapeName=pelvis_capsule, center=<0 -0.8234 0>}
  {shapeType=Capsule, length=12.414, radius=9.7779, shapeName=pelvis_capsule, center=<-3.905 -1.6865 0>}
 ]
 spine_01 [{shapeType=Capsule, length=0.505, radius=0.505, shapeName=spine_01_capsule}]
 thigh_l [
  {
   shapeType=Capsule
   length=33.5343
   radius=8.9715
   shapeName=thigh_l_capsule
   center=<-14.6636 -2.2956 -0.5616>
   rotation=<rot 88.6088 -5.0861 -1.0015>
  }
  {
   shapeType=Capsule
   length=22.4543
   radius=7.5457
   shapeName=thigh_l_capsule
   center=<-15.2209 -0.6982 -0.0363>
   rotation=<rot 90 180 -172.5581>
  }
 ]
 thigh_r [{
   shapeType=Capsule
   length=32.2921
   radius=9.1973
   shapeName=thigh_l_capsule
   center=<14.6827 1.6365 1.3346>
   rotation=<rot 88.1968 176.6878 0.773>
  }]
}
constraints [
 {
  bone1Name=spine_01
  bone2Name=pelvis
  swing1LimitDegrees=45
  swing1Motion=Free
  swing2LimitDegrees=45
  swing2Motion=Free
  twistLimitDegrees=45
  twistMotion=Free
 }
 {
  bone1Name=thigh_l
  bone2Name=pelvis
  swing1LimitDegrees=45
  swing1Motion=Free
  swing2LimitDegrees=45
  swing2Motion=Free
  twistLimitDegrees=45
  twistMotion=Free
 }
 {
  bone1Name=calf_l
  bone2Name=thigh_l
  swing1LimitDegrees=45
  swing1Motion=Free
  swing2LimitDegrees=45
  swing2Motion=Free
  twistLimitDegrees=45
  twistMotion=Free
 }
 {bone1Name=dyn_loinClothFront_l_MSTR, bone2Name=pelvis, swing1LimitDegrees=45, swing2LimitDegrees=20, twistLimitDegrees=0}
 {bone1Name=dyn_loinClothFront_l_01, bone2Name=dyn_loinClothFront_l_MSTR, swing1LimitDegrees=45, swing2LimitDegrees=20, twistLimitDegrees=0}
 {bone1Name=dyn_loinClothFront_l_02, bone2Name=dyn_loinClothFront_l_01, swing1LimitDegrees=45, swing2LimitDegrees=20, twistLimitDegrees=0}
 {bone1Name=dyn_loinClothFront_l_03, bone2Name=dyn_loinClothFront_l_02, swing1LimitDegrees=45, swing2LimitDegrees=50, twistLimitDegrees=50}
 {bone1Name=dyn_loinClothFront_l_04, bone2Name=dyn_loinClothFront_l_03, swing1LimitDegrees=45, swing2LimitDegrees=50, twistLimitDegrees=50}
 {bone1Name=dyn_loinClothFront_r_MSTR, bone2Name=pelvis, swing1LimitDegrees=45, swing2LimitDegrees=20, twistLimitDegrees=0}
 {bone1Name=dyn_loinClothFront_r_01, bone2Name=dyn_loinClothFront_r_MSTR, swing1LimitDegrees=45, swing2LimitDegrees=20, twistLimitDegrees=0}
 {bone1Name=dyn_loinClothFront_r_02, bone2Name=dyn_loinClothFront_r_01, swing1LimitDegrees=45, swing2LimitDegrees=20, twistLimitDegrees=0}
 {bone1Name=dyn_loinClothFront_r_03, bone2Name=dyn_loinClothFront_r_02, swing1LimitDegrees=45, swing2LimitDegrees=50, twistLimitDegrees=50}
 {bone1Name=dyn_loinClothFront_r_04, bone2Name=dyn_loinClothFront_r_03, swing1LimitDegrees=45, swing2LimitDegrees=50, twistLimitDegrees=50}
 {
  bone1Name=calf_r
  bone2Name=thigh_r
  swing1LimitDegrees=45
  swing1Motion=Free
  swing2LimitDegrees=45
  swing2Motion=Free
  twistLimitDegrees=45
  twistMotion=Free
 }
 {
  bone1Name=thigh_r
  bone2Name=pelvis
  swing1LimitDegrees=45
  swing1Motion=Free
  swing2LimitDegrees=45
  swing2Motion=Free
  twistLimitDegrees=45
  twistMotion=Free
 }
]
