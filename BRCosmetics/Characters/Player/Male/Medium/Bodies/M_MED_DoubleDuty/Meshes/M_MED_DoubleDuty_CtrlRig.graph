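# Corrective Control Rig for the DoubleDuty male medium player body, adjusting deform bones, loincloth, waist robe, left sleeve and belts.
# It runs the shared DeformRig function, then a sequence of bone offsets weighted by deformRig anim attributes.
# Thigh-raise values, scaled by tacticalSprint or NotCrouching, drive loincloth and robe offsets, plus parent constraints.
# A spherical pose reader drives sleeve offsets; crouch and waist-strap chains are left disconnected from execution.
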ControlRigBlueprint /BRCosmetics/Characters/Player/Male/Medium/Bodies/M_MED_DoubleDuty/Meshes/M_MED_DoubleDuty_CtrlRig
graphs [
 RigVMFunctionLibrary {links [], nodes []}
 RigVMFunction_Sequence_1_ContainedGraph {nodes [
   n1 = RigVMFunctionEntryNode()
   RigVMFunction_Sequence = &RigVMFunction_Sequence(ExecuteContext=@n1.ExecuteContext)
   n2 = &RigVMFunction_Sequence(ExecuteContext=@RigVMFunction_Sequence.B)
   n3 = &RigVMFunction_Sequence(ExecuteContext=@n2.B)
   RigVMFunctionReturnNode(A=@RigVMFunction_Sequence.A, B=@n2.A, C=@n3.A, D=@n3.B)
  ]}
 RigVMModel {nodes [
   Multiply = &RigVMFunction_MathDoubleMul(B=1, A=RigVMVariableNode() {variable=crouchMoving}.Value)
   &RigUnit_ModifyTransforms(
    ExecutePin=()
    ItemToModify="((Item=(Type=Bone,Name=\"dyn_loinClothFront_r_MSTR\"),Transform=(Rotation=(X=-0.0113,Y=0.0421,Z=0.2586,W=0.965),Translation=(X=2,Y=0,Z=0),Scale3D=(X=1,Y=1,Z=1))),(Item=(Type=Bone,Name=\"dyn_loinClothFront_r_01\"),Transform=(Rotation=(X=0.2004,Y=0.0045,Z=0.3376 ...(truncated)"
    WeightMinimum=0
    Weight=@Multiply.Result
   )
   > &RigUnit_ModifyTransforms(
    ItemToModify=((Item=(Type=Bone,Name="dyn_loinClothFront_l_MSTR"),Transform=(Rotation=(X=-0.102,Y=-0.082,Z=0.5677,W=0.8128),Translation=(X=2,Y=0,Z=0),Scale3D=(X=1,Y=1,Z=1))))
    WeightMinimum=0
    Weight=&RigVMFunction_MathDoubleMul(B=1, A=RigVMVariableNode() {variable=crouchMoving}.Value).Result
   )
   RigUnit_BeginExecution = &RigUnit_BeginExecution()
   FN_DeformRig_FNC = RigVMFunctionReferenceNode(
    DeformRig_ Curve="(deformRig_l_upperarm_up_alpha_72_#1=true,deformRig_r_upperarm_up_alpha_73_#2=true,deformRig_l_upperarm_fr_alpha_74_#3=False,deformRig_r_upperarm_fr_alpha_75_CB51B80C ...(truncated)"
    DeformRig_Bool="(upperarm_1_#4=True,elbow_3_#5=True,wrist_7_#6=True,groin_8_#7=True,glute_17_#8=false,knee_9_40C408324CA26 ...(truncated)"
    DeformRig_Debug=(upperarm_debug_18_#4=false,wrist_debug_20_#6=False,lat_debug_26_#9=False,groinGlute_debug_27_#7=False)
    DeformRig_Value="(upperarm_up_18_#10=(Rotation=(X=0.0,Y=-0.0,Z=0.0,W=1.0),Translation=(X=0.0,Y=0.0,Z=8.0),Scale3D=(X=0.70,Y=1.0,Z=1.0)),upperarm_fwd_fr_19_#11=(Rotation=(X=0.0,Y=-0.0,Z=0.0,W=1.0),Translation=(X=0.0, ...(truncated)"
    ExecuteContext=@RigUnit_BeginExecution.ExecutePin
   )
   > RigVMFunction_Sequence = RigVMAggregateNode()
   &RigUnit_ModifyTransforms(
    ItemToModify=((Item=(Type=Bone,Name="None"),Transform=(Rotation=(X=0,Y=0,Z=0,W=1),Translation=(X=0,Y=0,Z=0),Scale3D=(X=1,Y=1,Z=1))),(Item=(Type=Bone,Name="deform_upperarm_bk_r"),Transform=(Rotation=(X=0,Y=0,Z=0,W=1),Translation=(X=-5.5,Y=2,Z=3),Scale3D=(X=1,Y=1,Z=1))))
    WeightMinimum=0
    Weight=RigVMDispatchNode(BoneName=root, Default=0, Name=deformRig_r_upperarm_up_alpha) {template="DISPATCH_RigDispatch_GetAnimAttribute(in Name,in BoneName,in Default,out Value,out Found)"}.Value
    ExecutePin=@RigVMFunction_Sequence.A
   )
   &RigUnit_ModifyTransforms(
    ItemToModify=((Item=(Type=Bone,Name="dyn_loinClothFront_r_MSTR"),Transform=(Rotation=(X=0.2453,Y=0.2831,Z=0.7446,W=0.5524),Translation=(X=3.7242,Y=0.3569,Z=2.4229),Scale3D=(X=1,Y=1,Z=1))))
    WeightMinimum=0
    Weight=&RigVMFunction_MathDoubleMul(
     A=RigVMDispatchNode(BoneName=root, Default=0, Name=deformRig_r_thigh_up_alpha) {template="DISPATCH_RigDispatch_GetAnimAttribute(in Name,in BoneName,in Default,out Value,out Found)"}.Value
     B=RigVMVariableNode() {variable=tacticalSprint}.Value
    ).Result
    ExecutePin=@RigVMFunction_Sequence.B
   )
   DISPATCH_RigDispatch_GetAnimAttribute = RigVMDispatchNode(BoneName=root, Default=0, Name=deformRig_l_thigh_up_alpha) {template="DISPATCH_RigDispatch_GetAnimAttribute(in Name,in BoneName,in Default,out Value,out Found)"}
   > &RigUnit_ModifyTransforms(
    ItemToModify=((Item=(Type=Bone,Name="dyn_loinClothFront_l_MSTR"),Transform=(Rotation=(X=-0.3493,Y=-0.4958,Z=0.5988,W=0.5231),Translation=(X=3.9199,Y=-1.0746,Z=-2.5823),Scale3D=(X=1,Y=1,Z=1))))
    WeightMinimum=0
    Weight=&RigVMFunction_MathDoubleMul(A=@DISPATCH_RigDispatch_GetAnimAttribute.Value, B=RigVMVariableNode() {variable=tacticalSprint}.Value).Result
   )
   Remap = &RigVMFunction_MathDoubleRemap(
    SourceMaximum=0.1
    SourceMinimum=0
    TargetMaximum=1
    TargetMinimum=0
    bClamp=False
    Value=RigVMDispatchNode(BoneName=root, Default=0, Name=deformRig_r_thigh_up_alpha) {template="DISPATCH_RigDispatch_GetAnimAttribute(in Name,in BoneName,in Default,out Value,out Found)"}.Value
   )
   > &RigUnit_ParentConstraint(
    AdvancedSettings=(InterpolationType=Average,RotationOrderForFilter=XZY)
    Child=(Type=Bone,Name="dyn_waistRobe_02_fr")
    Filter=(TranslationFilter=(bX=True,bY=True,bZ=True),RotationFilter=(bX=True,bY=true,bZ=True),ScaleFilter=(bX=True,bY=True,bZ=True))
    Parents=((Item=(Type=Bone,Name="deform_groin_r"),Weight=1),(Item=(Type=Bone,Name="pelvis"),Weight=0))
    Weight=@Remap.Result
   )
   > &RigUnit_ParentConstraint(
    AdvancedSettings=(InterpolationType=Average,RotationOrderForFilter=XZY)
    Child=(Type=Bone,Name="base_waistRobe_01_r")
    Parents=((Item=(Type=Bone,Name="dyn_waistRobe_02_fr"),Weight=0.3),(Item=(Type=Bone,Name="thigh_r"),Weight=0.7))
    Weight=1
   )
   n1 = RigVMDispatchNode(BoneName=root, Default=0, Name=deformRig_l_thigh_up_alpha) {template="DISPATCH_RigDispatch_GetAnimAttribute(in Name,in BoneName,in Default,out Value,out Found)"}
   > &RigUnit_ModifyTransforms(
    ItemToModify="((Item=(Type=Bone,Name=\"base_waistRobe_01_fr\"),Transform=(Rotation=(X=-0.0698,Y=0,Z=0,W=0.9976),Translation=(X=0,Y=0.1279,Z=0.5403),Scale3D=(X=1,Y=1,Z=1))),(Item=(Type=Bone,Name=\"dyn_loinClothFront_r_MSTR\"),Transform=(Rotation=(X=0.0435,Y=-0.0038,Z=-0.0871 ...(truncated)"
    WeightMinimum=0
    Weight=&RigVMFunction_MathDoubleMul(
     A=&RigVMFunction_MathDoubleRemap(SourceMaximum=0.1, SourceMinimum=0, TargetMaximum=1, TargetMinimum=0, bClamp=False, Value=@n1.Value).Result
     B=RigVMVariableNode() {variable=NotCrouching}.Value
    ).Result
   )
   > &RigUnit_ModifyTransforms(
    ItemToModify="((Item=(Type=Bone,Name=\"dyn_loinClothFront_r_01\"),Transform=(Rotation=(X=-0.0868,Y=0.0868,Z=0.0076,W=0.9924),Translation=(X=0,Y=0,Z=0),Scale3D=(X=1,Y=1,Z=1))),(Item=(Type=Bone,Name=\"dyn_loinClothFront_r_05\"),Transform=(Rotation=(X=-0.1736,Y=0,Z=0,W=0.9848) ...(truncated)"
    WeightMinimum=0
    Weight=&RigVMFunction_MathDoubleMul(
     A=RigVMDispatchNode(BoneName=root, Default=0, Name=deformRig_l_thigh_up_alpha) {template="DISPATCH_RigDispatch_GetAnimAttribute(in Name,in BoneName,in Default,out Value,out Found)"}.Value
     B=RigVMVariableNode() {variable=NotCrouching}.Value
    ).Result
   )
   > &RigUnit_ModifyTransforms(
    ItemToModify="((Item=(Type=Bone,Name=\"dyn_loinClothFront_l_01\"),Transform=(Rotation=(X=0,Y=-0.1305,Z=0,W=0.9914),Translation=(X=0,Y=0,Z=0),Scale3D=(X=1,Y=1,Z=1))),(Item=(Type=Bone,Name=\"dyn_loinClothFront_l_05\"),Transform=(Rotation=(X=0.4226,Y=0,Z=0,W=0.9063),Translatio ...(truncated)"
    WeightMinimum=0
    Weight=&RigVMFunction_MathDoubleMul(
     A=RigVMDispatchNode(BoneName=root, Default=0, Name=deformRig_r_thigh_up_alpha) {template="DISPATCH_RigDispatch_GetAnimAttribute(in Name,in BoneName,in Default,out Value,out Found)"}.Value
     B=RigVMVariableNode() {variable=NotCrouching}.Value
    ).Result
   )
   n2 = &RigUnit_SphericalPoseReader(
    ActiveRegionScaleFactors=(PositiveWidth=1,NegativeWidth=1,PositiveHeight=1,NegativeHeight=1)
    ActiveRegionSize=0.2
    Debug=(bDrawDebug=false,bDraw2D=False,bDrawLocalAxes=False,DebugScale=25,DebugSegments=20,DebugThickness=0.25)
    DriverAxis=(X=1,Y=0,Z=0)
    DriverItem=(Type=Bone,Name="upperarm_l")
    FalloffRegionScaleFactors=(PositiveWidth=1,NegativeWidth=1,PositiveHeight=1,NegativeHeight=1)
    FalloffSize=0.4
    FlipHeightScaling=False
    FlipWidthScaling=False
    OptionalParentItem=(Type=Bone,Name="None")
    RotationOffset=(X=0,Y=-130,Z=35)
    ExecutePin=@RigVMFunction_Sequence.C
   )
   > &RigUnit_ModifyTransforms(
    ItemToModify="((Item=(Type=Bone,Name=\"base_sleeve_down_l\"),Transform=(Rotation=(X=0,Y=-0.0872,Z=0,W=0.9962),Translation=(X=0,Y=0,Z=-6),Scale3D=(X=1,Y=1,Z=1))),(Item=(Type=Bone,Name=\"base_sleeve_front_l\"),Transform=(Rotation=(X=0,Y=0,Z=0,W=1),Translation=(X=0,Y=-3,Z=0),S ...(truncated)"
    WeightMinimum=0
    Weight=@n2.OutputParam
   )
   > &RigUnit_ModifyTransforms(
    ItemToModify=((Item=(Type=Bone,Name="base_sleeve_front_l"),Transform=(Rotation=(X=0.061,Y=0.0174,Z=-0.0011,W=0.998),Translation=(X=0,Y=0,Z=0),Scale3D=(X=1,Y=1,Z=1))))
    Weight=1
    WeightMinimum=0
   )
   > &RigUnit_ModifyTransforms(
    ItemToModify="((Item=(Type=Bone,Name=\"base_belt_01_fr\"),Transform=(Rotation=(X=0,Y=-0.0872,Z=0,W=0.9962),Translation=(X=0,Y=0,Z=0),Scale3D=(X=1,Y=1,Z=1))),(Item=(Type=Bone,Name=\"base_belt_01_bck\"),Transform=(Rotation=(X=0,Y=-0.0436,Z=0,W=0.999),Translation=(X=-2,Y=0,Z=0 ...(truncated)"
    Weight=1
    WeightMinimum=0
   )
   n3 = &RigUnit_SphericalPoseReader(
    ActiveRegionScaleFactors=(PositiveWidth=1,NegativeWidth=1,PositiveHeight=1,NegativeHeight=1)
    ActiveRegionSize=0.07
    Debug=(bDrawDebug=True,bDraw2D=False,bDrawLocalAxes=False,DebugScale=25,DebugSegments=20,DebugThickness=0.25)
    DriverAxis=(X=1,Y=0,Z=0)
    DriverItem=(Type=Bone,Name="thigh_r")
    ExecutePin=()
    FalloffRegionScaleFactors=(PositiveWidth=0.6,NegativeWidth=1,PositiveHeight=0.7,NegativeHeight=1)
    FalloffSize=0.3
    FlipHeightScaling=False
    FlipWidthScaling=False
    OptionalParentItem=(Type=Bone,Name="None")
    RotationOffset=(X=-117.3866,Y=-2.032,Z=11.7389)
   )
   > &RigUnit_ModifyTransforms(
    ItemToModify="((Item=(Type=Bone,Name=\"dyn_loinClothFront_l_MSTR\"),Transform=(Rotation=(X=0,Y=0,Z=0,W=1),Translation=(X=0,Y=0,Z=-2),Scale3D=(X=1,Y=1,Z=1))),(Item=(Type=Bone,Name=\"dyn_loinClothFront_r_MSTR\"),Transform=(Rotation=(X=0.0872,Y=0,Z=0,W=0.9962),Translation=(X=0 ...(truncated)"
    WeightMinimum=0
    Weight=&RigVMFunction_MathDoubleMul(A=@n3.OutputParam, B=RigVMVariableNode() {variable=Crouch}.Value).Result
   )
   n4 = &RigUnit_SphericalPoseReader(
    ActiveRegionScaleFactors=(PositiveWidth=3.5,NegativeWidth=0.1,PositiveHeight=3,NegativeHeight=3)
    ActiveRegionSize=0.1
    Debug=(bDrawDebug=false,bDraw2D=false,DebugScale=25,DebugSegments=20,DebugThickness=0.25)
    DriverAxis=(X=1,Y=0,Z=0)
    DriverItem=(Type=Bone,Name="thigh_r")
    ExecutePin=()
    FalloffRegionScaleFactors=(PositiveWidth=1,NegativeWidth=1.1,PositiveHeight=1,NegativeHeight=1)
    FalloffSize=0.45
    OptionalParentItem=(Type=Bone,Name="pelvis")
    RotationOffset=(X=-120,Y=180,Z=-17)
   )
   > &RigUnit_ModifyTransforms(
    ItemToModify=((Item=(Type=Bone,Name="WaistStrap_MSTR"),Transform=(Rotation=(X=0,Y=0,Z=0,W=1),Translation=(X=0,Y=0,Z=0),Scale3D=(X=1,Y=1,Z=1))))
    ItemToModify.0.Transform.Rotation=&RigVMFunction_MathQuaternionFromEuler(Euler=(X=45,Y=-45,Z=0), RotationOrder=XYZ).Result
    Weight=@n4.OutputParam
   )
   RigVMRerouteNode(Value=@RigVMFunction_Sequence.D)
  ]}
]
hierarchy [
 root {loc=<0 0 0>, rot=<0 0 0 1>}
 pelvis {parent=root, loc=<0 -0.7071 93.1191>, rot=<0.0127 0.707 -0.0127 -0.707>}
 spine_01 {parent=pelvis, loc=<6.7494 -0.6388 0>, rot=<0 0 0.034 -0.9994>}
 spine_02 {parent=spine_01, loc=<7.2518 -0.035 0>, rot=<0 0 0.0238 -0.9997>}
 spine_03 {parent=spine_02, loc=<6.0104 0 0>, rot=<0 0 0.0149 -0.9999>}
 spine_04 {parent=spine_03, loc=<8.1024 0 0>, rot=<0 0 -0.0132 -0.9999>}
 spine_05 {parent=spine_04, loc=<10.0146 0 0>, rot=<0 0 -0.0015 -1>}
 attach_cape {parent=spine_05, loc=<9.0176 -11.7143 0>, rot=<0.0616 -0.7044 -0.0616 -0.7044>}
 neck_01 {parent=spine_05, loc=<11.3138 0.1529 0>, rot=<0 0 -0.2003 -0.9797>}
 neck_02 {parent=neck_01, loc=<4.5627 0 0>, rot=<0 0 0.0504 -0.9987>}
 head {parent=neck_02, loc=<5.3914 0 0>, rot=<0 0 0.115 -0.9934>}
 attach_backpack {parent=spine_05, loc=<-0.1219 -13.4732 0>, rot=<-0.0093 -0.707 0.0093 -0.707>}
 clavicle_l {parent=spine_05, loc=<4.3575 -0.1466 -2.3223>, rot=<0.0465 -0.7289 -0.009 -0.6829>}
 upperarm_l {parent=clavicle_l, loc=<15.5111 0 0>, rot=<0.0226 -0.3234 -0.0088 -0.9459>}
 lowerarm_l {parent=upperarm_l, loc=<26.6514 0 0>, rot=<0.0391 -0.0099 -0.2453 -0.9686>}
 hand_l {parent=lowerarm_l, loc=<22.3491 0 0>, rot=<-0.7081 -0.0507 0.0975 -0.6975>}
 thumb_01_l {parent=hand_l, loc=<1.3118 -0.3191 -2.459>, rot=<0.6546 -0.4042 0.0664 -0.6354>}
 thumb_02_l {parent=thumb_01_l, loc=<5.9336 -0.0448 0.0401>, rot=<-0.0037 0.0213 0.298 -0.9543>}
 thumb_03_l {parent=thumb_02_l, loc=<2.2223 0.3682 -0.0907>, rot=<0.0118 -0.013 0.0066 -0.9998>}
 index_metacarpal_l {parent=hand_l, loc=<1.6535 0.0183 -2.96>, rot=<0.0158 -0.0124 -0.0173 -0.9996>}
 index_01_l {parent=index_metacarpal_l, loc=<8.3008 0.0357 -0.0524>, rot=<-0.0044 0.0485 0.0924 -0.9945>}
 index_02_l {parent=index_01_l, loc=<3.3587 -0.3004 0.0452>, rot=<-0.002 0.0027 0.3225 -0.9466>}
 index_03_l {parent=index_02_l, loc=<2.5578 -0.123 -0.0252>, rot=<-0.002 0.0067 0.2558 -0.9667>}
 middle_metacarpal_l {parent=hand_l, loc=<2.1856 -0.1725 -0.5984>, rot=<0.006 0.0168 -0.0329 -0.9993>}
 middle_01_l {parent=middle_metacarpal_l, loc=<7.3275 -0.2447 0.0129>, rot=<-0.0208 0.0042 0.0858 -0.9961>}
 middle_02_l {parent=middle_01_l, loc=<4.0123 0.1401 0.1175>, rot=<-0.0218 0.0258 0.3512 -0.9357>}
 middle_03_l {parent=middle_02_l, loc=<2.6077 0.0534 0.1373>, rot=<0.0082 0.0188 0.1708 -0.9851>}
 ring_metacarpal_l {parent=hand_l, loc=<2.2074 -0.3625 1.3638>, rot=<-0.0284 0.0722 -0.0458 -0.9959>}
 ring_01_l {parent=ring_metacarpal_l, loc=<7.0295 -0.6197 -0.1446>, rot=<-0.0489 -0.0252 0.1892 -0.9804>}
 ring_02_l {parent=ring_01_l, loc=<3.6174 0.4647 -0.1027>, rot=<0 0 0.2806 -0.9598>}
 ring_03_l {parent=ring_02_l, loc=<2.4152 0.114 0.0199>, rot=<0 0 0.1189 -0.9929>}
 pinky_metacarpal_l {parent=hand_l, loc=<1.7616 -0.9753 3.3814>, rot=<-0.0397 0.0814 -0.069 -0.9935>}
 pinky_01_l {parent=pinky_metacarpal_l, loc=<7.1709 -0.4831 -0.1094>, rot=<-0.0654 -0.0035 0.2242 -0.9723>}
 pinky_02_l {parent=pinky_01_l, loc=<2.8274 0.0893 -0.1043>, rot=<0.0037 0.0011 0.2854 -0.9584>}
 pinky_03_l {parent=pinky_02_l, loc=<2.2046 0.0307 -0.0078>, rot=<-0.0002 0.0098 0.1445 -0.9895>}
 deform_wrist_dn_base_l {parent=hand_l, loc=<0 -2.2342 0>, rot=<0 0 0 -1>}
 deform_wrist_dn_l {parent=deform_wrist_dn_base_l, loc=<0 0 0>, rot=<0 0 0 1>}
 deform_wrist_up_base_l {parent=hand_l, loc=<0 2.2682 0>, rot=<0 0 0 -1>}
 deform_wrist_up_l {parent=deform_wrist_up_base_l, loc=<0 0 0>, rot=<0 0 0 1>}
 lowerarm_twist_02_l {parent=lowerarm_l, loc=<9.6199 0 0>, rot=<0 0 0 -1>}
 lowerarm_twist_01_l {parent=lowerarm_l, loc=<17.8549 0 0>, rot=<0 0 0 -1>}
 deform_elbow_in_base_l {parent=lowerarm_l, loc=<0.2112 5.259 0.8274>, rot=<0 0 0 -1>}
 deform_elbow_in_l {parent=deform_elbow_in_base_l, loc=<0 0 0>, rot=<0 0 0 1>}
 deform_elbow_out_base_l {parent=lowerarm_l, loc=<0 -4 0>, rot=<0 0 0 -1>}
 deform_elbow_out_l {parent=deform_elbow_out_base_l, loc=<0 0 0>, rot=<0 0 0 1>}
 upperarm_twist_01_l {parent=upperarm_l, loc=<0.0041 0 0>, rot=<0 0 0 1>}
 deform_upperarm_bk_base_l {parent=upperarm_twist_01_l, loc=<1.0813 -7.2193 -1.9453>, rot=<0 0 0 1>}
 deform_upperarm_bk_l {parent=deform_upperarm_bk_base_l, loc=<0 0 0>, rot=<0 0 0 1>}
 deform_upperarm_fr_base_l {parent=upperarm_twist_01_l, loc=<-0.3691 7.0506 -2.765>, rot=<0 0 0 1>}
 deform_upperarm_fr_l {parent=deform_upperarm_fr_base_l, loc=<0 0 0>, rot=<0 0 0 1>}
 deform_upperarm_up_base_l {parent=upperarm_twist_01_l, loc=<-3.2216 -1.1391 3.3311>, rot=<0 0 0 1>}
 deform_upperarm_up_l {parent=deform_upperarm_up_base_l, loc=<0 0 0>, rot=<0 0 0 1>}
 base_sleeve_down_l {parent=upperarm_twist_01_l, loc=<7.9376 1.3947 -12.0065>, rot=<0.9981 0.0173 0.058 -0.0085>}
 dyn_sleeve_down_l {parent=base_sleeve_down_l, loc=<0 0 0>, rot=<0 0 0 1>}
 base_sleeve_front_l {parent=upperarm_twist_01_l, loc=<2.7264 7.881 -4.3217>, rot=<0.6926 0.1128 0.0894 -0.7068>}
 dyn_sleeve_front_l {parent=base_sleeve_front_l, loc=<0 0 0>, rot=<0 0 0 1>}
 base_sleeve_up_l {parent=upperarm_twist_01_l, loc=<-1.8644 -0.9211 3.9042>, rot=<-0.0103 0.1609 -0.0164 -0.9868>}
 dyn_sleeve_up_l {parent=base_sleeve_up_l, loc=<0 0 0>, rot=<0 0 0 1>}
 base_sleeve_back_l {parent=upperarm_twist_01_l, loc=<1.5654 -7.2757 -4.6855>, rot=<-0.7058 0.0967 -0.12 -0.6914>}
 dyn_sleeve_back_l {parent=base_sleeve_back_l, loc=<0 0 0>, rot=<0 0 0 1>}
 upperarm_twist_02_l {parent=upperarm_l, loc=<13.7941 0 0>, rot=<0 0 0 1>}
 clavicle_r {parent=spine_05, loc=<4.3575 -0.1466 2.3223>, rot=<0.7319 0.0465 -0.6798 0.0088>}
 upperarm_r {parent=clavicle_r, loc=<-15.5111 0 0>, rot=<0.0226 -0.3234 -0.0088 -0.9459>}
 lowerarm_r {parent=upperarm_r, loc=<-26.6514 0 0>, rot=<0.0391 -0.0099 -0.2453 -0.9686>}
 hand_r {parent=lowerarm_r, loc=<-22.3491 0 0>, rot=<-0.7081 -0.0507 0.0975 -0.6975>}
 thumb_01_r {parent=hand_r, loc=<-1.3118 0.3191 2.459>, rot=<0.6546 -0.4042 0.0664 -0.6354>}
 thumb_02_r {parent=thumb_01_r, loc=<-5.9336 0.0448 -0.0401>, rot=<-0.0037 0.0213 0.298 -0.9543>}
 thumb_03_r {parent=thumb_02_r, loc=<-2.2223 -0.3682 0.0907>, rot=<0.0118 -0.013 0.0066 -0.9998>}
 index_metacarpal_r {parent=hand_r, loc=<-1.6535 -0.0183 2.96>, rot=<0.0158 -0.0124 -0.0173 -0.9996>}
 index_01_r {parent=index_metacarpal_r, loc=<-8.3008 -0.0357 0.0524>, rot=<-0.0044 0.0485 0.0924 -0.9945>}
 index_02_r {parent=index_01_r, loc=<-3.3587 0.3004 -0.0452>, rot=<-0.002 0.0027 0.3225 -0.9466>}
 index_03_r {parent=index_02_r, loc=<-2.5578 0.123 0.0252>, rot=<-0.002 0.0067 0.2558 -0.9667>}
 middle_metacarpal_r {parent=hand_r, loc=<-2.1856 0.1725 0.5984>, rot=<0.006 0.0168 -0.0329 -0.9993>}
 middle_01_r {parent=middle_metacarpal_r, loc=<-7.3275 0.2447 -0.0129>, rot=<-0.0208 0.0042 0.0858 -0.9961>}
 middle_02_r {parent=middle_01_r, loc=<-4.0123 -0.1401 -0.1175>, rot=<-0.0218 0.0258 0.3512 -0.9357>}
 middle_03_r {parent=middle_02_r, loc=<-2.6077 -0.0534 -0.1373>, rot=<0.0082 0.0188 0.1708 -0.9851>}
 ring_metacarpal_r {parent=hand_r, loc=<-2.2074 0.3625 -1.3638>, rot=<-0.0284 0.0722 -0.0458 -0.9959>}
 ring_01_r {parent=ring_metacarpal_r, loc=<-7.0295 0.6197 0.1446>, rot=<-0.0489 -0.0252 0.1892 -0.9804>}
 ring_02_r {parent=ring_01_r, loc=<-3.6174 -0.4647 0.1027>, rot=<0 0 0.2806 -0.9598>}
 ring_03_r {parent=ring_02_r, loc=<-2.4152 -0.114 -0.0199>, rot=<0 0 0.1189 -0.9929>}
 pinky_metacarpal_r {parent=hand_r, loc=<-1.7616 0.9753 -3.3814>, rot=<-0.0397 0.0814 -0.069 -0.9935>}
 pinky_01_r {parent=pinky_metacarpal_r, loc=<-7.1709 0.4831 0.1094>, rot=<-0.0654 -0.0035 0.2242 -0.9723>}
 pinky_02_r {parent=pinky_01_r, loc=<-2.8274 -0.0893 0.1043>, rot=<0.0037 0.0011 0.2854 -0.9584>}
 pinky_03_r {parent=pinky_02_r, loc=<-2.2046 -0.0307 0.0078>, rot=<-0.0002 0.0098 0.1445 -0.9895>}
 deform_wrist_dn_base_r {parent=hand_r, loc=<-0.0004 2.2345 -0.0001>, rot=<0 0 0 -1>}
 deform_wrist_dn_r {parent=deform_wrist_dn_base_r, loc=<0 0 0>, rot=<0 0 0 1>}
 deform_wrist_up_base_r {parent=hand_r, loc=<0.0003 -2.2684 0.0001>, rot=<0 0 0 -1>}
 deform_wrist_up_r {parent=deform_wrist_up_base_r, loc=<0 0 0>, rot=<0 0 0 1>}
 lowerarm_twist_02_r {parent=lowerarm_r, loc=<-9.6199 0 0>, rot=<0 0 0 -1>}
 lowerarm_twist_01_r {parent=lowerarm_r, loc=<-17.8549 0 0>, rot=<0 0 0 -1>}
 deform_elbow_in_base_r {parent=lowerarm_r, loc=<-0.2111 -5.2591 -0.8275>, rot=<0 0 0 -1>}
 deform_elbow_in_r {parent=deform_elbow_in_base_r, loc=<-0.0001 0 0.0002>, rot=<0 0 0 1>}
 deform_elbow_out_base_r {parent=lowerarm_r, loc=<0 4 -0.0001>, rot=<0 0 0 -1>}
 deform_elbow_out_r {parent=deform_elbow_out_base_r, loc=<0 0 0>, rot=<0 0 0 1>}
 upperarm_twist_01_r {parent=upperarm_r, loc=<-0.0041 0 0>, rot=<0 0 0 -1>}
 deform_upperarm_bk_base_r {parent=upperarm_twist_01_r, loc=<-1.0814 7.2192 1.9455>, rot=<0 0 0 -1>}
 deform_upperarm_bk_r {parent=deform_upperarm_bk_base_r, loc=<0 0 0>, rot=<0 0 0 1>}
 deform_upperarm_fr_base_r {parent=upperarm_twist_01_r, loc=<0.3689 -7.0506 2.7652>, rot=<0 0 0 -1>}
 deform_upperarm_fr_r {parent=deform_upperarm_fr_base_r, loc=<0 0 0>, rot=<0 0 0 1>}
 deform_upperarm_up_base_r {parent=upperarm_twist_01_r, loc=<3.2214 1.1391 -3.3308>, rot=<0 0 0 -1>}
 deform_upperarm_up_r {parent=deform_upperarm_up_base_r, loc=<0 0 0>, rot=<0 0 0 -1>}
 upperarm_twist_02_r {parent=upperarm_r, loc=<-13.7941 0 0>, rot=<0 0 0 -1>}
 deform_pec_aim_base_l {parent=spine_04, loc=<7.8732 12.6388 -2.4123>, rot=<0.0102 -0.4602 0.3979 -0.7936>}
 deform_pec_aim_l {parent=deform_pec_aim_base_l, loc=<0 0 0>, rot=<0 0 0 1>}
 deform_pec_base_l {parent=deform_pec_aim_l, loc=<10.766 0 0>, rot=<0 0 0 1>}
 deform_pec_l {parent=deform_pec_base_l, loc=<0.055 0 0>, rot=<0 0 0 1>}
 deform_pec_aim_base_r {parent=spine_04, loc=<7.8727 12.6388 2.4123>, rot=<-0.4602 -0.0102 0.7936 0.3979>}
 deform_pec_aim_r {parent=deform_pec_aim_base_r, loc=<0 0 0>, rot=<0.0006 -0.0021 0.0022 -1>}
 deform_pec_base_r {parent=deform_pec_aim_r, loc=<-10.7662 -0.0006 -0.0007>, rot=<0 0 0 1>}
 deform_pec_r {parent=deform_pec_base_r, loc=<0.0173 0 0>, rot=<0 0 0 1>}
 base_robe_01_r {parent=spine_04, loc=<0.1051 -1.4018 4.7111>, rot=<0.0003 0 0.9992 -0.0405>}
 dyn_robe_01_r {parent=base_robe_01_r, loc=<0.0004 0.0003 0.0002>, rot=<0.0176 0.0037 0.0207 -0.9996>}
 base_robe_02_l {parent=spine_03, loc=<2.309 -1.6929 -2.2942>, rot=<0.9383 0.0514 0.0184 0.3415>}
 dyn_robe_02_l {parent=base_robe_02_l, loc=<0.0002 -0.0001 0.0002>, rot=<-0.0125 -0.0285 0.044 -0.9985>}
 base_robe_01_l {parent=spine_03, loc=<0.1719 2.1786 -3.2383>, rot=<0.9528 0.0521 -0.0166 -0.2986>}
 dyn_robe_01_l {parent=base_robe_01_l, loc=<0.0025 0.001 -0.0001>, rot=<-0.0185 0.0255 0.0412 -0.9987>}
 base_robe_02_r {parent=spine_03, loc=<-0.7133 1.1583 1.3019>, rot=<0.0003 0 0.9985 -0.0546>}
 dyn_robe_02_r {parent=base_robe_02_r, loc=<0.0014 0.0008 0.0004>, rot=<0.0168 0.0121 0.0233 -0.9995>}
 base_belt_01_fr {parent=spine_01, loc=<1.699 8.32 -8.6822>, rot=<-0.2392 -0.7245 -0.512 -0.3948>}
 dyn_belt_01_fr {parent=base_belt_01_fr, loc=<0 0 0>, rot=<0 0 0 1>}
 dyn_belt_02_fr {parent=dyn_belt_01_fr, loc=<1.213 0 0>, rot=<0.0162 -0.3756 0.027 -0.9262>}
 dyn_belt_03_fr {parent=dyn_belt_02_fr, loc=<3.986 -0.0015 0.0351>, rot=<0.0064 -0.0438 -0.0044 -0.999>}
 dyn_belt_04_fr {parent=dyn_belt_03_fr, loc=<9.924 0 0>, rot=<0.0037 -0.0288 0.0022 -0.9996>}
 dyn_belt_05_fr {parent=dyn_belt_04_fr, loc=<9.776 0 0>, rot=<0.0015 -0.0118 0.0037 -0.9999>}
 dyn_belt_06_fr {parent=dyn_belt_05_fr, loc=<7.6508 0 0>, rot=<0 0 0 1>}
 base_belt_01_bck {parent=spine_01, loc=<-2.781 8.35 -7.768>, rot=<-0.0213 -0.8186 -0.5707 -0.0616>}
 dyn_belt_01_bck {parent=base_belt_01_bck, loc=<0 0 0>, rot=<0 0 0 1>}
 dyn_belt_02_bck {parent=dyn_belt_01_bck, loc=<10.356 0 0>, rot=<0.0021 -0.0322 0.002 -0.9995>}
 dyn_belt_03_bck {parent=dyn_belt_02_bck, loc=<9.499 0 0>, rot=<0.0087 0 0 -1>}
 base_centralBelt_l {parent=spine_01, loc=<0 7.6878 -7.5642>, rot=<-0.0114 0.707 0.707 -0.0114>}
 base_centralBelt_r {parent=spine_01, loc=<0 7.6635 7.5642>, rot=<-0.0114 0.707 0.707 -0.0114>}
 thigh_l {parent=pelvis, loc=<-2.5693 -0.0109 -7.7501>, rot=<0.0012 0.0572 0.0155 -0.9982>}
 calf_l {parent=thigh_l, loc=<-39.0484 0 0>, rot=<0.0006 -0.0161 -0.0391 -0.9991>}
 foot_l {parent=calf_l, loc=<-42.561 0.1077 0.4292>, rot=<0.0418 -0.0455 0.0343 -0.9975>}
 ball_l {parent=foot_l, loc=<-6.1599 13.8264 0.6806>, rot=<0.0001 -0.0001 -0.7063 -0.7079>}
 calf_twist_02_l {parent=calf_l, loc=<-18.8064 -0.0967 0.0397>, rot=<0.0237 0.0095 -0.0035 -0.9997>}
 calf_twist_01_l {parent=calf_l, loc=<-35.813 -0.0967 0.0397>, rot=<0.0237 0.0095 -0.0035 -0.9997>}
 deform_knee_in_base_l {parent=calf_l, loc=<1.298 -5.2957 -0.0496>, rot=<0 0 0 -1>}
 deform_knee_in_l {parent=deform_knee_in_base_l, loc=<0 0 0>, rot=<0 0 0 1>}
 deform_knee_out_base_l {parent=calf_l, loc=<0 4 0>, rot=<0 0 0 -1>}
 deform_knee_out_l {parent=deform_knee_out_base_l, loc=<0 0 0>, rot=<0 0 0 1>}
 thigh_twist_01_l {parent=thigh_l, loc=<-21.6303 0 0>, rot=<0.0013 0 0 -1>}
 deform_groin_base_l {parent=thigh_l, loc=<-2.1908 8.1142 -1.7442>, rot=<0 0 0 -1>}
 deform_groin_l {parent=deform_groin_base_l, loc=<0 0 0>, rot=<0 0 0 1>}
 thigh_r {parent=pelvis, loc=<-2.5693 -0.0109 7.7501>, rot=<0.0572 -0.0012 0.9982 0.0155>}
 calf_r {parent=thigh_r, loc=<39.0484 0 0>, rot=<0.0006 -0.0161 -0.0391 -0.9991>}
 foot_r {parent=calf_r, loc=<42.561 -0.1077 -0.4292>, rot=<0.0418 -0.0455 0.0343 -0.9975>}
 ball_r {parent=foot_r, loc=<6.1599 -13.8264 -0.6806>, rot=<0.0001 -0.0001 -0.7063 -0.7079>}
 calf_twist_02_r {parent=calf_r, loc=<18.8064 0.0967 -0.0397>, rot=<0.0237 0.0095 -0.0035 -0.9997>}
 calf_twist_01_r {parent=calf_r, loc=<35.813 0.0967 -0.0397>, rot=<0.0237 0.0095 -0.0035 -0.9997>}
 deform_knee_in_base_r {parent=calf_r, loc=<-1.298 5.2957 0.0496>, rot=<0 0 0 -1>}
 deform_knee_in_r {parent=deform_knee_in_base_r, loc=<0 0 0>, rot=<0 0 0 1>}
 deform_knee_out_base_r {parent=calf_r, loc=<0 -4 0>, rot=<0 0 0 -1>}
 deform_knee_out_r {parent=deform_knee_out_base_r, loc=<0 0 0>, rot=<0 0 0 1>}
 thigh_twist_01_r {parent=thigh_r, loc=<21.6303 0 0>, rot=<0.0013 0 0 -1>}
 deform_groin_base_r {parent=thigh_r, loc=<2.1908 -8.1142 1.7442>, rot=<0 0 0 -1>}
 deform_groin_r {parent=deform_groin_base_r, loc=<0 0 0>, rot=<0 0 0 1>}
 deform_glute_base_l {parent=pelvis, loc=<-8.8064 -8.8222 -8.4287>, rot=<0.0275 0.0501 -0.4472 -0.8926>}
 deform_glute_l {parent=deform_glute_base_l, loc=<0 0 0>, rot=<0 0 0 1>}
 deform_glute_base_r {parent=pelvis, loc=<-8.8062 -8.8222 8.4287>, rot=<0.0501 -0.0275 0.8926 -0.4472>}
 deform_glute_r {parent=deform_glute_base_r, loc=<0 0 0>, rot=<0 0 0 1>}
 base_waistRobe_01_fr {parent=pelvis, loc=<4.5714 9.6447 -0.0098>, rot=<0.2526 0.6605 0.7002 -0.0988>}
 dyn_waistRobe_01_fr {parent=base_waistRobe_01_fr, loc=<0 0 0>, rot=<0 0 0 1>}
 dyn_waistRobe_02_fr {parent=dyn_waistRobe_01_fr, loc=<6.0655 0 0>, rot=<-0.0454 -0.0732 -0.0493 -0.9951>}
 dyn_waistRobe_03_fr {parent=dyn_waistRobe_02_fr, loc=<11.7245 0 0>, rot=<-0.0521 -0.0695 -0.0392 -0.9954>}
 dyn_waistRobe_04_fr {parent=dyn_waistRobe_03_fr, loc=<9.0424 0 0>, rot=<0 0 0 1>}
 base_waistRobe_01_r {parent=pelvis, loc=<5.5029 0.1829 13.4841>, rot=<0.1158 0.0007 0.9933 -0.006>}
 dyn_waistRobe_01_r {parent=base_waistRobe_01_r, loc=<0 0 0>, rot=<0 0.0093 0 -1>}
 dyn_waistRobe_02_r {parent=dyn_waistRobe_01_r, loc=<5.4483 0 0>, rot=<-0.0003 -0.0001 -0.0013 -1>}
 dyn_waistRobe_03_r {parent=dyn_waistRobe_02_r, loc=<6.5307 0 0>, rot=<0.0002 0.0008 0.0007 -1>}
 dyn_waistRobe_04_r {parent=dyn_waistRobe_03_r, loc=<8.7943 0 0>, rot=<-0.0003 -0.0016 -0.0011 -1>}
 dyn_waistRobe_05_r {parent=dyn_waistRobe_04_r, loc=<7.9964 0 0>, rot=<0 0 0 1>}
 dyn_loinClothFront_r_MSTR {parent=pelvis, loc=<4.8017 8.1336 6.7755>, rot=<-0.0179 -0.9998 0 0>}
 dyn_loinClothFront_r_01 {parent=dyn_loinClothFront_r_MSTR, loc=<0 0 0>, rot=<0 0 -0.0477 -0.9989>}
 dyn_loinClothFront_r_02 {parent=dyn_loinClothFront_r_01, loc=<8.0633 -0.0411 0>, rot=<0 0 0.0245 -0.9997>}
 dyn_loinClothFront_r_03 {parent=dyn_loinClothFront_r_02, loc=<7.6272 0 0>, rot=<0 0 0 1>}
 dyn_loinClothFront_r_04 {parent=dyn_loinClothFront_r_03, loc=<9.2638 -0.1128 0>, rot=<0 0 0.0437 -0.999>}
 dyn_loinClothFront_r_05 {parent=dyn_loinClothFront_r_04, loc=<8.3572 0 0>, rot=<0 0 0 1>}
 dyn_loinClothFront_l_MSTR {parent=pelvis, loc=<4.8017 8.1336 -6.7755>, rot=<-0.0576 -0.9983 0 0>}
 dyn_loinClothFront_l_01 {parent=dyn_loinClothFront_l_MSTR, loc=<0 0 0>, rot=<0 0 -0.008 -1>}
 dyn_loinClothFront_l_02 {parent=dyn_loinClothFront_l_01, loc=<8.0634 -0.0411 0>, rot=<0 0 0.0245 -0.9997>}
 dyn_loinClothFront_l_03 {parent=dyn_loinClothFront_l_02, loc=<7.6273 0 0>, rot=<0 0 0 1>}
 dyn_loinClothFront_l_04 {parent=dyn_loinClothFront_l_03, loc=<9.2637 -0.1128 0>, rot=<0 0 0.0437 -0.999>}
 dyn_loinClothFront_l_05 {parent=dyn_loinClothFront_l_04, loc=<8.3572 0 0>, rot=<0 0 0 1>}
]
variables [
 Crouch {type=double, default=1}
 NotCrouching {type=double, default=0}
 crouchMoving {type=double, default=1}
 isParachute {type=double, default=1}
 runningVar {type=double, default=0}
 tacticalSprint {type=double, default=1}
]
