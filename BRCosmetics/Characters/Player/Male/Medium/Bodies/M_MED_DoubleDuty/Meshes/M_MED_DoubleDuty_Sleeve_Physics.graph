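# Physics asset for the left sleeve of the DoubleDuty male medium cosmetic body.
# Four base bones (front, back, up, down) with tiny capsules join an upper-arm capsule through joints with free swing and twist.
# Each carries a larger dynamic sleeve capsule limited to a narrow single-axis swing with no twist allowance.
# Front and back pieces bend on Swing1, up and down on Swing2.
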PhysicsAsset /BRCosmetics/Characters/Player/Male/Medium/Bodies/M_MED_DoubleDuty/Meshes/M_MED_DoubleDuty_Sleeve_Physics
bodies {
 base_sleeve_back_l [{shapeType=Capsule, length=0.505, radius=0.505, shapeName=base_sleeve_back_l_capsule, rotation=<rot 0 -90 0>}]
 base_sleeve_down_l [{shapeType=Capsule, length=0.505, radius=0.505, shapeName=base_sleeve_down_l_capsule, rotation=<rot 0 -90 0>}]
 base_sleeve_front_l [{shapeType=Capsule, length=0.505, radius=0.505, shapeName=base_sleeve_front_l_capsule, rotation=<rot 0 -90 0>}]
 base_sleeve_up_l [{shapeType=Capsule, length=0.505, radius=0.505, shapeName=base_sleeve_up_l_capsule, rotation=<rot 0 -90 0>}]
 dyn_sleeve_back_l [{
   shapeType=Capsule
   length=7.9883
   radius=2.9159
   shapeName=dyn_sleeve_back_l_capsule
   center=<2.8924 0 0>
   rotation=<rot 84.9552 -179.9985 -179.9985>
  }]
 dyn_sleeve_down_l [{shapeType=Capsule, length=8.3531, radius=3.7942, shapeName=dyn_sleeve_down_l_capsule, center=<2.8924 0 0>, rotation=<rot 90 0 0>}]
 dyn_sleeve_front_l [{shapeType=Capsule, length=9.236, radius=2.9159, shapeName=dyn_sleeve_front_l_capsule, center=<2.8924 0 0>, rotation=<rot 90 0 0>}]
 dyn_sleeve_up_l [{shapeType=Capsule, length=9.7249, radius=2.9159, shapeName=dyn_sleeve_up_l_capsule, center=<2.8924 0 0.6022>, rotation=<rot 90 0 0>}]
 upperarm_l [{
   shapeType=Capsule
   length=22.3052
   radius=4.0862
   shapeName=upperarm_l_capsule
   center=<8.8393 0 -1.6334>
   rotation=<rot 78.705 -179.9998 -179.9998>
  }]
}
constraints [
 {bone1Name=dyn_sleeve_back_l, bone2Name=base_sleeve_back_l, swing1LimitDegrees=10, swing2LimitDegrees=0, twistLimitDegrees=0}
 {bone1Name=dyn_sleeve_down_l, bone2Name=base_sleeve_down_l, swing1LimitDegrees=0, swing2LimitDegrees=10, twistLimitDegrees=0}
 {bone1Name=dyn_sleeve_front_l, bone2Name=base_sleeve_front_l, swing1LimitDegrees=10, swing2LimitDegrees=0, twistLimitDegrees=0}
 {bone1Name=dyn_sleeve_up_l, bone2Name=base_sleeve_up_l, swing1LimitDegrees=0, swing2LimitDegrees=12, twistLimitDegrees=0}
 {
  bone1Name=base_sleeve_down_l
  bone2Name=upperarm_l
  swing1LimitDegrees=45
  swing1Motion=Free
  swing2LimitDegrees=45
  swing2Motion=Free
  twistLimitDegrees=45
  twistMotion=Free
 }
 {
  bone1Name=base_sleeve_front_l
  bone2Name=upperarm_l
  swing1LimitDegrees=45
  swing1Motion=Free
  swing2LimitDegrees=45
  swing2Motion=Free
  twistLimitDegrees=45
  twistMotion=Free
 }
 {
  bone1Name=base_sleeve_up_l
  bone2Name=upperarm_l
  swing1LimitDegrees=45
  swing1Motion=Free
  swing2LimitDegrees=45
  swing2Motion=Free
  twistLimitDegrees=45
  twistMotion=Free
 }
 {
  bone1Name=base_sleeve_back_l
  bone2Name=upperarm_l
  swing1LimitDegrees=45
  swing1Motion=Free
  swing2LimitDegrees=45
  swing2Motion=Free
  twistLimitDegrees=45
  twistMotion=Free
 }
]
