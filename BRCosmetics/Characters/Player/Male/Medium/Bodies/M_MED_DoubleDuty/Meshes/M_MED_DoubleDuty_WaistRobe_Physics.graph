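# Physics asset for the waist robe of the DoubleDuty medium male player cosmetic, defining bodies for its dynamic cloth bones.
# Two capsule chains, front-right with four links and right with five, each hang from a tiny base-bone capsule.
# Joints have narrow swing and zero twist limits, except the looser root of the right chain.
# An unconstrained right-thigh capsule provides a leg shape, and the front-right tip carries two capsules.
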PhysicsAsset /BRCosmetics/Characters/Player/Male/Medium/Bodies/M_MED_DoubleDuty/Meshes/M_MED_DoubleDuty_WaistRobe_Physics
bodies {
 base_waistRobe_01_fr [{shapeType=Capsule, length=0.505, radius=0.505, shapeName=base_waistRobe_01_fr_capsule, rotation=<rot 0 -90 0>}]
 base_waistRobe_01_r [{shapeType=Capsule, length=0.505, radius=0.505, shapeName=base_waistRobe_01_r_capsule, rotation=<rot 0 -90 0>}]
 dyn_waistRobe_01_fr [{shapeType=Capsule, length=5, radius=3, shapeName=dyn_waistRobe_01_fr_capsule, center=<4.3061 0 0>, rotation=<rot 90 0 0>}]
 dyn_waistRobe_01_r [{shapeType=Capsule, length=2.6007, radius=3, shapeName=dyn_waistRobe_01_r_capsule, center=<2.6981 0 1.4555>, rotation=<rot 90 0 0>}]
 dyn_waistRobe_02_fr [{shapeType=Capsule, length=4.3425, radius=3.6575, shapeName=dyn_waistRobe_02_fr_capsule, center=<6.9418 0 0>, rotation=<rot 90 0 0>}]
 dyn_waistRobe_02_r [{shapeType=Capsule, length=4.917, radius=2.3428, shapeName=dyn_waistRobe_02_r_capsule, center=<3.4671 0 1.4555>, rotation=<rot 90 0 0>}]
 dyn_waistRobe_03_fr [{shapeType=Capsule, length=4.3425, radius=3.6575, shapeName=dyn_waistRobe_03_fr_capsule, center=<4.3061 0 -0.4863>, rotation=<rot 90 0 0>}]
 dyn_waistRobe_03_r [{shapeType=Capsule, length=5.6572, radius=2.3428, shapeName=dyn_waistRobe_03_r_capsule, center=<4.1818 0 1.4555>, rotation=<rot 90 0 0>}]
 dyn_waistRobe_04_fr [
  {
   shapeType=Capsule
   length=7.8074
   radius=3.6575
   shapeName=dyn_waistRobe_04_fr_capsule
   center=<4.3746 -1.0574 -1.688>
   rotation=<rot 73.7287 0.0001 0.0001>
  }
  {
   shapeType=Capsule
   length=14.1193
   radius=4.8864
   shapeName=dyn_waistRobe_04_fr_capsule
   center=<9.2798 -1.4663 -9.7818>
   rotation=<rot 19.5033 0.6797 -11.3479>
  }
 ]
 dyn_waistRobe_04_r [{shapeType=Capsule, length=5.6572, radius=2.3428, shapeName=dyn_waistRobe_04_r_capsule, center=<4.1818 0 1.4555>, rotation=<rot 90 0 0>}]
 dyn_waistRobe_05_r [{shapeType=Capsule, length=6.3956, radius=4.5776, shapeName=dyn_waistRobe_05_r_capsule, center=<4.1818 0 1.4555>, rotation=<rot 90 0 0>}]
 thigh_r [{
   shapeType=Capsule
   length=38.4232
   radius=9.6257
   shapeName=thigh_r_capsule
   center=<16.6293 1.2912 1.3511>
   rotation=<rot 83.5111 179.9993 179.9997>
  }]
}
constraints [
 {bone1Name=dyn_waistRobe_01_fr, bone2Name=base_waistRobe_01_fr, swing1LimitDegrees=10, swing2LimitDegrees=1, twistLimitDegrees=0}
 {bone1Name=dyn_waistRobe_02_fr, bone2Name=dyn_waistRobe_01_fr, swing1LimitDegrees=10, swing2LimitDegrees=1, twistLimitDegrees=0}
 {bone1Name=dyn_waistRobe_03_fr, bone2Name=dyn_waistRobe_02_fr, swing1LimitDegrees=10, swing2LimitDegrees=1, twistLimitDegrees=0}
 {bone1Name=dyn_waistRobe_04_fr, bone2Name=dyn_waistRobe_03_fr, swing1LimitDegrees=10, swing2LimitDegrees=1, twistLimitDegrees=0}
 {bone1Name=dyn_waistRobe_01_r, bone2Name=base_waistRobe_01_r, swing1LimitDegrees=10, swing2LimitDegrees=45, twistLimitDegrees=45}
 {bone1Name=dyn_waistRobe_02_r, bone2Name=dyn_waistRobe_01_r, swing1LimitDegrees=1, swing2LimitDegrees=10, twistLimitDegrees=0}
 {bone1Name=dyn_waistRobe_03_r, bone2Name=dyn_waistRobe_02_r, swing1LimitDegrees=1, swing2LimitDegrees=10, twistLimitDegrees=0}
 {bone1Name=dyn_waistRobe_04_r, bone2Name=dyn_waistRobe_03_r, swing1LimitDegrees=1, swing2LimitDegrees=10, twistLimitDegrees=0}
 {bone1Name=dyn_waistRobe_05_r, bone2Name=dyn_waistRobe_04_r, swing1LimitDegrees=1, swing2LimitDegrees=10, twistLimitDegrees=0}
]
