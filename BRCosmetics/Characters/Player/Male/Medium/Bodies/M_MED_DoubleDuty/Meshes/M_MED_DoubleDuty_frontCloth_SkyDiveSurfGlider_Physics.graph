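# Physics asset for the DoubleDuty male medium body's front loincloth, named for the skydive, surf and glider states.
# Left and right capsule chains each hang from the pelvis through a master bone and four segments.
# Tight angular limits keep each strip mostly swinging on one axis, with slight twist allowed only in the lower two segments.
# Pelvis and doubled thigh capsules supply body volumes, with thighs and spine_01 joined to the pelvis by angularly free constraints.
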PhysicsAsset /BRCosmetics/Characters/Player/Male/Medium/Bodies/M_MED_DoubleDuty/Meshes/M_MED_DoubleDuty_frontCloth_SkyDiveSurfGlider_Physics
bodies {
 dyn_loinClothFront_l_01 [{shapeType=Capsule, length=1.6, radius=3.3, shapeName=dyn_loinClothFront_l_01_capsule, center=<1.9849 4 -2.2>, rotation=<rot 90 0 0>}]
 dyn_loinClothFront_l_02 [{shapeType=Capsule, length=4, radius=3.55, shapeName=dyn_loinClothFront_l_02_capsule, center=<5.6778 2.3484 0>, rotation=<rot 90 0 0>}]
 dyn_loinClothFront_l_03 [{shapeType=Capsule, length=6.1654, radius=3.55, shapeName=dyn_loinClothFront_l_03_capsule, center=<5.6778 2.3484 0>, rotation=<rot 90 0 0>}]
 dyn_loinClothFront_l_04 [{shapeType=Capsule, length=10.8171, radius=3.55, shapeName=dyn_loinClothFront_l_04_capsule, center=<5.6778 2.3484 0>, rotation=<rot 90 0 0>}]
 dyn_loinClothFront_l_MSTR [{shapeType=Capsule, length=1, radius=1, shapeName=dyn_loinClothFront_l_MSTR_capsule}]
 dyn_loinClothFront_r_01 [{shapeType=Capsule, length=1.6, radius=3.3, shapeName=dyn_loinClothFront_r_01_capsule, center=<1.9849 4 2.2>, rotation=<rot 90 0 0>}]
 dyn_loinClothFront_r_02 [{shapeType=Capsule, length=4, radius=3.55, shapeName=dyn_loinClothFront_r_02_capsule, center=<5.6778 2.3484 0>, rotation=<rot 90 0 0>}]
 dyn_loinClothFront_r_03 [{shapeType=Capsule, length=6.15, radius=3.55, shapeName=dyn_loinClothFront_r_03_capsule, center=<5.6778 2.3484 0>, rotation=<rot 90 0 0>}]
 dyn_loinClothFront_r_04 [{shapeType=Capsule, length=10.8171, radius=3.55, shapeName=dyn_loinClothFront_r_04_capsule, center=<5.6778 2.3484 0>, rotation=<rot 90 0 0>}]
 dyn_loinClothFront_r_MSTR [{shapeType=Capsule, length=1, radius=1, shapeName=dyn_loinClothFront_r_MSTR_capsule}]
 pelvis [
  {shapeType=Capsule, length=12.6838, radius=10.6966, shapeName=pelvis_capsule, center=<0 -0.9941 0>}
  {shapeType=Capsule, length=11.664, radius=10.5279, shapeName=pelvis_capsule, center=<-3.905 -1.6865 0>}
 ]
 spine_01 [{shapeType=Capsule, length=0.505, radius=0.505, shapeName=spine_01_capsule}]
 thigh_l [
  {
   shapeType=Capsule
   length=31.4853
   radius=8.6946
   shapeName=thigh_l_capsule
   center=<-14.5502 0.6685 -0.3169>
   rotation=<rot 88.1968 176.6878 -179.227>
  }
  {
   shapeType=Capsule
   length=31.4853
   radius=8.6946
   shapeName=thigh_l_capsule
   center=<-14.8138 0.6855 3.9268>
   rotation=<rot 83.9447 -4.3141 -0.2304>
  }
 ]
 thigh_r [
  {
   shapeType=Capsule
   length=30.7353
   radius=9.4446
   shapeName=thigh_l_capsule
   center=<14.5252 -1.019 0.3167>
   rotation=<rot 88.1968 176.6878 0.773>
  }
  {
   shapeType=Capsule
   length=30.7353
   radius=9.4446
   shapeName=thigh_l_capsule
   center=<14.7223 -1.0309 -4.8109>
   rotation=<rot 84.1601 -4.3226 179.7611>
  }
 ]
}
constraints [
 {
  bone1Name=spine_01
  bone2Name=pelvis
  swing1LimitDegrees=45
  swing1Motion=Free
  swing2LimitDegrees=45
  swing2Motion=Free
  twistLimitDegrees=45
  twistMotion=Free
 }
 {
  bone1Name=thigh_l
  bone2Name=pelvis
  swing1LimitDegrees=45
  swing1Motion=Free
  swing2LimitDegrees=45
  swing2Motion=Free
  twistLimitDegrees=45
  twistMotion=Free
 }
 {bone1Name=dyn_loinClothFront_l_MSTR, bone2Name=pelvis, swing1LimitDegrees=45, swing2LimitDegrees=20, twistLimitDegrees=0}
 {bone1Name=dyn_loinClothFront_l_01, bone2Name=dyn_loinClothFront_l_MSTR, swing1LimitDegrees=20, swing2LimitDegrees=0, twistLimitDegrees=0}
 {bone1Name=dyn_loinClothFront_l_02, bone2Name=dyn_loinClothFront_l_01, swing1LimitDegrees=5, swing2LimitDegrees=0, twistLimitDegrees=0}
 {bone1Name=dyn_loinClothFront_l_03, bone2Name=dyn_loinClothFront_l_02, swing1LimitDegrees=20, swing2LimitDegrees=0, twistLimitDegrees=15}
 {bone1Name=dyn_loinClothFront_l_04, bone2Name=dyn_loinClothFront_l_03, swing1LimitDegrees=20, swing2LimitDegrees=0, twistLimitDegrees=15}
 {bone1Name=dyn_loinClothFront_r_MSTR, bone2Name=pelvis, swing1LimitDegrees=45, swing2LimitDegrees=20, twistLimitDegrees=0}
 {bone1Name=dyn_loinClothFront_r_01, bone2Name=dyn_loinClothFront_r_MSTR, swing1LimitDegrees=10, swing2LimitDegrees=0, twistLimitDegrees=0}
 {bone1Name=dyn_loinClothFront_r_02, bone2Name=dyn_loinClothFront_r_01, swing1LimitDegrees=5, swing2LimitDegrees=0, twistLimitDegrees=0}
 {bone1Name=dyn_loinClothFront_r_03, bone2Name=dyn_loinClothFront_r_02, swing1LimitDegrees=20, swing2LimitDegrees=0, twistLimitDegrees=15}
 {bone1Name=dyn_loinClothFront_r_04, bone2Name=dyn_loinClothFront_r_03, swing1LimitDegrees=20, swing2LimitDegrees=0, twistLimitDegrees=15}
 {
  bone1Name=thigh_r
  bone2Name=pelvis
  swing1LimitDegrees=45
  swing1Motion=Free
  swing2LimitDegrees=45
  swing2Motion=Free
  twistLimitDegrees=45
  twistMotion=Free
 }
]
