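# Animation Blueprint for the M_MED_DoubleDuty player body cosmetic, correcting and simulating its loincloth, waist robe and belt bones.
# Cached-pose stages combine thigh constraints, spring-driven AnimDynamics relative to the thighs, and belt Look Ats aimed at the left calf.
# A front-end blend adds menu-specific fixes before the final pass.
# Long chains of additive bone offsets are weighted by gameplay states like crouch, sprint, knee slide, DBNO, melee and shotgun.
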AnimBlueprint /BRCosmetics/Characters/Player/Male/Medium/Bodies/M_MED_DoubleDuty/Meshes/M_MED_DoubleDuty_AnimBP
target_skeleton=&/M_MED_DoubleDuty_Skeleton
graphs [
 AnimGraph {nodes [
   Component To Local()
   "Transform (Modify) Bone - Bone: dyn_loinClothFront_r_03"(Alpha=1, Translation="0.000,-6.390,2.714", Rotation="0,0,0", Scale="1,1,1") {
    rotationMode=BMM_Additive
    rotationSpace=BCS_BoneSpace
    scaleMode=BMM_Additive
    scaleSpace=BCS_BoneSpace
    translationMode=BMM_Additive
    translationSpace=BCS_BoneSpace
    alphaBoolBlend {blendInTime=0.3, blendOption=Linear, blendOutTime=1}
    boneToModify {boneName=dyn_loinClothFront_r_03}
    translation {y=-6.39, z=2.714}
   }
   n1 = Get crouchAlpha() -> (crouchAlpha:double)
   "Transform (Modify) Bone - Bone: dyn_belt_02_fr"(Alpha=@n1, Translation="0,0,2", Rotation="4,0,0", Scale="1,1,1") {
    rotationMode=BMM_Additive
    rotationSpace=BCS_BoneSpace
    translationMode=BMM_Additive
    translationSpace=BCS_BoneSpace
    alphaBoolBlend {blendOption=Linear}
    alphaScaleBiasClamp {bInterpResult=true}
    boneToModify {boneName=dyn_belt_02_fr}
    rotation {pitch=4}
    translation {z=2}
   }
   n2 = Get WalkNotCrouch() -> (WalkNotCrouch:double)
   n3 = "Transform (Modify) Bone - Bone: base_belt_01_fr"!(
    ComponentPose=Local To Component(LocalPose=Use cached pose 'CRCache_FrontEnd'())
    Alpha=@n2
    Translation="0,0,0"
    Rotation="9.303,-6.091,-0.992"
   ) -> (Pose) {
    rotationMode=BMM_Additive
    rotationSpace=BCS_BoneSpace
    scaleMode=BMM_Additive
    scaleSpace=BCS_BoneSpace
    translationMode=BMM_Additive
    translationSpace=BCS_BoneSpace
    alphaBoolBlend {blendOption=Linear}
    alphaScaleBiasClamp {bInterpResult=true}
    boneToModify {boneName=base_belt_01_fr}
    rotation=<rot 9.3031 -6.091 -0.992>
   }
   n4 = "Transform (Modify) Bone - Bone: base_belt_01_bck"!(ComponentPose=@n3, Alpha=@n2, Translation="0,0,0", Rotation="11.224,-5.780,-1.129") -> (Pose) {
    rotationMode=BMM_Additive
    rotationSpace=BCS_BoneSpace
    scaleMode=BMM_Additive
    scaleSpace=BCS_BoneSpace
    translationMode=BMM_Additive
    translationSpace=BCS_BoneSpace
    alphaBoolBlend {blendOption=Linear}
    alphaScaleBiasClamp {bInterpResult=true}
    boneToModify {boneName=base_belt_01_bck}
    rotation=<rot 11.224 -5.78 -1.129>
   }
   n5 = "Transform (Modify) Bone - Bone: dyn_loinClothFront_r_MSTR"(ComponentPose=@n4, Alpha=1, Translation="-0.185,0,3", Rotation="-10,0,0", Scale="1,1,1") {
    rotationMode=BMM_Additive
    rotationSpace=BCS_BoneSpace
    scaleMode=BMM_Additive
    scaleSpace=BCS_BoneSpace
    translationMode=BMM_Additive
    translationSpace=BCS_BoneSpace
    alphaBoolBlend {blendOption=Linear}
    alphaScaleBiasClamp {bInterpResult=true, interpSpeedDecreasing=8, interpSpeedIncreasing=8}
    boneToModify {boneName=dyn_loinClothFront_r_MSTR}
    rotation {pitch=-10}
    translation {x=-0.185, z=3}
   }
   n6 = "Transform (Modify) Bone - Bone: dyn_loinClothFront_l_MSTR"(ComponentPose=@n5, Alpha=1, Translation="0,0,-3.5", Rotation="0,0,-5", Scale="1,1,1") {
    rotationMode=BMM_Additive
    rotationSpace=BCS_BoneSpace
    scaleMode=BMM_Additive
    scaleSpace=BCS_BoneSpace
    translationMode=BMM_Additive
    translationSpace=BCS_BoneSpace
    alphaBoolBlend {blendOption=Linear}
    alphaScaleBiasClamp {bInterpResult=true, interpSpeedDecreasing=8, interpSpeedIncreasing=8}
    boneToModify {boneName=dyn_loinClothFront_l_MSTR}
    rotation {roll=-5}
    translation {z=-3.5}
   }
   n7 = "Transform (Modify) Bone - Bone: dyn_waistRobe_02_fr"!(ComponentPose=@n6, Alpha=1, Translation="0.165,0.005,2.469", Rotation="-9.712,-0.070,0.363") -> (Pose) {
    rotationMode=BMM_Additive
    rotationSpace=BCS_BoneSpace
    translationMode=BMM_Additive
    translationSpace=BCS_BoneSpace
    alphaBoolBlend {blendOption=Linear}
    alphaScaleBiasClamp {bInterpResult=true}
    boneToModify {boneName=dyn_waistRobe_02_fr}
    rotation=<rot -9.712 -0.07 0.363>
    translation=<0.165 0.005 2.469>
   }
   n8 = "Transform (Modify) Bone - Bone: dyn_waistRobe_02_fr"(ComponentPose=@n7, Alpha=1, Translation="0,0,3.5", Rotation="6,0,0") {
    rotationMode=BMM_Additive
    rotationSpace=BCS_BoneSpace
    translationMode=BMM_Additive
    translationSpace=BCS_BoneSpace
    alphaBoolBlend {blendOption=Linear}
    alphaScaleBiasClamp {bInterpResult=true, interpSpeedDecreasing=3}
    boneToModify {boneName=dyn_waistRobe_02_fr}
    rotation {pitch=6}
    translation {z=3.5}
   }
   n9 = "Transform (Modify) Bone - Bone: base_waistRobe_01_fr"!(ComponentPose=@n8, Alpha=Get runningStrap() -> (runningStrap:double), Translation="-2.114,4.295,4.621", Rotation="2.649,-0.107,0.097") -> (Pose) {
    rotationMode=BMM_Additive
    rotationSpace=BCS_BoneSpace
    translationMode=BMM_Additive
    translationSpace=BCS_BoneSpace
    alphaBoolBlend {blendOption=Linear}
    alphaScaleBiasClamp {bInterpResult=true, interpSpeedDecreasing=5, interpSpeedIncreasing=5}
    boneToModify {boneName=base_waistRobe_01_fr}
    rotation=<rot 2.649 -0.107 0.097>
    translation=<-2.114 4.295 4.621>
   }
   n10 = "Transform (Modify) Bone - Bone: dyn_loinClothFront_l_03"(ComponentPose=@n9, Alpha=1, Translation="0,0,0", Rotation="0,30,-0.031", Scale="1,1,1") {
    rotationMode=BMM_Additive
    rotationSpace=BCS_BoneSpace
    translationMode=BMM_Additive
    translationSpace=BCS_BoneSpace
    alphaBoolBlend {blendOption=Linear}
    alphaScaleBiasClamp {bInterpResult=true}
    boneToModify {boneName=dyn_loinClothFront_l_03}
    rotation {roll=-0.031, yaw=30}
   }
   n11 = "Transform (Modify) Bone - Bone: dyn_loinClothFront_r_03"(ComponentPose=@n10, Alpha=1, Translation="0,0,0", Rotation="10.586,3.323,-0.000", Scale="1,1,1") {
    rotationMode=BMM_Additive
    rotationSpace=BCS_BoneSpace
    translationMode=BMM_Additive
    translationSpace=BCS_BoneSpace
    alphaBoolBlend {blendOption=Linear}
    alphaScaleBiasClamp {bInterpResult=true}
    boneToModify {boneName=dyn_loinClothFront_r_03}
    rotation=<rot 10.586 3.323 0>
   }
   n12 = "Transform (Modify) Bone - Bone: attach_cape"(ComponentPose=@n11, Alpha=1, Translation="0,-0.5,0", Rotation="0,-4,0", Scale="1,1,1") {
    rotationMode=BMM_Additive
    rotationSpace=BCS_BoneSpace
    scaleMode=BMM_Additive
    scaleSpace=BCS_BoneSpace
    translationMode=BMM_Additive
    translationSpace=BCS_BoneSpace
    alphaBoolBlend {blendOption=Linear}
    alphaScaleBiasClamp {bInterpResult=true}
    boneToModify {boneName=attach_cape}
    rotation {yaw=-4}
    translation {y=-0.5}
   }
   n13 = "Transform (Modify) Bone - Bone: base_robe_01_r"(ComponentPose=@n12, Alpha=Get parachuteAlpha() -> (parachuteAlpha:double), Translation="0,0,5", Rotation="0,0,0", Scale="1,1,1") {
    translationMode=BMM_Additive
    translationSpace=BCS_BoneSpace
    alphaBoolBlend {blendOption=Linear}
    alphaScaleBiasClamp {bInterpResult=true}
    boneToModify {boneName=base_robe_01_r}
    translation {z=5}
   }
   n14 = Get jumpingAlpha() -> (jumpingAlpha:double)
   n15 = "Transform (Modify) Bone - Bone: base_robe_02_r"(ComponentPose=@n13, Alpha=@n14, Translation="0,0,5", Rotation="0,0,0", Scale="1,1,1") {
    translationMode=BMM_Additive
    translationSpace=BCS_BoneSpace
    alphaBoolBlend {blendOption=Linear}
    alphaScaleBiasClamp {bInterpResult=true}
    boneToModify {boneName=base_robe_02_r}
    translation {z=5}
   }
   n16 = "Transform (Modify) Bone - Bone: base_robe_01_r"(ComponentPose=@n15, Alpha=@n14, Translation="0,0,5", Rotation="0,0,0", Scale="1,1,1") {
    translationMode=BMM_Additive
    translationSpace=BCS_BoneSpace
    alphaBoolBlend {blendOption=Linear}
    alphaScaleBiasClamp {bInterpResult=true}
    boneToModify {boneName=base_robe_01_r}
    translation {z=5}
   }
   n17 = Get bIsKneeSliding()
   n18 = "Transform (Modify) Bone - Bone: dyn_loinClothFront_l_02"!(ComponentPose=@n16, bAlphaBoolEnabled:bool=@n17, Translation="-0.018,-1.886,0.252", Rotation="-5.821,-0.073,0.009", Scale="1,1,1") -> (Pose) {
    alphaInputType=Bool
    rotationMode=BMM_Additive
    rotationSpace=BCS_BoneSpace
    translationMode=BMM_Additive
    translationSpace=BCS_BoneSpace
    alphaBoolBlend {blendInTime=0.3, blendOption=Linear, blendOutTime=1}
    boneToModify {boneName=dyn_loinClothFront_l_02}
    rotation=<rot -5.821 -0.073 0.009>
    translation=<-0.018 -1.886 0.252>
   }
   n19 = "Transform (Modify) Bone - Bone: dyn_loinClothFront_r_03"!(ComponentPose=@n18, bAlphaBoolEnabled:bool=@n17, Translation="-0.181,-6.126,-2.802", Rotation="-0.728,25.651,1.894", Scale="1,1,1") -> (Pose) {
    alphaInputType=Bool
    rotationMode=BMM_Additive
    rotationSpace=BCS_BoneSpace
    translationMode=BMM_Additive
    translationSpace=BCS_BoneSpace
    alphaBoolBlend {blendInTime=0.3, blendOption=Linear, blendOutTime=1}
    boneToModify {boneName=dyn_loinClothFront_r_03}
    rotation=<rot -0.728 25.651 1.894>
    translation=<-0.181 -6.126 -2.802>
   }
   n20 = "Transform (Modify) Bone - Bone: dyn_belt_04_fr"(ComponentPose=@n19, Alpha=Get CrouchMove() -> (CrouchMove:double), Translation="0,0,1.5", Rotation="10,0,0", Scale="1,1,1") {
    rotationMode=BMM_Additive
    rotationSpace=BCS_BoneSpace
    translationMode=BMM_Additive
    translationSpace=BCS_BoneSpace
    alphaBoolBlend {blendOption=Linear}
    alphaScaleBiasClamp {bInterpResult=true, interpSpeedDecreasing=3, interpSpeedIncreasing=3}
    boneToModify {boneName=dyn_belt_04_fr}
    rotation {pitch=10}
    translation {z=1.5}
   }
   n21 = "Transform (Modify) Bone - Bone: dyn_waistRobe_02_fr"!(ComponentPose=@n20, bAlphaBoolEnabled:bool=@n17, Translation="0,0,0", Rotation="-11.899,36.490,4.222", Scale:Vector="1,1,1") -> (Pose) {
    alphaInputType=Bool
    rotationMode=BMM_Additive
    rotationSpace=BCS_BoneSpace
    translationMode=BMM_Additive
    translationSpace=BCS_BoneSpace
    alphaBoolBlend {blendInTime=0.3, blendOption=Linear, blendOutTime=1}
    boneToModify {boneName=dyn_waistRobe_02_fr}
    rotation=<rot -11.899 36.49 4.222>
   }
   n22 = Get MeeleAttackAlpha() -> (MeeleAttackAlpha:double)
   n23 = "Transform (Modify) Bone - Bone: dyn_loinClothFront_l_01"(ComponentPose=@n21, Alpha=@n22, Translation="0,0,0", Rotation="-3.922,7.395,16.177", Scale="1,1,1") {
    rotationMode=BMM_Additive
    rotationSpace=BCS_BoneSpace
    translationMode=BMM_Additive
    translationSpace=BCS_BoneSpace
    alphaBoolBlend {blendInTime=0.3, blendOption=Linear, blendOutTime=1}
    boneToModify {boneName=dyn_loinClothFront_l_01}
    rotation=<rot -3.922 7.395 16.177>
   }
   n24 = "Transform (Modify) Bone - Bone: dyn_waistRobe_03_r"(ComponentPose=@n23, Alpha=@n22, Translation="-0.034,-0.452,0.601", Rotation="3.354,-6.313,-0.370", Scale="1,1,1") {
    rotationMode=BMM_Additive
    rotationSpace=BCS_BoneSpace
    translationMode=BMM_Additive
    translationSpace=BCS_BoneSpace
    alphaBoolBlend {blendInTime=0.3, blendOption=Linear, blendOutTime=1}
    boneToModify {boneName=dyn_waistRobe_03_r}
    rotation=<rot 3.354 -6.313 -0.37>
    translation=<-0.034 -0.4518 0.601>
   }
   n25 = "Transform (Modify) Bone - Bone: dyn_loinClothFront_l_MSTR"!(ComponentPose=@n24, bAlphaBoolEnabled:bool=True, Translation="-0.697,1.605,-0.621", Rotation="-0.000,23.432,0.000", Scale="0.85,1,1") -> (Pose) {
    alphaInputType=Bool
    rotationMode=BMM_Additive
    rotationSpace=BCS_BoneSpace
    scaleMode=BMM_Additive
    scaleSpace=BCS_BoneSpace
    translationMode=BMM_Additive
    translationSpace=BCS_BoneSpace
    alphaBoolBlend {blendInTime=0.3, blendOption=Linear, blendOutTime=1}
    boneToModify {boneName=dyn_loinClothFront_l_MSTR}
    rotation=<rot 0 23.432 0>
    scale {x=0.85}
    translation=<-0.697 1.605 -0.621>
   }
   n26 = "Transform (Modify) Bone - Bone: dyn_loinClothFront_r_MSTR"!(ComponentPose=@n25, bAlphaBoolEnabled:bool=True, Translation="0,0,0", Rotation="-0.000,19.649,-0.000", Scale="0.85,1,1") -> (Pose) {
    alphaInputType=Bool
    rotationMode=BMM_Additive
    rotationSpace=BCS_BoneSpace
    scaleMode=BMM_Additive
    scaleSpace=BCS_BoneSpace
    translationMode=BMM_Additive
    translationSpace=BCS_BoneSpace
    alphaBoolBlend {blendInTime=0.3, blendOption=Linear, blendOutTime=1}
    boneToModify {boneName=dyn_loinClothFront_r_MSTR}
    rotation {yaw=19.649}
    scale {x=0.85}
   }
   n27 = "Transform (Modify) Bone - Bone: dyn_waistRobe_01_r"(ComponentPose=@n26, Alpha=Get rbanOff() -> (rbanOff:double), Translation="0,0,0", Rotation="18.444,-0.087,0.091", Scale="1,1,1") {
    rotationMode=BMM_Additive
    rotationSpace=BCS_BoneSpace
    scaleMode=BMM_Additive
    scaleSpace=BCS_BoneSpace
    translationMode=BMM_Additive
    translationSpace=BCS_BoneSpace
    alphaBoolBlend {blendInTime=0.3, blendOption=Linear, blendOutTime=1}
    boneToModify {boneName=dyn_waistRobe_01_r}
    rotation=<rot 18.444 -0.0871 0.0911>
   }
   n28 = Get DBNOMoving() -> (DBNOMoving:double)
   n29 = "Transform (Modify) Bone - Bone: dyn_loinClothFront_r_MSTR"!(ComponentPose=@n27, Alpha:float=@n28, Translation="0.000,-3.319,0.000", Rotation="-7.634,-8.508,0.867", Scale="0.779,1.000,1.000") -> (Pose) {
    rotationMode=BMM_Additive
    rotationSpace=BCS_BoneSpace
    scaleMode=BMM_Additive
    scaleSpace=BCS_BoneSpace
    translationMode=BMM_Additive
    translationSpace=BCS_BoneSpace
    alphaBoolBlend {blendInTime=0.3, blendOption=Linear, blendOutTime=1}
    alphaScaleBiasClamp {bInterpResult=true, interpSpeedDecreasing=5, interpSpeedIncreasing=5}
    boneToModify {boneName=dyn_loinClothFront_r_MSTR}
    rotation=<rot -7.634 -8.508 0.867>
    scale {x=0.779}
    translation=<0 -3.3193 0>
   }
   n30 = "Transform (Modify) Bone - Bone: dyn_loinClothFront_l_MSTR"!(ComponentPose=@n29, Alpha:float=@n28, Translation="0.023,-2.293,2.010", Rotation="-3.814,-26.779,0.713", Scale="0.85,1,1") -> (Pose) {
    rotationMode=BMM_Additive
    rotationSpace=BCS_BoneSpace
    scaleMode=BMM_Additive
    scaleSpace=BCS_BoneSpace
    translationMode=BMM_Additive
    translationSpace=BCS_BoneSpace
    alphaBoolBlend {blendInTime=0.3, blendOption=Linear, blendOutTime=1}
    alphaScaleBiasClamp {bInterpResult=true, interpSpeedDecreasing=5, interpSpeedIncreasing=5}
    boneToModify {boneName=dyn_loinClothFront_l_MSTR}
    rotation=<rot -3.814 -26.779 0.713>
    scale {x=0.85}
    translation=<0.023 -2.2929 2.01>
   }
   n31 = "Transform (Modify) Bone - Bone: base_waistRobe_01_fr"!(ComponentPose=@n30, Alpha=@n28, Translation="-2.537,0.000,2.390", Rotation="8.554,0.000,0.000", Scale:Vector="0.85,1,1") -> (Pose) {
    rotationMode=BMM_Additive
    rotationSpace=BCS_BoneSpace
    scaleMode=BMM_Additive
    scaleSpace=BCS_BoneSpace
    translationMode=BMM_Additive
    translationSpace=BCS_BoneSpace
    alphaBoolBlend {blendInTime=0.3, blendOption=Linear, blendOutTime=1}
    alphaScaleBiasClamp {bInterpResult=true, interpSpeedDecreasing=5, interpSpeedIncreasing=5}
    boneToModify {boneName=base_waistRobe_01_fr}
    rotation=<rot 8.554 0 0>
    scale {x=0.85}
    translation=<-2.537 0 2.3902>
   }
   n32 = "Transform (Modify) Bone - Bone: base_waistRobe_01_fr"(
    ComponentPose=@n31
    Alpha=Get AlphaShotgun() -> (AlphaShotgun:float)
    Translation="-0.000,0.188,1.337"
    Rotation="-0.099,0.828,9.589"
    Scale="1,1,1"
   ) {
    rotationMode=BMM_Additive
    rotationSpace=BCS_BoneSpace
    scaleMode=BMM_Additive
    scaleSpace=BCS_BoneSpace
    translationMode=BMM_Additive
    translationSpace=BCS_BoneSpace
    alphaBoolBlend {blendInTime=0.3, blendOption=Linear, blendOutTime=1}
    alphaScaleBiasClamp {bInterpResult=true, interpSpeedDecreasing=5, interpSpeedIncreasing=5}
    boneToModify {boneName=base_waistRobe_01_fr}
    rotation=<rot -0.099 0.828 9.589>
    translation {y=0.188, z=1.337}
   }
   n33 = "Transform (Modify) Bone - Bone: base_waistRobe_01_r"!(ComponentPose=@n32, bAlphaBoolEnabled:bool=True, Translation="0,0,0", Rotation="0,5,0", Scale="1,1,1") -> (Pose) {
    alphaInputType=Bool
    rotationMode=BMM_Additive
    rotationSpace=BCS_BoneSpace
    scaleMode=BMM_Additive
    scaleSpace=BCS_BoneSpace
    translationMode=BMM_Additive
    translationSpace=BCS_BoneSpace
    alphaBoolBlend {blendInTime=0.3, blendOption=Linear, blendOutTime=1}
    alphaScaleBiasClamp {bInterpResult=true, interpSpeedDecreasing=5, interpSpeedIncreasing=5}
    boneToModify {boneName=base_waistRobe_01_r}
    rotation {yaw=5}
   }
   n34 = "Transform (Modify) Bone - Bone: dyn_waistRobe_03_fr"(ComponentPose=@n33, Alpha=1, Translation="-2.189,8.202,-1.553", Rotation="51.814,26.370,23.273", Scale="1,1,1") {
    rotationMode=BMM_Additive
    rotationSpace=BCS_BoneSpace
    scaleMode=BMM_Additive
    scaleSpace=BCS_BoneSpace
    translationMode=BMM_Additive
    translationSpace=BCS_BoneSpace
    alphaBoolBlend {blendInTime=0.3, blendOption=Linear, blendOutTime=1}
    alphaScaleBiasClamp {bInterpResult=true, interpSpeedDecreasing=5, interpSpeedIncreasing=5}
    boneToModify {boneName=dyn_waistRobe_03_fr}
    rotation=<rot 51.814 26.37 23.273>
    translation=<-2.189 8.202 -1.553>
   }
   n35 = "Transform (Modify) Bone - Bone: dyn_loinClothFront_l_02"(ComponentPose=@n34, bAlphaBoolEnabled=True, Translation="-1.816,0.004,-2.131", Rotation="-0.000,11.913,-0.000", Scale="1,1,1") {
    alphaInputType=Bool
    rotationMode=BMM_Additive
    rotationSpace=BCS_BoneSpace
    translationMode=BMM_Additive
    translationSpace=BCS_BoneSpace
    alphaBoolBlend {blendOption=Linear}
    boneToModify {boneName=dyn_loinClothFront_l_02}
    rotation=<rot 0 11.913 0>
    translation=<-1.816 0.0039 -2.131>
   }
   n36 = "Transform (Modify) Bone - Bone: dyn_loinClothFront_r_03"!(ComponentPose=@n35, Alpha:float=1, Translation="0,0,0", Rotation="0.092,16.465,0.071", Scale="1,1,1") -> (Pose) {
    rotationMode=BMM_Additive
    rotationSpace=BCS_BoneSpace
    translationMode=BMM_Additive
    translationSpace=BCS_BoneSpace
    alphaBoolBlend {blendOption=Linear}
    alphaScaleBiasClamp {bInterpResult=true, interpSpeedDecreasing=7, interpSpeedIncreasing=7}
    boneToModify {boneName=dyn_loinClothFront_r_03}
    rotation=<rot 0.092 16.465 0.071>
   }
   n37 = "Transform (Modify) Bone - Bone: dyn_loinClothFront_l_02"!(ComponentPose=@n36, Alpha:float=1, Translation="0,0,0", Rotation="0.005,12.717,0.001", Scale="1,1,1") -> (Pose) {
    rotationMode=BMM_Additive
    rotationSpace=BCS_BoneSpace
    translationMode=BMM_Additive
    translationSpace=BCS_BoneSpace
    alphaBoolBlend {blendOption=Linear}
    alphaScaleBiasClamp {bInterpResult=true, interpSpeedDecreasing=7, interpSpeedIncreasing=7}
    boneToModify {boneName=dyn_loinClothFront_l_02}
    rotation=<rot 0.005 12.7171 0.001>
   }
   n38 = "Transform (Modify) Bone - Bone: dyn_waistRobe_02_r"(ComponentPose=@n37, Alpha=Get sprintingAlpha() -> (sprintingAlpha:double), Translation="0,0,0", Rotation="0,0,0", Scale="1,1.5,1") {
    rotationSpace=BCS_BoneSpace
    scaleMode=BMM_Additive
    scaleSpace=BCS_BoneSpace
    translationSpace=BCS_BoneSpace
    alphaBoolBlend {blendOption=Linear}
    alphaScaleBiasClamp {bInterpResult=true}
    boneToModify {boneName=dyn_waistRobe_02_r}
    scale {y=1.5}
   }
   n39 = "Transform (Modify) Bone - Bone: dyn_loinClothFront_r_03"(ComponentPose=@n38, Alpha=1, Translation="-0.150,-0.780,0.461", Rotation="0,0,0", Scale="1,1,1") {
    rotationMode=BMM_Additive
    rotationSpace=BCS_BoneSpace
    translationMode=BMM_Additive
    translationSpace=BCS_BoneSpace
    alphaBoolBlend {blendOption=Linear}
    alphaScaleBiasClamp {bInterpResult=true}
    boneToModify {boneName=dyn_loinClothFront_r_03}
    translation=<-0.15 -0.7799 0.461>
   }
   n40 = "Transform (Modify) Bone - Bone: dyn_waistRobe_02_fr"!(ComponentPose=@n39, Alpha:float=1, Translation="0,0,0", Rotation="0.002,-0.054,20.742", Scale="1,1,1") -> (Pose) {
    rotationMode=BMM_Additive
    rotationSpace=BCS_BoneSpace
    translationMode=BMM_Additive
    translationSpace=BCS_BoneSpace
    alphaBoolBlend {blendOption=Linear}
    boneToModify {boneName=dyn_waistRobe_02_fr}
    rotation=<rot 0.002 -0.054 20.742>
   }
   n41 = Get bIsKneeSliding()
   n42 = "Transform (Modify) Bone - Bone: dyn_loinClothFront_l_01"!(ComponentPose=@n40, bAlphaBoolEnabled:bool=@n41, Translation="0.402,-1.010,5.361", Rotation="-19.304,-1.292,15.610", Scale="1,1,1") -> (Pose) {
    alphaInputType=Bool
    rotationMode=BMM_Additive
    rotationSpace=BCS_BoneSpace
    translationMode=BMM_Additive
    translationSpace=BCS_BoneSpace
    alphaBoolBlend {blendInTime=0.3, blendOption=Linear, blendOutTime=1}
    boneToModify {boneName=dyn_loinClothFront_l_01}
    rotation=<rot -19.304 -1.292 15.61>
    translation=<0.4019 -1.0101 5.3607>
   }
   n43 = "Transform (Modify) Bone - Bone: dyn_loinClothFront_l_03"!(ComponentPose=@n42, bAlphaBoolEnabled:bool=@n41, Translation="0,0,0", Rotation="20.172,17.918,4.622", Scale="1,1,1") -> (Pose) {
    alphaInputType=Bool
    rotationMode=BMM_Additive
    rotationSpace=BCS_BoneSpace
    translationMode=BMM_Additive
    translationSpace=BCS_BoneSpace
    alphaBoolBlend {blendInTime=0.3, blendOption=Linear, blendOutTime=1}
    boneToModify {boneName=dyn_loinClothFront_l_03}
    rotation=<rot 20.172 17.918 4.622>
   }
   Output Pose(Result=Component To Local(ComponentPose=@n43)) {layerGroup=None}
   n44 = Get SprintConst() -> (SprintConst:float)
   n45 = "Constraint - dyn_loinClothFront_r_MSTR (PP)"!(
    ComponentPose=Local To Component(LocalPose=Use cached pose 'CRCache_PreFixes'())
    Alpha=@n44
    ConstraintWeights_0=0.7
    ConstraintWeights_1:float=0.3
   ) -> (Pose) {
    alphaBoolBlend {blendOption=Linear}
    boneToModify {boneName=dyn_loinClothFront_r_MSTR}
    constraintSetup [{perAxis {bY=false}, targetBone {boneName=thigh_r}}, {targetBone {boneName=dyn_loinClothFront_l_MSTR}}]
    constraintWeights [1,1]
   }
   n46 = "Constraint - dyn_loinClothFront_l_MSTR (PP)"!(ComponentPose=@n45, Alpha=@n44, ConstraintWeights_0=0.7, ConstraintWeights_1:float=0.3) -> (Pose) {
    alphaBoolBlend {blendOption=Linear}
    boneToModify {boneName=dyn_loinClothFront_l_MSTR}
    constraintSetup [{perAxis {bY=false}, targetBone {boneName=thigh_l}}, {targetBone {boneName=dyn_loinClothFront_r_MSTR}}]
    constraintWeights [1,1]
   }
   n47 = "Constraint - dyn_loinClothFront_r_MSTR (P)"(ComponentPose=@n46, Alpha=1, ConstraintWeights_0=1) {
    alphaBoolBlend {blendOption=Linear}
    boneToModify {boneName=dyn_loinClothFront_r_MSTR}
    constraintSetup [{perAxis {bY=false}, targetBone {boneName=thigh_r}}]
    constraintWeights [1]
   }
   n48 = "Constraint - dyn_loinClothFront_l_MSTR (P)"(ComponentPose=@n47, Alpha=1, ConstraintWeights_0=1) {
    alphaBoolBlend {blendOption=Linear}
    boneToModify {boneName=dyn_loinClothFront_l_MSTR}
    constraintSetup [{perAxis {bY=false}, targetBone {boneName=thigh_l}}]
    constraintWeights [1]
   }
   n49 = "Transform (Modify) Bone - Bone: dyn_loinClothFront_l_MSTR"!(ComponentPose=@n48, bAlphaBoolEnabled:bool=True, Translation="5.143,3.469,-1.796", Rotation="2.070,21.813,37.001", Scale="1,1,1") -> (Pose) {
    alphaInputType=Bool
    rotationMode=BMM_Additive
    rotationSpace=BCS_BoneSpace
    scaleMode=BMM_Additive
    scaleSpace=BCS_BoneSpace
    translationMode=BMM_Additive
    translationSpace=BCS_BoneSpace
    alphaBoolBlend {blendOption=Linear}
    boneToModify {boneName=dyn_loinClothFront_l_MSTR}
    rotation=<rot 2.07 21.813 37.001>
    translation=<5.143 3.469 -1.796>
   }
   n50 = "Transform (Modify) Bone - Bone: dyn_loinClothFront_r_MSTR"!(ComponentPose=@n49, bAlphaBoolEnabled:bool=True, Translation="6.118,-0.828,3.611", Rotation="-4.262,31.844,-19.878", Scale="1,1,1") -> (Pose) {
    alphaInputType=Bool
    rotationMode=BMM_Additive
    rotationSpace=BCS_BoneSpace
    scaleMode=BMM_Additive
    scaleSpace=BCS_BoneSpace
    translationMode=BMM_Additive
    translationSpace=BCS_BoneSpace
    alphaBoolBlend {blendOption=Linear}
    boneToModify {boneName=dyn_loinClothFront_r_MSTR}
    rotation=<rot -4.262 31.8441 -19.878>
    translation=<6.118 -0.828 3.6109>
   }
   n51 = "Transform (Modify) Bone - Bone: dyn_loinClothFront_r_MSTR"!(ComponentPose=@n50, Alpha:float=0, Translation="0.014,0.252,0.063", Rotation="17.998,-9.797,-18.360", Scale="1,1,1") -> (Pose) {
    rotationMode=BMM_Additive
    rotationSpace=BCS_BoneSpace
    translationMode=BMM_Additive
    translationSpace=BCS_BoneSpace
    alphaBoolBlend {blendInTime=0.3, blendOption=Linear, blendOutTime=1}
    alphaScaleBiasClamp {bInterpResult=true}
    boneToModify {boneName=dyn_loinClothFront_r_MSTR}
    rotation=<rot 17.998 -9.797 -18.36>
    translation=<0.014 0.2523 0.0631>
   }
   n52 = "Transform (Modify) Bone - Bone: dyn_loinClothFront_l_MSTR"!(ComponentPose=@n51, Alpha:float=0, Translation="7.053,-5.039,-9.133", Rotation="22.929,-40.747,79.169", Scale="1,1,1") -> (Pose) {
    rotationMode=BMM_Additive
    rotationSpace=BCS_BoneSpace
    translationMode=BMM_Additive
    translationSpace=BCS_BoneSpace
    alphaBoolBlend {blendInTime=0.3, blendOption=Linear, blendOutTime=1}
    alphaScaleBiasClamp {bInterpResult=true}
    boneToModify {boneName=dyn_loinClothFront_l_MSTR}
    rotation=<rot 22.929 -40.747 79.169>
    translation=<7.053 -5.039 -9.133>
   }
   n53 = "Transform (Modify) Bone - Bone: dyn_loinClothFront_r_03"(ComponentPose=@n52, Alpha=0, Translation="9.421,-1.104,-0.060", Rotation="0.189,-58.387,-2.204", Scale="1,1,1") {
    rotationMode=BMM_Additive
    rotationSpace=BCS_BoneSpace
    translationMode=BMM_Additive
    translationSpace=BCS_BoneSpace
    alphaBoolBlend {blendInTime=0.3, blendOption=Linear, blendOutTime=1}
    alphaScaleBiasClamp {bInterpResult=true}
    boneToModify {boneName=dyn_loinClothFront_r_03}
    rotation=<rot 0.189 -58.387 -2.204>
    translation=<9.421 -1.104 -0.06>
   }
   n54 = "Transform (Modify) Bone - Bone: dyn_loinClothFront_l_03"!(ComponentPose=@n53, Alpha:float=0, Translation="8.258,-3.174,-1.051", Rotation="20.226,-62.629,-3.672", Scale="1,1,1") -> (Pose) {
    rotationMode=BMM_Additive
    rotationSpace=BCS_BoneSpace
    translationMode=BMM_Additive
    translationSpace=BCS_BoneSpace
    alphaBoolBlend {blendInTime=0.3, blendOption=Linear, blendOutTime=1}
    alphaScaleBiasClamp {bInterpResult=true}
    boneToModify {boneName=dyn_loinClothFront_l_03}
    rotation=<rot 20.226 -62.629 -3.672>
    translation=<8.258 -3.174 -1.051>
   }
   n55 = "Transform (Modify) Bone - Bone: base_belt_01_bck"!(ComponentPose=@n54, Alpha=1, Translation="0,0,0", Rotation="0.009,-60,-0.013", Scale:Vector="1,1,1") -> (Pose) {
    rotationMode=BMM_Additive
    rotationSpace=BCS_BoneSpace
    translationMode=BMM_Additive
    translationSpace=BCS_BoneSpace
    alphaBoolBlend {blendInTime=0.3, blendOption=Linear, blendOutTime=1}
    alphaScaleBiasClamp {bInterpResult=true}
    boneToModify {boneName=base_belt_01_bck}
    rotation=<rot 0.009 -60 -0.013>
   }
   n56 = "Transform (Modify) Bone - Bone: base_belt_01_fr"!(ComponentPose=@n55, Alpha=1, Translation="0,0,0", Rotation="-30,-50,0", Scale:Vector="1,1,1") -> (Pose) {
    rotationMode=BMM_Additive
    rotationSpace=BCS_BoneSpace
    translationMode=BMM_Additive
    translationSpace=BCS_BoneSpace
    alphaBoolBlend {blendInTime=0.3, blendOption=Linear, blendOutTime=1}
    alphaScaleBiasClamp {bInterpResult=true}
    boneToModify {boneName=base_belt_01_fr}
    rotation {pitch=-30, yaw=-50}
   }
   n57 = "Transform (Modify) Bone - Bone: dyn_belt_02_fr"(ComponentPose=@n56, Alpha=1, Translation="0,0,0", Rotation="25,0,0", Scale="1,1,1") {
    rotationMode=BMM_Additive
    rotationSpace=BCS_BoneSpace
    translationMode=BMM_Additive
    translationSpace=BCS_BoneSpace
    alphaBoolBlend {blendOption=Linear}
    boneToModify {boneName=dyn_belt_02_fr}
    rotation {pitch=25}
   }
   n58 = Anim Dynamics - Bone: dyn_loinClothFront_l_02(ComponentPose=@n57, Alpha=1, ExternalForce="0,0,0") {
    angularDampingOverride=1
    angularSpringConstant=20000
    bAngularSpring=true
    bOverrideAngularDamping=true
    bUseGravityOverride=true
    gravityScale=0.35
    simulationSpace=BoneRelative
    alphaBoolBlend {blendOption=Linear}
    boundBone {boneName=dyn_loinClothFront_l_02}
    componentAppliedLinearAccClamp=<10000 10000 10000>
    componentLinearAccScale=<0.5 0.5 0.5>
    componentLinearVelScale=<0.5 0.5 0.5>
    gravityOverride {z=-980}
    physicsBodyDefinitions [{
      boundBone {boneName=dyn_loinClothFront_l_02}
      boxExtents=<20 30 30>
      constraintSetup {angularLimitsMax=<0 7 15>, angularLimitsMin=<0 -7 -1>, angularTarget=<1 0 0>}
      localJointOffset=<4 0 0>
     }]
    relativeSpaceBone {boneName=thigh_l}
   }
   n59 = Anim Dynamics - Bone: dyn_loinClothFront_l_03(ComponentPose=@n58, Alpha=1, ExternalForce="0,0,0") {
    angularDampingOverride=1
    angularSpringConstant=20000
    bAngularSpring=true
    bOverrideAngularDamping=true
    bUseGravityOverride=true
    gravityScale=0.35
    simulationSpace=BoneRelative
    alphaBoolBlend {blendOption=Linear}
    boundBone {boneName=dyn_loinClothFront_l_03}
    componentAppliedLinearAccClamp=<10000 10000 10000>
    componentLinearAccScale=<0.5 0.5 0.5>
    componentLinearVelScale=<0.5 0.5 0.5>
    gravityOverride {z=-980}
    physicsBodyDefinitions [{
      boundBone {boneName=dyn_loinClothFront_l_03}
      boxExtents=<20 30 30>
      constraintSetup {angularLimitsMax=<0 7 15>, angularLimitsMin=<0 -7 -1>, angularTarget=<1 0 0>}
      localJointOffset=<4 0 0>
     }]
    relativeSpaceBone {boneName=thigh_l}
   }
   n60 = Anim Dynamics - Bone: dyn_loinClothFront_l_04(ComponentPose=@n59, Alpha=1, ExternalForce="0,0,0") {
    angularDampingOverride=1
    angularSpringConstant=20000
    bAngularSpring=true
    bOverrideAngularDamping=true
    bUseGravityOverride=true
    gravityScale=0.35
    simulationSpace=BoneRelative
    alphaBoolBlend {blendOption=Linear}
    boundBone {boneName=dyn_loinClothFront_l_04}
    componentAppliedLinearAccClamp=<10000 10000 10000>
    componentLinearAccScale=<0.5 0.5 0.5>
    componentLinearVelScale=<0.5 0.5 0.5>
    gravityOverride {z=-980}
    physicsBodyDefinitions [{
      boundBone {boneName=dyn_loinClothFront_l_04}
      boxExtents=<20 30 30>
      constraintSetup {angularLimitsMax=<0 7 15>, angularLimitsMin=<0 -7 -1>, angularTarget=<1 0 0>}
      localJointOffset=<4 0 0>
     }]
    relativeSpaceBone {boneName=thigh_l}
   }
   n61 = Anim Dynamics - Bone: dyn_loinClothFront_r_03(ComponentPose=@n60, Alpha=1, ExternalForce="0,0,0") {
    angularDampingOverride=1
    angularSpringConstant=20000
    bAngularSpring=true
    bOverrideAngularDamping=true
    bUseGravityOverride=true
    gravityScale=0.35
    simulationSpace=BoneRelative
    alphaBoolBlend {blendOption=Linear}
    boundBone {boneName=dyn_loinClothFront_r_03}
    componentAppliedLinearAccClamp=<10000 10000 10000>
    componentLinearAccScale=<0.5 0.5 0.5>
    componentLinearVelScale=<0.5 0.5 0.5>
    gravityOverride {z=-980}
    physicsBodyDefinitions [{
      boundBone {boneName=dyn_loinClothFront_r_03}
      boxExtents=<20 30 30>
      constraintSetup {angularLimitsMax=<0 7 15>, angularLimitsMin=<0 -7 -1>, angularTarget=<1 0 0>}
      localJointOffset=<4 0 0>
     }]
    relativeSpaceBone {boneName=thigh_r}
   }
   n62 = Anim Dynamics - Bone: dyn_loinClothFront_r_04(ComponentPose=@n61, Alpha=1, ExternalForce="0,0,0") {
    angularDampingOverride=1
    angularSpringConstant=20000
    bAngularSpring=true
    bOverrideAngularDamping=true
    bUseGravityOverride=true
    gravityScale=0.35
    simulationSpace=BoneRelative
    alphaBoolBlend {blendOption=Linear}
    boundBone {boneName=dyn_loinClothFront_r_04}
    componentAppliedLinearAccClamp=<10000 10000 10000>
    componentLinearAccScale=<0.5 0.5 0.5>
    componentLinearVelScale=<0.5 0.5 0.5>
    gravityOverride {z=-980}
    physicsBodyDefinitions [{
      boundBone {boneName=dyn_loinClothFront_r_04}
      boxExtents=<20 30 30>
      constraintSetup {angularLimitsMax=<0 7 15>, angularLimitsMin=<0 -7 -1>, angularTarget=<1 0 0>}
      localJointOffset=<4 0 0>
     }]
    relativeSpaceBone {boneName=thigh_r}
   }
   Save cached pose 'CrCache_SpringDynamics'(Pose=Component To Local(ComponentPose=@n62)) {cacheName=CrCache_SpringDynamics}
   n63 = "Constraint - dyn_loinClothFront_r_02 (P)"(ComponentPose=Local To Component(LocalPose=Use cached pose 'CRCache_Dynamics'()), Alpha=0.3, ConstraintWeights_0=1) {
    alphaBoolBlend {blendOption=Linear}
    boneToModify {boneName=dyn_loinClothFront_r_02}
    constraintSetup [{targetBone {boneName=dyn_waistRobe_02_fr}}]
    constraintWeights [1]
   }
   n64 = "Constraint - dyn_loinClothFront_l_01 (P)"(ComponentPose=@n63, Alpha=1, ConstraintWeights_0=1) {
    alphaBoolBlend {blendOption=Linear}
    boneToModify {boneName=dyn_loinClothFront_l_01}
    constraintSetup [{targetBone {boneName=dyn_loinClothFront_r_02}}]
    constraintWeights [1]
   }
   n65 = Component To Local(ComponentPose=@n64)
   n66 = Use cached pose 'CRCache_Dynamics'()
   n67 = NOT Boolean(A=Get bIsPlayingEmote())
   n68 = "Transform (Modify) Bone - Bone: dyn_waistRobe_01_fr"!(
    ComponentPose=Local To Component(LocalPose=Blend Poses by bool(BlendPose_0=@n65, BlendPose_1=@n66, bActiveValue=@n67) {blendType=Linear})
    Alpha=1
    Translation="0,0,0"
    Rotation="-4,0,0"
   ) -> (Pose) {
    rotationMode=BMM_Additive
    rotationSpace=BCS_BoneSpace
    translationMode=BMM_Replace
    translationSpace=BCS_BoneSpace
    alphaBoolBlend {blendOption=Linear}
    alphaScaleBiasClamp {bInterpResult=true, interpSpeedDecreasing=5, interpSpeedIncreasing=5}
    boneToModify {boneName=dyn_waistRobe_01_fr}
    rotation {pitch=-4}
   }
   n69 = "Transform (Modify) Bone - Bone: dyn_loinClothFront_r_MSTR"!(ComponentPose=@n68, Alpha=1, Translation="0,-0.8,1.4", Rotation="0,-3,0") -> (Pose) {
    rotationMode=BMM_Additive
    rotationSpace=BCS_BoneSpace
    translationMode=BMM_Additive
    translationSpace=BCS_BoneSpace
    alphaBoolBlend {blendOption=Linear}
    alphaScaleBiasClamp {bInterpResult=true, interpSpeedDecreasing=5, interpSpeedIncreasing=5}
    boneToModify {boneName=dyn_loinClothFront_r_MSTR}
    rotation {yaw=-3}
    translation {y=-0.8, z=1.4}
   }
   n70 = "Transform (Modify) Bone - Bone: dyn_belt_02_bck"!(ComponentPose=@n69, Alpha=1, Translation="0,0,-1.5", Rotation="1.3,0,0") -> (Pose) {
    rotationMode=BMM_Additive
    rotationSpace=BCS_BoneSpace
    translationMode=BMM_Additive
    translationSpace=BCS_BoneSpace
    alphaBoolBlend {blendOption=Linear}
    alphaScaleBiasClamp {bInterpResult=true, interpSpeedDecreasing=5, interpSpeedIncreasing=5}
    boneToModify {boneName=dyn_belt_02_bck}
    rotation {pitch=1.3}
    translation {z=-1.5}
   }
   n71 = "Transform (Modify) Bone - Bone: dyn_belt_03_fr"!(ComponentPose=@n70, Alpha=1, Translation="0,0,-1.2", Rotation="-4,0,0") -> (Pose) {
    rotationMode=BMM_Additive
    rotationSpace=BCS_BoneSpace
    translationMode=BMM_Additive
    translationSpace=BCS_BoneSpace
    alphaBoolBlend {blendOption=Linear}
    alphaScaleBiasClamp {bInterpResult=true, interpSpeedDecreasing=5, interpSpeedIncreasing=5}
    boneToModify {boneName=dyn_belt_03_fr}
    rotation {pitch=-4}
    translation {z=-1.2}
   }
   n72 = "Transform (Modify) Bone - Bone: dyn_waistRobe_01_fr"!(ComponentPose=@n71, bAlphaBoolEnabled:bool=True, Translation="0,0,0", Rotation="5.320,-0.000,0.000") -> (Pose) {
    alphaInputType=Bool
    rotationMode=BMM_Additive
    rotationSpace=BCS_BoneSpace
    translationMode=BMM_Additive
    translationSpace=BCS_BoneSpace
    alphaBoolBlend {blendOption=Linear}
    alphaScaleBiasClamp {bInterpResult=true, interpSpeedDecreasing=5, interpSpeedIncreasing=5}
    boneToModify {boneName=dyn_waistRobe_01_fr}
    rotation=<rot 5.3201 0 0>
   }
   n73 = "Transform (Modify) Bone - Bone: alt_loinClothFront_01_r"!(ComponentPose=@n72, bAlphaBoolEnabled:bool=True, Translation="0,0,0", Rotation="3.465,-7.264,0.004") -> (Pose) {
    alphaInputType=Bool
    rotationMode=BMM_Additive
    rotationSpace=BCS_BoneSpace
    translationMode=BMM_Additive
    translationSpace=BCS_BoneSpace
    alphaBoolBlend {blendOption=Linear}
    alphaScaleBiasClamp {bInterpResult=true, interpSpeedDecreasing=5, interpSpeedIncreasing=5}
    boneToModify {boneName=alt_loinClothFront_01_r}
    rotation=<rot 3.465 -7.264 0.004>
   }
   n74 = "Transform (Modify) Bone - Bone: alt_loinClothFront_01_l"!(ComponentPose=@n73, bAlphaBoolEnabled:bool=True, Translation="0,0,0", Rotation="-3.700,-1.047,0.042") -> (Pose) {
    alphaInputType=Bool
    rotationMode=BMM_Additive
    rotationSpace=BCS_BoneSpace
    translationMode=BMM_Additive
    translationSpace=BCS_BoneSpace
    alphaBoolBlend {blendOption=Linear}
    alphaScaleBiasClamp {bInterpResult=true, interpSpeedDecreasing=5, interpSpeedIncreasing=5}
    boneToModify {boneName=alt_loinClothFront_01_l}
    rotation=<rot -3.7 -1.0469 0.042>
   }
   Save cached pose 'CRCache_FrontEnd'(Pose=Blend Poses by bool(BlendPose_0=Component To Local(ComponentPose=@n74), BlendPose_1=Use cached pose 'CRCache_Dynamics'(), bActiveValue=Get bIsInFrontEnd()) {blendType=Linear}) {cacheName=CRCache_FrontEnd}
   n75 = Get rbanOff()
   n76 = Look At - Bone: base_belt_01_bck(ComponentPose=Local To Component(LocalPose=Use cached pose 'CRCache_Constraints'()), Alpha=@n75) {
    alphaBoolBlend {blendOption=Linear}
    alphaScaleBiasClamp {bInterpResult=true, interpSpeedDecreasing=5, interpSpeedIncreasing=5}
    boneToModify {boneName=base_belt_01_bck}
    lookAtLocation=<12.4469 12.8198 -1.3163>
    lookAtTarget {boneReference {boneName=calf_l}}
    lookAt_Axis {axis=<1 0>}
   }
   n77 = Look At - Bone: dyn_belt_02_fr(ComponentPose=@n76, Alpha=@n75) {
    alphaBoolBlend {blendOption=Linear}
    alphaScaleBiasClamp {bInterpResult=true, interpSpeedDecreasing=5, interpSpeedIncreasing=5}
    boneToModify {boneName=dyn_belt_02_fr}
    lookAtLocation=<17.6849 16.0497 -1.7909>
    lookAtTarget {boneReference {boneName=calf_l}}
    lookAt_Axis {axis=<1 0>}
   }
   n78 = Get RBANalpha() -> (RBANalpha:double)
   n79 = Look At - Bone: base_belt_01_bck(ComponentPose=@n77, Alpha=@n78) {
    alphaBoolBlend {blendOption=Linear}
    alphaScaleBiasClamp {bInterpResult=true, interpSpeedDecreasing=5, interpSpeedIncreasing=5}
    boneToModify {boneName=base_belt_01_bck}
    lookAtLocation=<-0.9484 19.5544 -0.0027>
    lookAtTarget {boneReference {boneName=calf_l}}
    lookAt_Axis {axis=<1 0>}
   }
   n80 = Look At - Bone: dyn_belt_02_fr(ComponentPose=@n79, Alpha=@n78) {
    alphaBoolBlend {blendOption=Linear}
    alphaScaleBiasClamp {bInterpResult=true, interpSpeedDecreasing=5, interpSpeedIncreasing=5}
    boneToModify {boneName=dyn_belt_02_fr}
    lookAtLocation=<0 31.2218 0>
    lookAtTarget {boneReference {boneName=calf_l}}
    lookAt_Axis {axis=<1 0>}
   }
   n81 = Get CrouchRBANoff() -> (CrouchRBANoff:double)
   n82 = "Transform (Modify) Bone - Bone: dyn_belt_02_fr"(ComponentPose=@n80, Alpha=@n81, Translation="0,0,0", Rotation="-6,0,0", Scale="1,1,1") {
    rotationMode=BMM_Additive
    rotationSpace=BCS_BoneSpace
    alphaBoolBlend {blendOption=Linear}
    boneToModify {boneName=dyn_belt_02_fr}
    rotation {pitch=-6}
   }
   n83 = "Transform (Modify) Bone - Bone: dyn_belt_01_bck"(ComponentPose=@n82, Alpha=@n81, Translation="0,0,0", Rotation="12,0,0", Scale="1,1,1") {
    rotationMode=BMM_Additive
    rotationSpace=BCS_BoneSpace
    alphaBoolBlend {blendOption=Linear}
    boneToModify {boneName=dyn_belt_01_bck}
    rotation {pitch=12}
   }
   n84 = Get AlphaShotgun()
   n85 = "Transform (Modify) Bone - Bone: lowerarm_twist_01_l"(ComponentPose=@n83, Alpha=@n84, Translation="0,0,0", Rotation="0,0,25", Scale="1,1.3,1.5") {
    rotationMode=BMM_Additive
    rotationSpace=BCS_BoneSpace
    scaleMode=BMM_Additive
    scaleSpace=BCS_BoneSpace
    alphaBoolBlend {blendOption=Linear}
    boneToModify {boneName=lowerarm_twist_01_l}
    rotation {roll=25}
    scale {y=1.3, z=1.5}
   }
   n86 = "Transform (Modify) Bone - Bone: lowerarm_twist_02_l"(ComponentPose=@n85, Alpha=@n84, Translation="0,0,0", Rotation="0,0,-25", Scale="1,1.1,1.1") {
    rotationMode=BMM_Additive
    rotationSpace=BCS_BoneSpace
    scaleMode=BMM_Additive
    scaleSpace=BCS_BoneSpace
    alphaBoolBlend {blendOption=Linear}
    boneToModify {boneName=lowerarm_twist_02_l}
    rotation {roll=-25}
    scale {y=1.1, z=1.1}
   }
   n87 = "Transform (Modify) Bone - Bone: base_centralBelt_r"(ComponentPose=@n86, Alpha=@n1, Translation="-2.5,0,0", Rotation="0,0,0", Scale="1,1,1") {
    rotationMode=BMM_Additive
    rotationSpace=BCS_BoneSpace
    translationMode=BMM_Additive
    translationSpace=BCS_BoneSpace
    alphaBoolBlend {blendOption=Linear}
    alphaScaleBiasClamp {bInterpResult=true, interpSpeedDecreasing=5, interpSpeedIncreasing=5}
    boneToModify {boneName=base_centralBelt_r}
    translation {x=-2.5}
   }
   n88 = "Transform (Modify) Bone - Bone: dyn_belt_02_fr"(ComponentPose=@n87, Alpha=Get crouchAlpha(), Translation="0,0,2", Rotation="10,0,0", Scale="1,1,1") {
    rotationMode=BMM_Additive
    rotationSpace=BCS_BoneSpace
    translationMode=BMM_Additive
    translationSpace=BCS_BoneSpace
    alphaBoolBlend {blendOption=Linear}
    boneToModify {boneName=dyn_belt_02_fr}
    rotation {pitch=10}
    translation {z=2}
   }
   n89 = "Transform (Modify) Bone - Bone: dyn_belt_02_fr"(ComponentPose=@n88, Alpha=Get CrouchRBANoff(), Translation="0,0,0", Rotation="-20,0,0", Scale="1,1,1") {
    rotationMode=BMM_Additive
    rotationSpace=BCS_BoneSpace
    alphaBoolBlend {blendOption=Linear}
    boneToModify {boneName=dyn_belt_02_fr}
    rotation {pitch=-20}
   }
   n90 = "Transform (Modify) Bone - Bone: dyn_belt_02_fr"!(ComponentPose=@n89, bAlphaBoolEnabled:bool=false, Translation="-0.064,0.000,0.433", Rotation="8.768,0.000,-0.000", Scale="1,1,1") -> (Pose) {
    alphaInputType=Bool
    rotationMode=BMM_Additive
    rotationSpace=BCS_BoneSpace
    translationMode=BMM_Additive
    translationSpace=BCS_BoneSpace
    alphaBoolBlend {blendOption=Linear}
    boneToModify {boneName=dyn_belt_02_fr}
    rotation=<rot 8.768 0 0>
    translation=<-0.0639 0 0.4325>
   }
   n91 = "Transform (Modify) Bone - Bone: attach_cape"(ComponentPose=@n90, Alpha=Get UpperArmRot_l() -> (UpperArmRot_l:double), Translation="0,-0.25,0", Rotation="0,-4,0", Scale="1,1,1") {
    rotationMode=BMM_Additive
    rotationSpace=BCS_BoneSpace
    translationMode=BMM_Additive
    translationSpace=BCS_BoneSpace
    alphaBoolBlend {blendOption=Linear}
    alphaScaleBiasClamp {bInterpResult=true, interpSpeedDecreasing=5, interpSpeedIncreasing=5}
    boneToModify {boneName=attach_cape}
    rotation {yaw=-4}
    translation {y=-0.25}
   }
   n92 = "Transform (Modify) Bone - Bone: base_waistRobe_01_fr"(ComponentPose=@n91, Alpha=0.5, Translation="0,1.5,-0.5", Rotation="15,0,0", Scale="1,1,1") {
    rotationMode=BMM_Additive
    rotationSpace=BCS_BoneSpace
    translationMode=BMM_Additive
    translationSpace=BCS_BoneSpace
    alphaBoolBlend {blendOption=Linear}
    alphaScaleBiasClamp {bInterpResult=true, interpSpeedDecreasing=5, interpSpeedIncreasing=5}
    boneToModify {boneName=base_waistRobe_01_fr}
    rotation {pitch=15}
    translation {y=1.5, z=-0.5}
   }
   n93 = "Transform (Modify) Bone - Bone: attach_cape"(ComponentPose=@n92, Alpha=Get sprintingAlpha(), Translation="0,-2.8,-3.6", Rotation="0,0,0", Scale="1,1,1") {
    rotationMode=BMM_Additive
    rotationSpace=BCS_BoneSpace
    translationMode=BMM_Additive
    translationSpace=BCS_BoneSpace
    alphaBoolBlend {blendOption=Linear}
    alphaScaleBiasClamp {bInterpResult=true}
    boneToModify {boneName=attach_cape}
    translation {y=-2.8, z=-3.6}
   }
   n94 = Get bIsDBNO()
   n95 = "Transform (Modify) Bone - Bone: dyn_belt_02_fr"(ComponentPose=@n93, bAlphaBoolEnabled=@n94, Translation="0.779,1.170,1.364", Rotation="22.634,-6.221,0.269", Scale="1,1,1") {
    alphaInputType=Bool
    rotationMode=BMM_Additive
    rotationSpace=BCS_BoneSpace
    translationMode=BMM_Additive
    translationSpace=BCS_BoneSpace
    alphaBoolBlend {blendOption=Linear}
    boneToModify {boneName=dyn_belt_02_fr}
    rotation=<rot 22.6341 -6.221 0.269>
    translation=<0.779 1.17 1.364>
   }
   n96 = "Transform (Modify) Bone - Bone: base_belt_01_bck"!(ComponentPose=@n95, bAlphaBoolEnabled:bool=@n94, Translation="0.363,-0.000,0.837", Rotation="25.230,0.014,-0.018", Scale="1,1,1") -> (Pose) {
    alphaInputType=Bool
    rotationMode=BMM_Additive
    rotationSpace=BCS_BoneSpace
    translationMode=BMM_Additive
    translationSpace=BCS_BoneSpace
    alphaBoolBlend {blendOption=Linear}
    boneToModify {boneName=base_belt_01_bck}
    rotation=<rot 25.2301 0.014 -0.018>
    translation {x=0.363, z=0.837}
   }
   n97 = Get On Rail() -> (On Rail:double)
   n98 = "Transform (Modify) Bone - Bone: dyn_belt_02_fr"!(ComponentPose=@n96, Alpha:float=@n97, Translation="-1.283,-1.294,0.384", Rotation="3.041,-37.926,-4.024", Scale="1,1,1") -> (Pose) {
    rotationMode=BMM_Additive
    rotationSpace=BCS_BoneSpace
    translationMode=BMM_Additive
    translationSpace=BCS_BoneSpace
    alphaBoolBlend {blendOption=Linear}
    alphaScaleBiasClamp {bInterpResult=true, interpSpeedDecreasing=5, interpSpeedIncreasing=5}
    boneToModify {boneName=dyn_belt_02_fr}
    rotation=<rot 3.041 -37.926 -4.024>
    translation=<-1.2829 -1.294 0.3836>
   }
   n99 = "Transform (Modify) Bone - Bone: base_belt_01_bck"!(ComponentPose=@n98, Alpha:float=@n97, Translation="-0.216,-0.000,1.084", Rotation="-6.284,-35.283,-7.717", Scale="1,1,1") -> (Pose) {
    rotationMode=BMM_Additive
    rotationSpace=BCS_BoneSpace
    translationMode=BMM_Additive
    translationSpace=BCS_BoneSpace
    alphaBoolBlend {blendOption=Linear}
    alphaScaleBiasClamp {bInterpResult=true, interpSpeedDecreasing=5, interpSpeedIncreasing=5}
    boneToModify {boneName=base_belt_01_bck}
    rotation=<rot -6.284 -35.283 -7.717>
    translation {x=-0.216, z=1.084}
   }
   n100 = "Transform (Modify) Bone - Bone: base_belt_01_fr"(ComponentPose=@n99, Alpha=1, Translation="-0.856,-0.577,-1.430", Rotation="8.118,-38.040,-36.913", Scale="1,1,1") {
    rotationMode=BMM_Additive
    rotationSpace=BCS_BoneSpace
    translationMode=BMM_Additive
    translationSpace=BCS_BoneSpace
    alphaBoolBlend {blendOption=Linear}
    alphaScaleBiasClamp {bInterpResult=true, interpSpeedDecreasing=5, interpSpeedIncreasing=5}
    boneToModify {boneName=base_belt_01_fr}
    rotation=<rot 8.118 -38.04 -36.913>
    translation=<-0.856 -0.577 -1.43>
   }
   n101 = "Transform (Modify) Bone - Bone: base_belt_01_fr"(ComponentPose=@n100, Alpha=1, Translation="0,0,0", Rotation="-14.954,0.010,0.022", Scale="1,1,1") {
    rotationMode=BMM_Additive
    rotationSpace=BCS_BoneSpace
    translationMode=BMM_Additive
    translationSpace=BCS_BoneSpace
    alphaBoolBlend {blendOption=Linear}
    alphaScaleBiasClamp {bInterpResult=true}
    boneToModify {boneName=base_belt_01_fr}
    rotation=<rot -14.954 0.01 0.022>
   }
   n102 = "Transform (Modify) Bone - Bone: base_belt_01_bck"(ComponentPose=@n101, Alpha=Get CrouchMove(), Translation="0,0,0", Rotation="-16.918,-52.582,-19.884", Scale="1,1,1") {
    rotationMode=BMM_Additive
    rotationSpace=BCS_BoneSpace
    translationMode=BMM_Additive
    translationSpace=BCS_BoneSpace
    alphaBoolBlend {blendOption=Linear}
    alphaScaleBiasClamp {bInterpResult=true, interpSpeedDecreasing=5, interpSpeedIncreasing=5}
    boneToModify {boneName=base_belt_01_bck}
    rotation=<rot -16.918 -52.582 -19.884>
   }
   Save cached pose 'CRCache_PreFixes'(Pose=Component To Local(ComponentPose=@n102)) {cacheName=CRCache_PreFixes}
   n103 = Anim Dynamics - Bone: dyn_robe_01_r(ComponentPose=Local To Component(LocalPose=Use cached pose 'CRCache_RigidBodies'()), Alpha=0.9, ExternalForce="0,0,0") {
    angularDampingOverride=1
    angularSpringConstant=20000
    bAngularSpring=true
    bOverrideAngularDamping=true
    bUseGravityOverride=true
    gravityScale=0.35
    simulationSpace=World
    alphaBoolBlend {blendOption=Linear}
    boundBone {boneName=dyn_robe_01_r}
    componentLinearAccScale=<1 1 1>
    componentLinearVelScale=<1 1 1>
    gravityOverride {z=-980}
    physicsBodyDefinitions [{
      boundBone {boneName=dyn_robe_01_r}
      boxExtents=<20 30 30>
      constraintSetup {angularLimitsMax=<20 20 5>, angularLimitsMin=<-20 -40 -5>, angularTarget=<1 0 0>}
      localJointOffset=<3 0 0>
     }]
    relativeSpaceBone {boneName=pelvis}
   }
   n104 = Anim Dynamics - Bone: dyn_robe_02_r(ComponentPose=@n103, Alpha=0.9, ExternalForce="0,0,0") {
    angularDampingOverride=1
    angularSpringConstant=20000
    bAngularSpring=true
    bOverrideAngularDamping=true
    bUseGravityOverride=true
    gravityScale=0.35
    simulationSpace=World
    alphaBoolBlend {blendOption=Linear}
    boundBone {boneName=dyn_robe_02_r}
    componentLinearAccScale=<1 1 1>
    componentLinearVelScale=<1 1 1>
    gravityOverride {z=-980}
    physicsBodyDefinitions [{
      boundBone {boneName=dyn_robe_02_r}
      boxExtents=<20 30 30>
      constraintSetup {angularLimitsMax=<20 40 15>, angularLimitsMin=<-20 -40 -5>, angularTarget=<1 0 0>}
      localJointOffset=<5 0 0>
     }]
    relativeSpaceBone {boneName=pelvis}
   }
   n105 = Anim Dynamics - Bone: dyn_robe_02_l(ComponentPose=@n104, Alpha=0.7, ExternalForce="0,0,0") {
    angularDampingOverride=1
    angularSpringConstant=20000
    bAngularSpring=true
    bOverrideAngularDamping=true
    bUseGravityOverride=true
    gravityScale=0.35
    simulationSpace=World
    alphaBoolBlend {blendOption=Linear}
    boundBone {boneName=dyn_robe_02_l}
    componentLinearAccScale=<1 1 1>
    componentLinearVelScale=<1 1 1>
    gravityOverride {z=-980}
    physicsBodyDefinitions [{
      boundBone {boneName=dyn_robe_02_l}
      boxExtents=<30 40 40>
      constraintSetup {angularLimitsMax=<50 5 15>, angularLimitsMin=<-50 -5 -15>, angularTarget=<1 0 0>}
      localJointOffset=<4 0 0>
     }]
    relativeSpaceBone {boneName=pelvis}
   }
   n106 = Anim Dynamics - Bone: dyn_robe_01_l(ComponentPose=@n105, Alpha=1, ExternalForce="0,0,0") {
    angularDampingOverride=1
    angularSpringConstant=20000
    bAngularSpring=true
    bOverrideAngularDamping=true
    bUseGravityOverride=true
    gravityScale=0.35
    simulationSpace=World
    alphaBoolBlend {blendOption=Linear}
    boundBone {boneName=dyn_robe_01_l}
    componentLinearAccScale=<1 1 1>
    componentLinearVelScale=<1 1 1>
    gravityOverride {z=-980}
    physicsBodyDefinitions [{
      boundBone {boneName=dyn_robe_01_l}
      boxExtents=<10 20 20>
      constraintSetup {angularLimitsMax=<50 5 15>, angularLimitsMin=<-50 -5 -15>, angularTarget=<1 0 0>}
      localJointOffset=<4 0 0>
     }]
    relativeSpaceBone {boneName=pelvis}
   }
   n107 = Anim Dynamics - Bone: dyn_robe_03_r(ComponentPose=@n106, Alpha=1, ExternalForce="0,0,0") {
    angularDampingOverride=1
    angularSpringConstant=20000
    bAngularSpring=true
    bOverrideAngularDamping=true
    bUseGravityOverride=true
    gravityScale=0.35
    simulationSpace=RootRelative
    alphaBoolBlend {blendOption=Linear}
    boundBone {boneName=dyn_robe_03_r}
    componentLinearAccScale=<1 1 1>
    componentLinearVelScale=<1 1 1>
    gravityOverride {z=-980}
    physicsBodyDefinitions [{
      boundBone {boneName=dyn_robe_03_r}
      boxExtents=<20 30 30>
      constraintSetup {angularLimitsMax=<0 20 10>, angularLimitsMin=<0 -20 -10>, angularTarget=<1 0 0>}
      localJointOffset=<4 0 0>
     }]
   }
   Save cached pose 'CRCache_Dynamics'(Pose=Component To Local(ComponentPose=@n107)) {cacheName=CRCache_Dynamics}
   n108 = "Transform (Modify) Bone - Bone: base_centralBelt_l"!(
    ComponentPose=Local To Component(LocalPose=Use cached pose 'CRCache_Input'())
    Alpha=Get thigh_rot_up_l() -> (thigh_rot_up_l:double)
    Translation="-5.717,-0.000,0.000"
    Rotation="0,0,0"
   ) -> (Pose) {
    rotationMode=BMM_Additive
    rotationSpace=BCS_BoneSpace
    scaleMode=BMM_Additive
    scaleSpace=BCS_BoneSpace
    translationMode=BMM_Additive
    translationSpace=BCS_BoneSpace
    alphaBoolBlend {blendOption=Linear}
    boneToModify {boneName=base_centralBelt_l}
    translation {x=-5.717}
   }
   n109 = "Transform (Modify) Bone - Bone: base_centralBelt_l"!(ComponentPose=@n108, Alpha=Get crouchAlpha(), Translation="-6,0,10", Rotation="0,0,0", Scale:Vector="1,1,1") -> (Pose) {
    rotationMode=BMM_Additive
    rotationSpace=BCS_BoneSpace
    translationMode=BMM_Additive
    translationSpace=BCS_BoneSpace
    alphaBoolBlend {blendOption=Linear}
    alphaScaleBiasClamp {bInterpResult=true, interpSpeedDecreasing=5, interpSpeedIncreasing=5}
    boneToModify {boneName=base_centralBelt_l}
    translation {x=-6, z=10}
   }
   n110 = "Constraint - base_belt_01_fr (PPPPPPPP)"!(
    ComponentPose=@n109
    Alpha=1
    ConstraintWeights_0=0.5269
    ConstraintWeights_1:float=0.2053
    ConstraintWeights_2:float=0.094
    ConstraintWeights_3:float=0.078
    ConstraintWeights_4:float=0.055
    ConstraintWeights_5:float=0.0181
    ConstraintWeights_6:float=0.018
    ConstraintWeights_7:float=0.0048
   ) -> (Pose) {
    alphaBoolBlend {blendOption=Linear}
    boneToModify {boneName=base_belt_01_fr}
    constraintSetup [
     {targetBone {boneName=pelvis}}
     {targetBone {boneName=base_centralBelt_l}}
     {targetBone {boneName=spine_02}}
     {targetBone {boneName=spine_01}}
     {targetBone {boneName=spine_03}}
     {targetBone {boneName=thigh_twist_01_l}}
     {targetBone {boneName=deform_groin_l}}
     {targetBone {boneName=base_centralBelt_r}}
    ]
    constraintWeights [1,1,1,1,1,1,1,1]
   }
   n111 = "Constraint - base_belt_01_bck (PPPPPPPP)"!(
    ComponentPose=@n110
    Alpha=1
    ConstraintWeights_0=0.5433
    ConstraintWeights_1:float=0.2888
    ConstraintWeights_2:float=0.0532
    ConstraintWeights_3:float=0.0451
    ConstraintWeights_4:float=0.0307
    ConstraintWeights_5:float=0.0204
    ConstraintWeights_6:float=0.0166
    ConstraintWeights_7:float=0.0018
   ) -> (Pose) {
    alphaBoolBlend {blendOption=Linear}
    boneToModify {boneName=base_belt_01_bck}
    constraintSetup [
     {targetBone {boneName=pelvis}}
     {targetBone {boneName=base_centralBelt_l}}
     {targetBone {boneName=spine_02}}
     {targetBone {boneName=spine_01}}
     {targetBone {boneName=spine_03}}
     {targetBone {boneName=deform_groin_l}}
     {targetBone {boneName=thigh_twist_01_l}}
     {targetBone {boneName=base_centralBelt_r}}
    ]
    constraintWeights [1,1,1,1,1,1,1,1]
   }
   n112 = "Transform (Modify) Bone - Bone: base_robe_01_r"!(ComponentPose=@n111, bAlphaBoolEnabled:bool=True, Translation="-0.000,1.408,2.000", Rotation="0,0,0", Scale="1.000,2.909,1.000") -> (Pose) {
    alphaInputType=Bool
    rotationMode=BMM_Additive
    rotationSpace=BCS_BoneSpace
    scaleMode=BMM_Additive
    scaleSpace=BCS_BoneSpace
    translationMode=BMM_Additive
    translationSpace=BCS_BoneSpace
    alphaBoolBlend {blendOption=Linear}
    boneToModify {boneName=base_robe_01_r}
    scale {y=2.909}
    translation=<0 1.4077 2>
   }
   n113 = "Transform (Modify) Bone - Bone: thigh_beltBody_l"(ComponentPose=@n112, Alpha=1, Translation="-0.045,-3.030,0.065", Rotation="0.001,-4.727,0.001", Scale="1,1,1") {
    rotationMode=BMM_Additive
    rotationSpace=BCS_BoneSpace
    scaleMode=BMM_Additive
    scaleSpace=BCS_BoneSpace
    translationMode=BMM_Additive
    translationSpace=BCS_BoneSpace
    alphaBoolBlend {blendOption=Linear}
    boneToModify {boneName=thigh_beltBody_l}
    rotation=<rot 0.001 -4.7271 0.001>
    translation=<-0.045 -3.03 0.065>
   }
   Save cached pose 'CRCache_Constraints'(Pose=Component To Local(ComponentPose=@n113)) {cacheName=CRCache_Constraints}
   n114 = Get RunningSkydive() -> (RunningSkydive:bool)
   n115 = "Transform (Modify) Bone - Bone: dyn_loinClothFront_l_MSTR"!(
    ComponentPose=Local To Component(LocalPose=Use cached pose 'CRCache_PreFixes'())
    bAlphaBoolEnabled:bool=True
    Translation="0,0,-3"
    Rotation="0.002,20,0"
    Scale="1,1,1"
   ) -> (Pose) {
    alphaInputType=Bool
    rotationMode=BMM_Additive
    rotationSpace=BCS_BoneSpace
    translationMode=BMM_Additive
    translationSpace=BCS_BoneSpace
    alphaBoolBlend {blendInTime=0.3, blendOption=Linear, blendOutTime=1}
    boneToModify {boneName=dyn_loinClothFront_l_MSTR}
    rotation {pitch=0.002, yaw=20}
    translation {z=-3}
   }
   n116 = "Transform (Modify) Bone - Bone: dyn_loinClothFront_r_MSTR"!(ComponentPose=@n115, bAlphaBoolEnabled:bool=True, Translation="0,0,3", Rotation="8,20,0", Scale:Vector="1,1,1") -> (Pose) {
    alphaInputType=Bool
    rotationMode=BMM_Additive
    rotationSpace=BCS_BoneSpace
    translationMode=BMM_Additive
    translationSpace=BCS_BoneSpace
    alphaBoolBlend {blendInTime=0.3, blendOption=Linear, blendOutTime=1}
    boneToModify {boneName=dyn_loinClothFront_r_MSTR}
    rotation {pitch=8, yaw=20}
    translation {z=3}
   }
   n117 = "Transform (Modify) Bone - Bone: dyn_loinClothFront_l_MSTR"(ComponentPose=@n116, bAlphaBoolEnabled=True, Translation="1.616,-0.808,1.714", Rotation="-3.679,61.298,-0.706", Scale="1,1,1") {
    alphaInputType=Bool
    rotationMode=BMM_Additive
    rotationSpace=BCS_BoneSpace
    translationMode=BMM_Additive
    translationSpace=BCS_BoneSpace
    alphaBoolBlend {blendInTime=0.3, blendOption=Linear, blendOutTime=1}
    boneToModify {boneName=dyn_loinClothFront_l_MSTR}
    rotation=<rot -3.679 61.298 -0.706>
    translation=<1.616 -0.808 1.7145>
   }
   n118 = "Transform (Modify) Bone - Bone: dyn_loinClothFront_r_MSTR"(ComponentPose=@n117, bAlphaBoolEnabled=True, Translation="0.793,-0.592,-2.454", Rotation="-1.412,61.621,-11.628", Scale="1,1,1") {
    alphaInputType=Bool
    rotationMode=BMM_Additive
    rotationSpace=BCS_BoneSpace
    translationMode=BMM_Additive
    translationSpace=BCS_BoneSpace
    alphaBoolBlend {blendInTime=0.3, blendOption=Linear, blendOutTime=1}
    boneToModify {boneName=dyn_loinClothFront_r_MSTR}
    rotation=<rot -1.412 61.621 -11.628>
    translation=<0.793 -0.592 -2.454>
   }
   n119 = "Transform (Modify) Bone - Bone: dyn_loinClothFront_l_MSTR"!(ComponentPose=@n118, Alpha:float=1, Translation="-0.104,0,-2", Rotation="-6,0,0", Scale="1,1,1") -> (Pose) {
    rotationMode=BMM_Additive
    rotationSpace=BCS_BoneSpace
    translationMode=BMM_Additive
    translationSpace=BCS_BoneSpace
    alphaBoolBlend {blendInTime=0.3, blendOption=Linear, blendOutTime=1}
    alphaScaleBiasClamp {bInterpResult=true}
    boneToModify {boneName=dyn_loinClothFront_l_MSTR}
    rotation {pitch=-6}
    translation {x=-0.104, z=-2}
   }
   n120 = "Transform (Modify) Bone - Bone: dyn_loinClothFront_r_MSTR"!(ComponentPose=@n119, Alpha:float=1, Translation="0.143,0,2", Rotation="-11,0,0", Scale="1,1,1") -> (Pose) {
    rotationMode=BMM_Additive
    rotationSpace=BCS_BoneSpace
    translationMode=BMM_Additive
    translationSpace=BCS_BoneSpace
    alphaBoolBlend {blendInTime=0.3, blendOption=Linear, blendOutTime=1}
    alphaScaleBiasClamp {bInterpResult=true}
    boneToModify {boneName=dyn_loinClothFront_r_MSTR}
    rotation {pitch=-11}
    translation {x=0.143, z=2}
   }
   n121 = RigidBody!(
    ComponentPose=@n120
    Alpha
    OverrideWorldGravity:Vector=Get GravityOverride()
    ExternalForce
    ComponentLinearAccScale:Vector="0.3,0.3,0.3"
    ComponentLinearVelScale:Vector="0.3,0.3,0.3"
   ) -> (Pose) {
    bClampLinearTranslationLimitToRefPose=true
    bOverrideWorldGravity=true
    cachedBoundsScale=1.3
    overridePhysicsAsset=&/M_MED_DoubleDuty_CrouchWalk_Physics1
    simulationSpace=BaseBoneSpace
    alphaBoolBlend {blendOption=Linear}
    alphaScaleBiasClamp {interpSpeedDecreasing=8, interpSpeedIncreasing=8}
    baseBoneRef {boneName=pelvis}
    componentAppliedLinearAccClamp=<3000 3000 3000>
    componentLinearAccScale=<0.4 0.4 0.4>
    componentLinearVelScale=<0.4 0.4 0.4>
    simSpaceSettings {maxAngularAcceleration=1000, maxAngularVelocity=3, maxLinearAcceleration=700, maxLinearVelocity=500, velocityScaleZ=0, worldAlpha=0.2}
   }
   n122 = Get AccScaleVals() -> (AccScaleVals:Vector)
   n123 = RigidBody(
    ComponentPose=Local To Component(LocalPose=Use cached pose 'CRCache_PreFixes'())
    OverrideWorldGravity=Get GravityOverride()
    ComponentLinearAccScale=@n122
    ComponentLinearVelScale=@n122
   ) {
    bClampLinearTranslationLimitToRefPose=true
    bOverrideWorldGravity=true
    overridePhysicsAsset=&/M_MED_DoubleDuty_frontCloth_Physics
    simulationSpace=BaseBoneSpace
    alphaBoolBlend {blendOption=Linear}
    baseBoneRef {boneName=pelvis}
    componentAppliedLinearAccClamp=<6000 6000 6000>
    componentLinearAccScale=<0.4 0.4 0.4>
    componentLinearVelScale=<0.4 0.4 0.4>
   }
   n124 = "Transform (Modify) Bone - Bone: dyn_loinClothFront_r_MSTR"!(
    ComponentPose=@n123
    Alpha=Get isDetectedRotationInEmote() -> (isDetectedRotationInEmote:double)
    Translation="0,0,0"
    Rotation="-30,50,-0.02"
   ) -> (Pose) {
    rotationMode=BMM_Additive
    rotationSpace=BCS_BoneSpace
    translationMode=BMM_Additive
    translationSpace=BCS_BoneSpace
    alphaBoolBlend {blendOption=Linear}
    alphaScaleBiasClamp {bInterpResult=true, interpSpeedDecreasing=5, interpSpeedIncreasing=5}
    boneToModify {boneName=dyn_loinClothFront_r_MSTR}
    rotation=<rot -30 50 -0.02>
   }
   n125 = Get isPlayingEmote() -> (isPlayingEmote:double)
   n126 = "Transform (Modify) Bone - Bone: base_waistRobe_01_fr"!(ComponentPose=@n124, Alpha=@n125, Translation="0,2.5,2", Rotation="0,0,0") -> (Pose) {
    rotationMode=BMM_Additive
    rotationSpace=BCS_BoneSpace
    translationMode=BMM_Additive
    translationSpace=BCS_BoneSpace
    alphaBoolBlend {blendOption=Linear}
    alphaScaleBiasClamp {bInterpResult=true, interpSpeedDecreasing=5, interpSpeedIncreasing=5}
    boneToModify {boneName=base_waistRobe_01_fr}
    translation {y=2.5, z=2}
   }
   n127 = "Transform (Modify) Bone - Bone: dyn_waistRobe_02_fr"!(ComponentPose=@n126, Alpha=@n125, Translation="0,0,2", Rotation="0,0,0") -> (Pose) {
    rotationMode=BMM_Additive
    rotationSpace=BCS_BoneSpace
    translationMode=BMM_Additive
    translationSpace=BCS_BoneSpace
    alphaBoolBlend {blendOption=Linear}
    alphaScaleBiasClamp {bInterpResult=true, interpSpeedDecreasing=5, interpSpeedIncreasing=5}
    boneToModify {boneName=dyn_waistRobe_02_fr}
    translation {z=2}
   }
   n128 = Blend Poses by bool(
    BlendPose_0=Component To Local(ComponentPose=@n121)
    BlendPose_1=Component To Local(ComponentPose=@n127)
    BlendTime_0=0.5
    BlendTime_1=0.5
    bActiveValue=Get DBNOCrouchBool() -> (DBNOCrouchBool:bool)
   ) {blendType=Linear}
   n129 = Get AccScaleVals()
   n130 = RigidBody(
    ComponentPose=Local To Component(LocalPose=Use cached pose 'CRCache_PreFixes'())
    OverrideWorldGravity=Get GravityOverride()
    ComponentLinearAccScale=@n129
    ComponentLinearVelScale=@n129
   ) {
    bClampLinearTranslationLimitToRefPose=true
    bOverrideWorldGravity=true
    overridePhysicsAsset=&/M_MED_DoubleDuty_frontCloth_SkyDive_Physics
    simulationSpace=BaseBoneSpace
    alphaBoolBlend {blendOption=Linear}
    baseBoneRef {boneName=pelvis}
    componentAppliedLinearAccClamp=<6000 6000 6000>
    componentLinearAccScale=<0.4 0.4 0.4>
    componentLinearVelScale=<0.4 0.4 0.4>
   }
   n131 = "Transform (Modify) Bone - Bone: dyn_loinClothFront_l_03"(ComponentPose=@n130, Alpha=1, Translation="0,0,0", Rotation="0,15,0", Scale="1,1,1") {
    rotationMode=BMM_Additive
    rotationSpace=BCS_BoneSpace
    translationMode=BMM_Additive
    translationSpace=BCS_BoneSpace
    alphaBoolBlend {blendOption=Linear}
    alphaScaleBiasClamp {bInterpResult=true}
    boneToModify {boneName=dyn_loinClothFront_l_03}
    rotation {yaw=15}
   }
   n132 = "Transform (Modify) Bone - Bone: dyn_loinClothFront_r_03"(ComponentPose=@n131, Alpha=1, Translation="0,0,0", Rotation="-5,5,0", Scale="1,1,1") {
    rotationMode=BMM_Additive
    rotationSpace=BCS_BoneSpace
    translationMode=BMM_Additive
    translationSpace=BCS_BoneSpace
    alphaBoolBlend {blendOption=Linear}
    alphaScaleBiasClamp {bInterpResult=true}
    boneToModify {boneName=dyn_loinClothFront_r_03}
    rotation {pitch=-5, yaw=5}
   }
   n133 = Look At - Bone: dyn_loinClothFront_r_MSTR(ComponentPose=Local To Component(LocalPose=Use cached pose 'CRCache_PreFixes'()), Alpha=1) {
    alphaBoolBlend {blendOption=Linear}
    boneToModify {boneName=dyn_loinClothFront_r_MSTR}
    lookAtLocation=<-3.0135 -9.732 -16.8396>
    lookAtTarget {boneReference {boneName=calf_r}}
    lookAt_Axis {axis=<1 0>}
    lookUp_Axis {axis {x=-1}}
   }
   n134 = Look At - Bone: dyn_loinClothFront_l_MSTR(ComponentPose=@n133, Alpha=1) {
    alphaBoolBlend {blendOption=Linear}
    boneToModify {boneName=dyn_loinClothFront_l_MSTR}
    lookAtLocation=<7.1118 -4.8666 16.8024>
    lookAtTarget {boneReference {boneName=calf_l}}
    lookAt_Axis {axis=<1 0>}
    lookUp_Axis {axis {x=-1}}
   }
   n135 = RigidBody(
    ComponentPose=@n134
    OverrideWorldGravity=Get GravityOverride()
    ComponentLinearAccScale="0.3,0.3,0.3"
    ComponentLinearVelScale="0.3,0.3,0.3"
   ) {
    bClampLinearTranslationLimitToRefPose=true
    bOverrideWorldGravity=true
    overridePhysicsAsset=&/M_MED_DoubleDuty_frontCloth_SkyDiveSurfGlider_Physics
    simulationSpace=BaseBoneSpace
    alphaBoolBlend {blendOption=Linear}
    baseBoneRef {boneName=pelvis}
    componentAppliedLinearAccClamp=<3000 0 3000>
    componentLinearAccScale=<0.4 0.4 0.4>
    componentLinearVelScale=<0.4 0.4 0.4>
    simSpaceSettings {
     maxAngularAcceleration=1000
     maxAngularVelocity=3
     maxLinearAcceleration=700
     maxLinearVelocity=500
     velocityScaleZ=0
     worldAlpha=0.2
     externalLinearDragV=<0.2 0.2>
    }
   }
   n136 = Blend Poses by bool(
    BlendPose_0=Use cached pose 'CrCache_SpringDynamics'()
    BlendPose_1=Blend Poses by bool(
     BlendPose_0=Component To Local(ComponentPose=@n135)
     BlendPose_1=Component To Local(ComponentPose=@n132)
     bActiveValue=Get DoubleDutyGliderAlpha() -> (DoubleDutyGliderAlpha:bool)
    ) {blendType=Linear}
    BlendTime_0=1
    BlendTime_1=1
    bActiveValue=Get Sprint general() -> (Sprint general:bool)
   ) {blendType=Linear}
   n137 = "Transform (Modify) Bone - Bone: base_belt_01_bck"!(
    ComponentPose=Local To Component(LocalPose=Blend Poses by bool(BlendPose_0=@n136, BlendPose_1=@n128, BlendTime_0=0.5, BlendTime_1=1, bActiveValue=@n114) {blendType=Linear})
    Alpha=1
    Translation="0,0,0"
    Rotation="7,-0.008,-0.008"
   ) -> (Pose) {
    rotationMode=BMM_Additive
    rotationSpace=BCS_BoneSpace
    scaleMode=BMM_Additive
    scaleSpace=BCS_BoneSpace
    translationMode=BMM_Additive
    translationSpace=BCS_BoneSpace
    alphaBoolBlend {blendOption=Linear}
    alphaScaleBiasClamp {bInterpResult=true}
    boneToModify {boneName=base_belt_01_bck}
    rotation=<rot 7 -0.008 -0.008>
   }
   n138 = "Transform (Modify) Bone - Bone: dyn_loinClothFront_r_MSTR"(ComponentPose=@n137, Alpha=1, Translation="0,-1.5,4.7", Rotation="-8,0,0") {
    rotationMode=BMM_Additive
    rotationSpace=BCS_BoneSpace
    translationMode=BMM_Additive
    translationSpace=BCS_BoneSpace
    alphaBoolBlend {blendOption=Linear}
    alphaScaleBiasClamp {bInterpResult=true, interpSpeedDecreasing=5, interpSpeedIncreasing=5}
    boneToModify {boneName=dyn_loinClothFront_r_MSTR}
    rotation {pitch=-8}
    translation {y=-1.5, z=4.7}
   }
   n139 = Get GravityOverride()
   n140 = RigidBody!(ComponentPose=@n138, Alpha, OverrideWorldGravity=@n139, ExternalForce) -> (Pose) {
    bOverrideWorldGravity=true
    overridePhysicsAsset=&/M_MED_DoubleDuty_Belts_Physics
    simulationSpace=BaseBoneSpace
    alphaBoolBlend {blendOption=Linear}
    alphaScaleBiasClamp {interpSpeedDecreasing=5, interpSpeedIncreasing=5}
    baseBoneRef {boneName=pelvis}
    componentAppliedLinearAccClamp=<3000 3000 3000>
    componentLinearAccScale {y=0.3, z=0.3}
    componentLinearVelScale {y=0.3, z=0.3}
    simSpaceSettings {
     maxAngularAcceleration=1000
     maxAngularVelocity=3
     maxLinearAcceleration=700
     maxLinearVelocity=500
     velocityScaleZ=0
     worldAlpha=0.2
     externalLinearDragV=<0.2 0.2>
    }
   }
   n141 = RigidBody(ComponentPose=@n140, OverrideWorldGravity=@n139) {
    bOverrideWorldGravity=true
    overridePhysicsAsset=&/M_MED_DoubleDuty_WaistRobe_Physics
    simulationSpace=BaseBoneSpace
    alphaBoolBlend {blendOption=Linear}
    baseBoneRef {boneName=pelvis}
    componentAppliedLinearAccClamp=<1000 1000 1000>
    componentLinearAccScale=<0.1 0.1 0.1>
    componentLinearVelScale=<0.1 0.1 0.1>
    simSpaceSettings {
     maxAngularAcceleration=1000
     maxAngularVelocity=3
     maxLinearAcceleration=700
     maxLinearVelocity=500
     velocityScaleZ=0
     externalLinearDragV=<0.2 0.2>
    }
   }
   n142 = RigidBody(ComponentPose=@n141, Alpha=0.8, OverrideWorldGravity="0,0,0") {
    bOverrideWorldGravity=true
    overridePhysicsAsset=&/M_MED_DoubleDuty_Sleeve_Physics
    simulationSpace=BaseBoneSpace
    alphaBoolBlend {blendOption=Linear}
    alphaScaleBiasClamp {bInterpResult=true, interpSpeedDecreasing=5, interpSpeedIncreasing=5}
    baseBoneRef {boneName=spine_05}
    componentAppliedLinearAccClamp=<3000 3000 3000>
    componentLinearAccScale=<0.3 0.3 0.3>
    componentLinearVelScale=<0.3 0.3 0.3>
    simSpaceSettings {
     maxAngularAcceleration=1000
     maxAngularVelocity=3
     maxLinearAcceleration=700
     maxLinearVelocity=500
     velocityScaleZ=0
     externalLinearDragV=<0.2 0.2>
    }
   }
   n143 = Get isPlayingEmote()
   n144 = "Transform (Modify) Bone - Bone: dyn_belt_04_fr"!(ComponentPose=@n142, Alpha=@n143, Translation="0,0,0", Rotation="10,0,0") -> (Pose) {
    rotationMode=BMM_Additive
    rotationSpace=BCS_BoneSpace
    translationMode=BMM_Additive
    translationSpace=BCS_BoneSpace
    alphaBoolBlend {blendOption=Linear}
    alphaScaleBiasClamp {bInterpResult=true, interpSpeedDecreasing=5, interpSpeedIncreasing=5}
    boneToModify {boneName=dyn_belt_04_fr}
    rotation {pitch=10}
   }
   n145 = "Transform (Modify) Bone - Bone: dyn_loinClothFront_l_02"!(ComponentPose=@n144, Alpha=@n143, Translation="0,2,0", Rotation="0,0,0") -> (Pose) {
    rotationMode=BMM_Additive
    rotationSpace=BCS_BoneSpace
    translationMode=BMM_Additive
    translationSpace=BCS_BoneSpace
    alphaBoolBlend {blendOption=Linear}
    alphaScaleBiasClamp {bInterpResult=true, interpSpeedDecreasing=5, interpSpeedIncreasing=5}
    boneToModify {boneName=dyn_loinClothFront_l_02}
    translation {y=2}
   }
   Save cached pose 'CRCache_RigidBodies'(Pose=Component To Local(ComponentPose=@n145)) {cacheName=CRCache_RigidBodies}
   n146 = Control Rig!(
    Source=Copy Pose From Mesh() {bUseAttachedParent=true}
    Alpha
    crouchMoving:double=Get CrouchMovingCTRL() -> (CrouchMovingCTRL:double)
    tacticalSprint:double=Get sprintingAlpha()
    Crouch:double=Get sideToSideCrouch() -> (sideToSideCrouch:double)
    NotCrouching:double=Get notCrouch() -> (notCrouch:double)
   ) -> (Pose) {
    alphaBoolBlend {blendOption=Linear}
    alphaScaleBiasClamp {bInterpResult=true, interpSpeedDecreasing=5, interpSpeedIncreasing=5}
    controlRigAssetReference {blueprintRigClass=&/M_MED_DoubleDuty_CtrlRig.M_MED_DoubleDuty_CtrlRig_C}
   }
   Save cached pose 'CRCache_Input'(Pose=@n146) {cacheName=CRCache_Input}
   Use cached pose 'CRCache_Dynamics'()
  ]}
 EventGraph {nodes [
   Event BlueprintInitializeAnimation()
   n1 = GetOwnerSkeletalMeshComponent()
   > Set SkelMesh(SkelMesh=@n1)
   > Set Skel_Mesh(Skel_Mesh=@n1)
   > Set Skel Mesh(Skel Mesh=@n1)
   > Set fort_player(fort_player:Fort Player Anim Instance Object Reference=GetOwnerFortPlayerAnimInstance!(self:Custom Character Part Anim Instance Object Reference) -> (ReturnValue:Fort Player Anim Instance Object Reference)) -> (Output_Get:Fort Player Anim Instance Object Reference)
   n2 = TryGetPawnOwner()
   > Is Valid(InputObject=@n2)
   > n3 = Cast To FortPlayerPawnAthena(Object=@n2)
   > n4 = Set PlayerPawn(PlayerPawn:Fort Player Pawn Athena Object Reference=@n3) -> (Output_Get:Fort Player Pawn Athena Object Reference)
   > Set Body_Mesh(Body_Mesh=GetSkeletalMeshForPartType(self=@n4))
   n5 = Event BlueprintUpdateAnimation()
   > n6 = Sequence!(execute) -> (then_0, then_1, then_2:Exec, then_3:Exec, then_4:Exec, then_5:Exec)
   n7 = GetOwnerSkeletalMeshComponent()
   n8 = BreakRotator(InRot=BreakTransform(InTransform=GetDeltaTransformFromRefPose(self=@n7, BoneName=upperarm_l, BaseName=clavicle_l)).Rotation)
   n9 = Branch(
    execute=@n6.then_3
    Condition=OR Boolean(A="float <= float"(A=@n8.Pitch, B=-50) -> (ReturnValue=~), B="float <= float"(A=@n8.Yaw, B=3) -> (ReturnValue=~))
   )
   > Set UpperArmRot_l(UpperArmRot_l:double=1) -> (Output_Get:double)
   Set UpperArmRot_l(execute=@n9.else, UpperArmRot_l=0)
   Is Valid(exec=@n6.then_0, InputObject=Get PlayerPawn() -> (PlayerPawn:Fort Player Pawn Athena Object Reference))
   > n10 = Branch(Condition=Get bIsInFrontEnd())
   Set WPO Globals Function(
    execute=@n10.else
    Material=Get Mat Instance Dynamic()
    Overlay Material=Get Overlay Mat Instance Dynamic()
    Skeletal Mesh=Get SkelMesh()
    "Local Noise Transform  Location"=Get Local Noise Transform Location() -> (Local Noise Transform Location:Vector)
    Local Noise Offset=Get Local Noise Offset()
   )
   > Set WPO Regional Function!(
    execute
    Material:Material Instance Dynamic Object Reference=Get Mat Instance Dynamic()
    Overlay Material:Material Instance Dynamic Object Reference=Get Overlay Mat Instance Dynamic()
    Skeletal Mesh:Skeletal Mesh Component Object Reference=Get SkelMesh()
    Velocity Transform:Name=foot_r
    Velocity Interp Speed:double=Get Velocity Interp Speed()
    Collision Transform:Name=calf_r
    Collision Transform Axis:EAxis Enum
    Invert Collision Transform Axis:bool=false
    Delta Time X:double=@n5.DeltaTimeX
    Previous Velocity Transform Location:"Vector (by ref)"=Get R1 Velocity Transform Location()
    Previous Velocity Transform Velocity:"Vector (by ref)"=Get R1 Velocity Transform Velocity()
    Region:int=1
   ) -> (then)
   > Set WPO Regional Function(
    Material=Get Mat Instance Dynamic()
    Overlay Material=Get Overlay Mat Instance Dynamic()
    Skeletal Mesh=Get SkelMesh()
    Velocity Transform=foot_l
    Velocity Interp Speed=Get Velocity Interp Speed()
    Collision Transform=calf_l
    Invert Collision Transform Axis=true
    Delta Time X=@n5.DeltaTimeX
    Previous Velocity Transform Location=Get R2 Velocity Transform Location()
    Previous Velocity Transform Velocity=Get R2 Velocity Transform Velocity()
    Region=2
   )
   > n11 = Branch(Condition=OR Boolean(A=Get bIsSkydiveFloating(), B=Get bIsParachuteOpened()))
   > Set WPO Pho Wind Intensity(WPO Pho Wind Intensity=0.8)
   > n12 = Set WPO Pho Turbulence(WPO Pho Turbulence:double=2) -> (Output_Get:double)
   SetScalarParameterValue(
    execute=[@n14.then, @n12.then]
    self=[Get Mat Instance Dynamic(), Get Overlay Mat Instance Dynamic()]
    ParameterName=ClothSim_WindStrength
    Value=Get WPO Pho Wind Intensity()
   )
   n13 = Get WPO Pho Turbulence() -> (WPO Pho Turbulence:double)
   > SetScalarParameterValue(self=[Get Mat Instance Dynamic(), Get Overlay Mat Instance Dynamic()], ParameterName=ClothSim_R1_TurbulenceStrength, Value=@n13)
   > SetScalarParameterValue(self=[Get Overlay Mat Instance Dynamic(), Get Mat Instance Dynamic()], ParameterName=ClothSim_R2_TurbulenceStrength, Value=@n13)
   Set WPO Pho Wind Intensity(execute=@n11.else, WPO Pho Wind Intensity=0)
   > n14 = Set WPO Pho Turbulence(WPO Pho Turbulence=0.1)
   n15 = Gravity Override State Function DEPRECATED(
    execute=@n6.then_1
    CharacterPartAnimInstance=Self-Reference()
    GameTimeInSeconds=GetGameTimeInSeconds()
    GravityOverrideParams=Get Gravity Override Params()
   )
   > Set GravityOverride(GravityOverride=@n15)
   > n16 = Cast To FortPlayerAnimInstance(Object=GetAnimInstance(self=GetOwnerSkeletalMeshComponent()))
   > n17 = Switch on EFortWeaponCoreAnimation!(execute, Selection:EFortWeaponCoreAnimation Enum=Get WeaponCoreAnim(self=@n16)) -> (
    Melee:Exec
    Pistol:Exec
    Shotgun:Exec
    PaperBlueprint:Exec
    Rifle:Exec
    MeleeOneHand:Exec
    MachinePistol:Exec
    RocketLauncher:Exec
    GrenadeLauncher:Exec
    GoingCommando:Exec
    AssaultRifle:Exec
    TacticalShotgun:Exec
    SniperRifle:Exec
    TrapPlacement:Exec
    ShoulderLauncher:Exec
    AbilityDecoTool:Exec
    Crossbow:Exec
    C4:Exec
    RemoteControl:Exec
    DualWield:Exec
    AR_BullPup:Exec
    AR_ForwardGrip:Exec
    MedPackPaddles:Exec
    SMG_P90:Exec
    AR_DrumGun:Exec
    Consumable_Small:Exec
    Consumable_Large:Exec
    Balloon:Exec
    MountedTurret:Exec
    CreativeTool:Exec
    ExplosiveBow:Exec
    AshtonIndigo:Exec
    AshtonChicago:Exec
    MeleeDualWield:Exec
    Unarmed:Exec
   )
   Set AlphaShotgun(
    execute=[
     @n17.AssaultRifle
     @n17.GrenadeLauncher
     @n17.MachinePistol
     @n17.Melee
     @n17.MeleeOneHand
     @n17.PaperBlueprint
     @n17.Pistol
     @n17.Rifle
     @n17.RocketLauncher
     @n17.Shotgun
    ]
    AlphaShotgun:float=0
   ) -> (Output_Get:float)
   Branch(execute=@n17.GoingCommando)
   > Set AlphaShotgun(AlphaShotgun=1)
   n18 = Sequence!(execute=@n6.then_2) -> (
    then_0
    then_1
    then_2
    then_3
    then_4
    then_5
    then_6:Exec
    then_7:Exec
    then_8:Exec
    then_9:Exec
    then_10:Exec
    then_11:Exec
    then_12:Exec
    then_13:Exec
   )
   n19 = Branch(execute=@n18.then_11, Condition=Get bIsTacticalSprinting())
   > Set SprintConst(SprintConst:float=0.5) -> (Output_Get:float)
   Set SprintConst(execute=@n19.else, SprintConst=1)
   > n20 = Branch(Condition=Get DoubleDutyEmote() -> (DoubleDutyEmote:bool))
   > Set SprintConst(SprintConst=0)
   Set SprintConst(execute=@n20.else, SprintConst=1)
   n21 = Get LeftBool() -> (LeftBool:bool)
   n22 = Get RightBool() -> (RightBool:bool)
   n23 = Branch(execute=@n18.then_12, Condition=AND Boolean!(A=Get bIsCrouching(), B=Get bIsMoving2D(), C:bool=OR Boolean(A=@n21, B=@n22)) -> (ReturnValue))
   > Set sideToSideCrouch(sideToSideCrouch:double=1) -> (Output_Get:double)
   Set sideToSideCrouch(execute=@n23.else, sideToSideCrouch=0)
   n24 = Sequence!(execute=@n18.then_13) -> (then_0, then_1, then_2)
   n25 = GetAnimInstance(self=GetOwnerSkeletalMeshComponent())
   > Is Valid(InputObject=@n25)
   n26 = Montage_IsPlaying(self=@n25, Montage=/BRCosmetics/Animation/Game/MainPlayer/Emotes/DoubleDuty/CMM/Emotes_DoubleDuty_CMM_M)
   > Set DoubleDutyEmote(DoubleDutyEmote:bool=@n26) -> (Output_Get:bool)
   Is Valid(exec=@n24.then_1, InputObject=@n25)
   > Set isCustomEmotePlaying(isCustomEmotePlaying:double=SelectFloat(A=1, B=0, bPickA=@n26)) -> (Output_Get:double)
   n27 = Branch(
    execute=@n24.then_2
    Condition=OR Boolean!(A=Get bIsSprinting(), B=Get bIsTacticalSprinting(), C:bool=Get DoubleDutyEmote()) -> (ReturnValue)
   )
   > Set Sprint general(Sprint general:bool=true) -> (Output_Get:bool)
   n28 = Get DoubleDutyEmote()
   > Branch(Condition=AND Boolean!(A=Get bIsSprinting(), B=NOT Boolean(A=@n28)) -> (ReturnValue))
   > Set Runningnot Emote(Runningnot Emote:bool=true) -> (Output_Get:bool)
   Set Sprint general(execute=@n27.else, Sprint general=false)
   > Set Runningnot Emote(Runningnot Emote=false)
   n29 = PlayerMovement(execute=@n18.then_5, AnimInstance=Self-Reference(), XDotProductCoefficient=100, YDotProductCoefficient=100, ZDotProductCoefficient=100)
   n30 = BreakVector(InVec=MakeVector(X=@n29, Y=@n29.YMovement))
   n31 = "float < float"(A=@n30, B=-50)
   > Set Left(Left:double="To Float (Boolean)"(InBool=@n31)) -> (Output_Get:double)
   > Set LeftBool(LeftBool:bool=@n31) -> (Output_Get:bool)
   n32 = "float > float"(A=@n30, B=50)
   > Set Right(Right:double="To Float (Boolean)"(InBool=@n32)) -> (Output_Get:double)
   > Set RightBool(RightBool:bool=@n32) -> (Output_Get:bool)
   n33 = "float > float"(A=@n30.Y, B=40)
   > Set Forward(Forward:double="To Float (Boolean)"(InBool=@n33)) -> (Output_Get:double)
   > Set forwardBool(forwardBool:bool=@n33) -> (Output_Get:bool)
   n34 = "float < float"(A=@n30.Y, B=-40)
   > Set Backward(Backward:double="To Float (Boolean)"(InBool=@n34)) -> (Output_Get:double)
   > Set BackwardBool(BackwardBool:bool=@n34) -> (Output_Get:bool)
   n35 = Sequence!(execute=@n18.then_9) -> (then_0, then_1)
   n36 = Branch(execute=@n35.then_1, Condition=AND Boolean(A=Get bIsSkydiveFloating(), B=Get RightBool()))
   > Set rightSkydiving(rightSkydiving:double=1) -> (Output_Get:double)
   Set rightSkydiving(execute=@n36.else, rightSkydiving=0)
   n37 = Branch(execute=@n18.then_3, Condition=Get bRigidBodyEnabled())
   > Set RBANalpha(RBANalpha:double=1) -> (Output_Get:double)
   > Set rbanOff(rbanOff:double=0) -> (Output_Get:double)
   Set RBANalpha(execute=@n37.else, RBANalpha=0)
   > Set rbanOff(rbanOff=1)
   n38 = Branch(execute=@n18.then_4, Condition=OR Boolean!(A=Get bIsJumping(), B=Get bIsFalling()) -> (ReturnValue))
   > Set jumpingAlpha(jumpingAlpha:double=1) -> (Output_Get:double)
   Set jumpingAlpha(execute=@n38.else, jumpingAlpha=0)
   n39 = Sequence!(execute=@n18.then_7) -> (then_0, then_1, then_2:Exec, then_3:Exec, then_4:Exec, then_5:Exec)
   n40 = NOT Boolean(A=Get bIsCrouching())
   n41 = Branch(execute=@n39.then_3, Condition=@n40)
   > Set notCrouch(notCrouch:double=1) -> (Output_Get:double)
   Set notCrouch(execute=@n41.else, notCrouch=0)
   n42 = Get bIsPlayingEmote()
   n43 = Branch(execute=@n39.then_5, Condition=@n42)
   > Set isPlayingEmote(isPlayingEmote:double=1) -> (Output_Get:double)
   Set isPlayingEmote(execute=@n43.else, isPlayingEmote=0)
   n44 = NOT Boolean(A=OR Boolean!(A=Get bIsSkydiveFloating(), B=Get bIsSkydiveDiveMode(), C:bool=Get bIsSkyDiving()) -> (ReturnValue))
   n45 = Get bIsMoving2D()
   Set WalkNotCrouch(execute=@n39.then_0, WalkNotCrouch:double="To Float (Boolean)"(InBool=AND Boolean!(A=@n45, B=@n40, C:bool=@n44) -> (ReturnValue))) -> (Output_Get:double)
   n46 = AND Boolean!(A=Get bIsCrouching(), B=NOT Boolean(A=Get bIsMoving2D())) -> (ReturnValue)
   > Set CrouchNotMove(CrouchNotMove:double="To Float (Boolean)"(InBool=@n46)) -> (Output_Get:double)
   n47 = Get bIsMoving2D()
   n48 = Get bIsCrouching()
   > Set CrouchMove(CrouchMove:double="To Float (Boolean)"(InBool=AND Boolean(A=@n48, B=@n47))) -> (Output_Get:double)
   > n49 = Branch(Condition=@n46)
   n50 = Branch(execute=@n49.else, Condition=Get bIsTacticalSprinting())
   > Set sprintingAlpha(sprintingAlpha:double=1) -> (Output_Get:double)
   Set sprintingAlpha(execute=@n50.else, sprintingAlpha=0)
   n51 = BreakTransform(InTransform=GetDeltaTransformFromRefPose(self=GetOwnerSkeletalMeshComponent(), BoneName=thigh_l, BaseName=pelvis))
   Set thigh_rot_up_l(execute=@n39.then_2, thigh_rot_up_l:double=MapRangeClamped(Value=BreakRotator(InRot=@n51.Rotation).Yaw, InRangeA=-20, InRangeB=-100)) -> (Output_Get:double)
   n52 = AND Boolean!(
    A=Get bIsSprinting()
    B=Get forwardBool() -> (forwardBool:bool)
    C:bool=NOT Boolean(A=Get bIsCrouching())
    D:bool=NOT Boolean(A=Get bIsSurfaceSwimming())
    E:bool=NOT Boolean(A=Get bIsJumping())
    F:bool=NOT Boolean(A=Get bIsFalling())
   ) -> (ReturnValue)
   > Set RunningSkydive(RunningSkydive:bool=OR Boolean(A=Get bIsSkyDiving(), B=@n52, C=Get DoubleDutyEmote())) -> (Output_Get:bool)
   > n53 = Branch(Condition=@n52)
   > Set runningStrap(runningStrap:double=1) -> (Output_Get:double)
   Set runningStrap(execute=@n53.else, runningStrap=0)
   n54 = Get forwardBool()
   n55 = AND Boolean!(A=Get LeftBool(), B=@n54) -> (ReturnValue)
   n56 = AND Boolean(A=@n54, B=Get RightBool())
   n57 = Get BackwardBool() -> (BackwardBool:bool)
   n58 = AND Boolean(A=Get LeftBool(), B=@n57)
   n59 = AND Boolean(A=@n57, B=Get RightBool())
   n60 = AND Boolean!(
    A=Get bIsCrouching()
    B=Get bIsMoving2D()
    C:bool=OR Boolean!(A=Get forwardBool(), B=Get BackwardBool()) -> (ReturnValue)
    D:bool=NOT Boolean(A=OR Boolean!(A=@n55, B=@n56, C:bool=@n58, D:bool=@n59) -> (ReturnValue))
   ) -> (ReturnValue)
   > Set CrouchMovingCTRL(CrouchMovingCTRL:double=FInterpTo(Current=Get CrouchMovingCTRL(), Target="To Float (Boolean)"(InBool=@n60), DeltaTime=@n5.DeltaTimeX, InterpSpeed=1.5)) -> (Output_Get:double)
   > n61 = Branch(Condition=AND Boolean!(A=Get bIsDBNO(), B=Get bIsMoving2D()) -> (ReturnValue))
   > Set DBNOMoving(DBNOMoving:double=1) -> (Output_Get:double)
   Set DBNOMoving(execute=@n61.else, DBNOMoving=0)
   Set DBNOCrouchBool(execute=@n39.then_1, DBNOCrouchBool:bool=OR Boolean(A=@n48, B=Get bIsDBNO(), C=Get bIsJumping(), D=Get bIsFalling())) -> (Output_Get:bool)
   n62 = Get bIsCrouching()
   n63 = Get bIsMoving2D()
   > n64 = Branch(Condition=NOT Boolean(A=OR Boolean!(A=@n63, B=@n62) -> (ReturnValue)))
   > Set idleAlpha(idleAlpha:double=1) -> (Output_Get:double)
   Set idleAlpha(execute=@n64.else, idleAlpha=0)
   n65 = GetOwnerSkeletalMeshComponent()
   n66 = BreakRotator(InRot=BreakTransform(InTransform=GetDeltaTransformFromRefPose(self=@n65, BoneName=pelvis, BaseName=root)).Rotation)
   n67 = Branch(execute=@n39.then_4, Condition=AND Boolean(A="float > float"(A=@n66.Yaw, B=23), B=@n42))
   > Set isDetectedRotationInEmote(isDetectedRotationInEmote:double=1) -> (Output_Get:double)
   Set isDetectedRotationInEmote(execute=@n67.else, isDetectedRotationInEmote=0)
   n68 = Get bIsParachuteOpened()
   n69 = Get bIsSkydiveFloating()
   n70 = Get bIsSkydiveDiveMode()
   n71 = Get bIsSkyDiving()
   n72 = Branch(execute=@n18.then_6, Condition=AND Boolean(A=OR Boolean!(A=@n69, B=@n70, C:bool=@n71) -> (ReturnValue), B=NOT Boolean(A=@n68)))
   > Set SkydiveAlpha(SkydiveAlpha:double=1) -> (Output_Get:double)
   Set SkydiveAlpha(execute=@n72.else, SkydiveAlpha=0)
   > n73 = Branch(Condition=Get bIsPlayingMeleeAnim())
   > Set MeeleAttackAlpha(MeeleAttackAlpha:double=1) -> (Output_Get:double)
   Set MeeleAttackAlpha(execute=@n73.else, MeeleAttackAlpha=0)
   > n74 = Sequence!(execute) -> (then_0, then_1)
   n75 = OR Boolean!(A=Get LeftBool(), B=Get RightBool()) -> (ReturnValue)
   n76 = Get forwardBool()
   n77 = Get BackwardBool()
   > n78 = Branch(Condition=AND Boolean!(A=Get bIsCrouching(), B=Get bIsMoving2D(), C:bool=OR Boolean(A=@n76, B=@n77), D:bool=NOT Boolean(A=@n75)) -> (ReturnValue))
   > Set crouchAiming(crouchAiming:double=1) -> (Output_Get:double)
   > n79 = Branch(Condition=AND Boolean!(A=Get bIsCrouching(), B=Get bIsMoving2D(), C=Get BackwardBool()) -> (ReturnValue))
   > Set BackCrouching(BackCrouching:double=1) -> (Output_Get:double)
   > Set crouchAiming(crouchAiming=0.5)
   Set BackCrouching(execute=@n79.else, BackCrouching=0)
   Set crouchAiming(execute=@n78.else, crouchAiming=0)
   n80 = Branch(execute=@n74.then_1, Condition=AND Boolean!(A=Get bIsCrouching(), B=Get bIsMoving2D()) -> (ReturnValue))
   > Delay(Duration=0.5)
   > Set beltCrouchMove(beltCrouchMove:double=1) -> (Output_Get:double)
   Set beltCrouchMove(execute=@n80.else, beltCrouchMove=0)
   n81 = Branch(execute=@n18.then_0, Condition=Get bIsParachuteOpened())
   > Set parachuteAlpha(parachuteAlpha:double=1) -> (Output_Get:double)
   Set parachuteAlpha(execute=@n81.else, parachuteAlpha=0)
   n82 = OR Boolean(A=Get bIsCrouching(), B=Get bIsCrouchSprinting())
   n83 = Branch(execute=@n18.then_1, Condition=@n82)
   > Set crouchAlpha(crouchAlpha:double=1) -> (Output_Get:double)
   Set crouchAlpha(execute=@n83.else, crouchAlpha=0)
   n84 = Get bRigidBodyEnabled()
   n85 = Branch(execute=@n18.then_2, Condition=AND Boolean(A=@n82, B=NOT Boolean(A=@n84)))
   > Set CrouchRBANoff(CrouchRBANoff:double=1) -> (Output_Get:double)
   Set CrouchRBANoff(execute=@n85.else, CrouchRBANoff=0)
   n86 = Branch(execute=@n6.then_4, Condition=Get bIsSkyDiving())
   > Set AccScaleVals(AccScaleVals:Vector="0.1,0,0") -> (Output_Get:Vector)
   Set AccScaleVals(execute=@n86.else, AccScaleVals="0.15,0.3,0.2")
   n87 = Sequence!(execute=@n6.then_5) -> (then_0, then_1, then_2:Exec, then_3:Exec, then_4:Exec)
   n88 = float * float(
    A=GetAlphaFromReferenceTransformDelta(SkeletalMeshComponent=Get Skel Mesh(), BoneName=pelvis, BaseName=thigh_l, InputRange=(X=0,Y=50), TwistAxis="0,0,1")
    B=1.15
   ) {is=title K2Node_PromotableOperator}
   Set Leg_L_Rotate_Bool(execute=@n87.then_1, Leg_L_Rotate_Bool:bool="To Boolean (Integer)"!(InInt:int=Truncate(A=@n88)) -> (ReturnValue:bool)) -> (Output_Get:bool)
   n89 = Get Leg_L_Rotate_Bool() -> (Leg_L_Rotate_Bool:bool)
   n90 = AND Boolean!(
    A=@n89
    B=NOT Boolean(A=Get bIsCrouching())
    C:bool=NOT Boolean(A=Get bIsJumping())
    D:bool=NOT Boolean(A=Get bIsFalling())
    E:bool=NOT Boolean(A=Get bIsCrouchSprinting())
    F:bool=NOT Boolean(A=Get bIsSprinting())
    G:bool=NOT Boolean(A=Get bIsDBNO())
   ) -> (ReturnValue)
   n91 = Branch(execute=@n87.then_2, Condition=@n90)
   > Set On Rail(On Rail:double=1) -> (Output_Get:double)
   Set On Rail(execute=@n91.else, On Rail=0)
   n92 = Get BackwardBool()
   n93 = Get forwardBool()
   > n94 = Branch(Condition=AND Boolean!(A=OR Boolean(A=@n92, B=@n93), B=Get bIsDBNO(), C=@n89) -> (ReturnValue))
   > Set dbnoFixes(dbnoFixes:double=1) -> (Output_Get:double)
   Set dbnoFixes(execute=@n94.else, dbnoFixes=0)
   Set Surf_glider(execute=@n87.then_0, Surf_glider:bool=false) -> (Output_Get:bool)
   n95 = Get fort_player() -> (fort_player:Fort Player Anim Instance Object Reference)
   > n96 = Switch on EGliderType(Selection=Break Fort Anim Input Player Glider Anim Asset!(FortAnimInput_PlayerGliderAnimAsset:"Fort Anim Input Player Glider Anim Asset Structure (by ref)"=Get CurrentGliderAnimSet!(self:Fort Player Anim Instance Object Reference=@n95) -> (CurrentGliderAnimSet:Fort Anim Input Player Glider Anim Asset Structure)) -> (
     Default_Base_BS:Blend Space Object Reference
     Default_BodyAdditive_MaleMedium_BS:Blend Space Object Reference
     Default_BodyAdditive_MaleLarge_BS:Blend Space Object Reference
     Default_BodyAdditive_FemaleSmall_BS:Blend Space Object Reference
     Default_BodyAdditive_FemaleMedium_BS:Blend Space Object Reference
     Default_BodyAdditive_FemaleLarge_BS:Blend Space Object Reference
     Default_TurnAdditive_MaleMedium_BS:Blend Space Object Reference
     Default_TurnAdditive_MaleLarge_BS:Blend Space Object Reference
     Default_TurnAdditive_FemaleSmall_BS:Blend Space Object Reference
     Default_TurnAdditive_FemaleMedium_BS:Blend Space Object Reference
     Default_TurnAdditive_FemaleLarge_BS:Blend Space Object Reference
     Into_Base_BS:Blend Space Object Reference
     Into_Base_MaleMedium_BS:Blend Space Object Reference
     Into_BodyAdditive_MaleMedium_BS:Blend Space Object Reference
     Into_BodyAdditive_MaleLarge_BS:Blend Space Object Reference
     Into_BodyAdditive_FemaleSmall_BS:Blend Space Object Reference
     Into_BodyAdditive_FemaleMedium_BS:Blend Space Object Reference
     Into_BodyAdditive_FemaleLarge_BS:Blend Space Object Reference
     Into_TurnAdditive_MaleMedium_BS:Blend Space Object Reference
     Into_TurnAdditive_MaleLarge_BS:Blend Space Object Reference
     Into_TurnAdditive_FemaleSmall_BS:Blend Space Object Reference
     Into_TurnAdditive_FemaleMedium_BS:Blend Space Object Reference
     Into_TurnAdditive_FemaleLarge_BS:Blend Space Object Reference
     Lean_MaleMedium_BS:Blend Space Object Reference
     Lean_MaleLarge_BS:Blend Space Object Reference
     Lean_FemaleSmall_BS:Blend Space Object Reference
     Lean_FemaleMedium_BS:Blend Space Object Reference
     Lean_FemaleLarge_BS:Blend Space Object Reference
     LeanAdditive_Center_MaleMedium_Pose:Anim Sequence Object Reference
     LeanAdditive_Center_MaleLarge_Pose:Anim Sequence Object Reference
     LeanAdditive_Center_FemaleSmall_Pose:Anim Sequence Object Reference
     LeanAdditive_Center_FemaleMedium_Pose:Anim Sequence Object Reference
     LeanAdditive_Center_FemaleLarge_Pose:Anim Sequence Object Reference
     LeanAdditive_Into_BS:Blend Space Object Reference
     LeanAdditive_ForwardInto_Anim:Anim Sequence Object Reference
     LeanAdditive_ForwardInto_FromDeploy_Anim:Anim Sequence Object Reference
     LeanAdditive_BackInto_Anim:Anim Sequence Object Reference
     LeanAdditive_LeftInto_Anim:Anim Sequence Object Reference
     LeanAdditive_RightInto_Anim:Anim Sequence Object Reference
     LeanAdditive_ForwardOut_Anim:Anim Sequence Object Reference
     LeanAdditive_BackOut_Anim:Anim Sequence Object Reference
     LeanAdditive_LeftOut_Anim:Anim Sequence Object Reference
     LeanAdditive_RightOut_Anim:Anim Sequence Object Reference
     ToGlide_BS:Blend Space Object Reference
     ToGlide_Lean_BS:Blend Space Object Reference
     ToDive_BS:Blend Space Object Reference
     ToDive_Lean_BS:Blend Space Object Reference
     Dive_WeaponR_Additive_BS:Blend Space Object Reference
     Glide_WeaponR_Additive_Anim:Anim Sequence Object Reference
     GenericAdditive_Male_BS:Blend Space Object Reference
     GenericAdditive_Female_BS:Blend Space Object Reference
     RootModPitchMin:float=~
     RootModPitchMax:float=~
     RootModYOffsetMin:float=~
     RootModYOffsetMax:float=~
     PelvisModVertical:Pelvis Mod Vertical Input Spring Structure
     PelvisModLateral:Pelvis Mod Lateral Input Spring Structure
     PelvisModBankAngle:Pelvis Mod Bank Angle Input Spring Structure
     ThighModLegAngle:Thigh Mod Leg Angle Input Structure
     ThighModLegPitch:Thigh Mod Leg Pitch Spring Structure
     ThighModLegBank:Thigh Mod Leg Bank Spring Structure
     ThighModLeftLegDrag:Thigh Mod Left Leg Drag Structure
     PlayerGliderType:EGliderType Enum
     bEnableSpringMods:bool=~
     bAllowPlayerDeployRootMod:bool=~
     bUseSurfStyle:bool=~
     bUseLeanPlayRateAdjust:bool=~
     bUseForwardAs_Y_BS:bool=~
     bUseDeployCurveForIntoPlayRateAdjust:bool=~
     LeanForwardValue:float=~
     IdleValue:float=~
     LeanBackValue:float=~
     IdleInterpToSpeed:float=~
     LeanForwardInterpToSpeed:float=~
     LeanBackInterpToSpeed:float=~
    ) {is=title K2Node_BreakStruct}.PlayerGliderType)
   n97 = Set Surf_glider(execute=[@n96.Cape, @n96.HangGlider, @n96.NoGlider, @n96.Stairs, @n96.Surfing, @n96.Umbrella], Surf_glider=false)
   n98 = Branch(
    execute=[@n97.then, @n99.then]
    Condition=AND Boolean!(A=Get bIsParachuteOpened(), B=Get Surf_glider() -> (Surf_glider:bool)) -> (ReturnValue)
   )
   > Set DoubleDutyGliderAlpha(DoubleDutyGliderAlpha:bool=1) -> (Output_Get:bool)
   Set DoubleDutyGliderAlpha(execute=@n98.else, DoubleDutyGliderAlpha=false)
   n99 = Set Surf_glider(execute=@n96.SurfingSimple, Surf_glider=true)
   n100 = NOT Boolean(A=Get bIsDBNO())
   n101 = NOT Boolean(A=Get Leg_L_Rotate_Bool())
   n102 = NOT Boolean(A=Get bIsCrouching())
   n103 = NOT Boolean(A=Get bIsJumping())
   n104 = NOT Boolean(A=Get bIsCrouchSprinting())
   n105 = NOT Boolean(A=Get bIsSprinting())
   n106 = NOT Boolean(A=Get bIsFalling())
   n107 = OR Boolean!(A=Get forwardBool(), B=Get BackwardBool(), C:bool) -> (ReturnValue)
   Set WalkingAlpha(
    execute=@n87.then_3
    WalkingAlpha:double="To Float (Boolean)"(InBool=AND Boolean!(A=@n107, B=@n102, C:bool=@n103, D:bool=@n106, E:bool=@n104, F:bool=@n105, G:bool=@n100, H:bool=@n101) -> (ReturnValue))
   ) -> (Output_Get:double)
   n108 = NOT Boolean(A=Get bIsPlayingEmote())
   n109 = Get bIsInFrontEnd()
   Set FrontEndFix(execute=@n87.then_4, FrontEndFix:double="To Float (Boolean)"(InBool=AND Boolean!(A=@n109, B=@n108) -> (ReturnValue))) -> (Output_Get:double)
   Event OnCharacterCustomizationLoaded()
   > n110 = Set SkelMesh(SkelMesh=GetPartSkeletalMeshComponent())
   > n111 = Get Dynamic Material Instance(SkeletalMeshComponent=@n110)
   > Set Mat Instance Dynamic(Mat Instance Dynamic=@n111)
   > n112 = GetOrCreateDynamicOverlayMaterial(MeshComponent=@n110, MaterialName=None)
   > Set Overlay Mat Instance Dynamic(Overlay Mat Instance Dynamic=@n112)
   "To Float (Boolean)"(InBool=@n90)
   Get bIsInFrontEnd()
   Get bIsTargeting()
  ]}
]
variables [
 AccScaleVals {type=Vector}
 AlphaShotgun {type=number}
 BackCrouching {type=number}
 Backward {type=number}
 BackwardBool {type=boolean}
 beltCrouchMove {type=number}
 Body_Mesh {type=SkeletalMeshComponent}
 Crouch Alpha {type=number}
 crouchAiming {type=number}
 crouchAlpha {type=number}
 CrouchMove {type=number}
 CrouchMovingCTRL {type=number, default=1}
 CrouchNotMove {type=number}
 CrouchRBANoff {type=number}
 DBNOCrouchBool {type=boolean}
 dbnoFixes {type=number}
 DBNOMoving {type=number}
 DoubleDutyEmote {type=boolean}
 DoubleDutyGliderAlpha {type=boolean}
 fort_player {type=FortPlayerAnimInstance, default=null}
 Forward {type=number}
 forwardBool {type=boolean}
 FrontEndFix {type=number}
 Gravity Override Params {type=GravityOverrideParamsStruct}
 GravityOverride {type=Vector}
 idleAlpha {type=number}
 isCustomEmotePlaying {type=number}
 isDetectedRotationInEmote {type=number}
 isPlayingEmote {type=number}
 jumpingAlpha {type=number}
 Left {type=number}
 LeftBool {type=boolean}
 Leg_L_Rotate_Bool {type=boolean}
 Local Noise Offset {type=Vector}
 Local Noise Transform Location {type=Vector}
 Mat Instance Dynamic {type=MaterialInstanceDynamic}
 MeeleAttackAlpha {type=number}
 notCrouch {type=number}
 On Rail {type=number}
 Overlay Mat Instance Dynamic {type=MaterialInstanceDynamic}
 parachuteAlpha {type=number}
 PlayerPawn {type=FortPlayerPawnAthena}
 R1 Velocity Transform Location {type=Vector}
 R1 Velocity Transform Velocity {type=Vector}
 R2 Velocity Transform Location {type=Vector}
 R2 Velocity Transform Velocity {type=Vector}
 RBANalpha {type=number}
 rbanOff {type=number}
 Right {type=number}
 RightBool {type=boolean}
 rightSkydiving {type=number}
 Runningnot Emote {type=boolean}
 RunningSkydive {type=boolean}
 runningStrap {type=number}
 sideToSideCrouch {type=number}
 Skel Mesh {type=SkeletalMeshComponent}
 Skel_Mesh {type=SkeletalMeshComponent}
 SkelMesh {type=SkeletalMeshComponent}
 SkydiveAlpha {type=number}
 Sprint general {type=boolean}
 SprintConst {type=number, default=1}
 sprintingAlpha {type=number}
 Surf_glider {type=boolean}
 thigh_rot_up_l {type=number}
 UpperArmRot_l {type=number}
 Velocity Interp Speed {type=number, default=30}
 WalkingAlpha {type=number}
 WalkNotCrouch {type=number}
 WPO Pho Turbulence {type=number}
 WPO Pho Wind Intensity {type=number}
]
